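# Animation Blueprint for the player character driving the Ostrich mech, supplying the VehicleFinalPoseOverride layer on the Fortnite player skeleton.
# A cached driver idle gets an additive legs-bounce from a single-state machine, with play rate and alpha read from variables not declared here.
# bIsFemale selects that pose with or without an additive CMF pose correction, then feeds the FullBody slot.
# Other override layers and the event graph have no logic.
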AnimBlueprint /Game/Athena/DrivableVehicles/Mech/Animation/Ostrich_Driver_AnimBP
target_skeleton=&Fortnite_M_Avg_Player_Skeleton
graphs [
 AnimGraph {nodes [Output Pose() {layerGroup=None}]}
 EventGraph {links [], nodes []}
 VehicleFinalPoseOverride {nodes [
   Input Pose() {name=InFinalPose}
   n1 = Apply Additive(
    Base=Use cached pose 'IdlePose'()
    Additive="Sequence Player (additive) 'OstrichVehicle_Driver_Pose_CMF'"() {sequence=/Game/Animation/Game/MainPlayer/Vehicles/OstrichVehicle/OstrichVehicle_Driver_Pose_CMF}
   ) {alphaBoolBlend {blendOption=Linear}}
   n2 = Use cached pose 'IdlePose'()
   n3 = Get bIsFemale() -> (bIsFemale:bool)
   Output Pose(Result=Slot 'FullBody'(Source=Blend Poses by bool(BlendPose_0=@n1, BlendPose_1=@n2, bActiveValue=@n3) {blendType=Linear}) {slotName=FullBody}) {layerGroup=VehicleOverride}
   AnimGraphNode_StateMachine_0 = LegsBounce() {is=title AnimGraphNode_StateMachine}
   Save cached pose 'IdlePose'(Pose=Apply Additive(
     Base=Sequence Player 'OstrichVehicle_Driver_Idle_CMM'() {sequence=/Game/Animation/Game/MainPlayer/Vehicles/OstrichVehicle/OstrichVehicle_Driver_Idle_CMM}
     Additive=@AnimGraphNode_StateMachine_0
     Alpha=Get LegsBounceAlphaValue() -> (LegsBounceAlphaValue:float)
    ) {alphaBoolBlend {blendOption=Linear}}) {cacheName=IdlePose}
  ]}
 LegsBounce {
  entry=BounceAdditive
  path=VehicleFinalPoseOverride.AnimGraphNode_StateMachine_0.LegsBounce
  ids=id
  nodes [AnimStateNode_0 = BounceAdditive() {is=title AnimStateNode, bAlwaysResetOnEntry=true}]
 }
 BounceAdditive {
  path=VehicleFinalPoseOverride.AnimGraphNode_StateMachine_0.LegsBounce.AnimStateNode_0.BounceAdditive
  nodes [Output Animation Pose(Result="Sequence Player (additive) 'OstrichVehicle_Driver_Bounce_CMM'"!(PlayRate:float=Get BouncePlayRateValue() -> (BouncePlayRateValue:float)) -> (Pose) {sequence=/Game/Animation/Game/MainPlayer/Vehicles/OstrichVehicle/OstrichVehicle_Driver_Bounce_CMM})]
 }
 VehicleFullBodyOverride {nodes [
   Input Pose() {name=InPoseFullBody}
   Output Pose() {layerGroup=VehicleOverride}
  ]}
 VehicleLowerBodyOverride {nodes [
   Input Pose() {name=InPoseLowerBodyDefault}
   Input Pose() {name=InPoseUpperBody}
   Output Pose() {layerGroup=VehicleOverride}
  ]}
 VehicleSplitBodyOverride {nodes [
   Input Pose() {name=InPoseSplitBody}
   Input Pose() {name=InPoseUpperAndLowerBody}
   Input Pose() {name=InPoseUpperAndLowerBodyRemoveRoot}
   Output Pose() {layerGroup=Vehicle Override}
  ]}
]
variables []
