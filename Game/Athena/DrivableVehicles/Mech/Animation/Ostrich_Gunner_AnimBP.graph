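# Player-skeleton anim blueprint for the Ostrich mech gunner, overriding the vehicle final pose layer.
# An aim-driven lean plus idle additive feeds a directional impact-reaction state machine, with rocket and shotgun fire layered onto left and right arms.
# Two mirrored recoil states restart the body additive on each fire-flag toggle.
# A gender additive, FullBody slot and hand and foot Leg IK finish the pose.
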AnimBlueprint /Game/Athena/DrivableVehicles/Mech/Animation/Ostrich_Gunner_AnimBP
target_skeleton=&Fortnite_M_Avg_Player_Skeleton
graphs [
 AnimGraph {nodes [Output Pose() {layerGroup=None}]}
 EventGraph {links [], nodes []}
 VehicleFinalPoseOverride {nodes [
   Input Pose() {name=InFinalPose}
   n1 = "Blend Poses (EFortCustomGender)"!(
    BlendPose_0:Pose Link Structure=Use cached pose 'PostMaskedPose'()
    BlendPose_1:Pose Link Structure=Apply Additive(
     Base=Use cached pose 'PostMaskedPose'()
     Additive="Sequence Player (additive) 'OstrichVehicle_Passenger_Idle_Pose_CMF'"() {sequence=/Game/Animation/Game/MainPlayer/Vehicles/OstrichVehicle/OstrichVehicle_Passenger_Idle_Pose_CMF}
    ) {alphaBoolBlend {blendOption=Linear}}
    BlendTime_0:float=0.1
    BlendTime_1:float=0.1
    ActiveEnumValue:Byte=Get Gender() -> (Gender:EFortCustomGender Enum=Invalid)
   ) -> (Pose:Pose Link Structure) {
    is=title AnimGraphNode_BlendListByEnum
    blendType=Linear
    blendPose [{linkId=-1, sourceLinkId=-1}, {linkId=-1, sourceLinkId=-1}]
    blendTime [0.1,0.1]
   }
   n2 = Leg IK!(
    ComponentPose:Component Space Pose Link Structure=Local To Component(LocalPose=Slot 'FullBody'(Source=@n1) {slotName=FullBody})
    Alpha:float=Get LeftHandIKAlpha() -> (LeftHandIKAlpha:float)
   ) -> (Pose:Component Space Pose Link Structure) {
    is=title AnimGraphNode_LegIK
    alphaBoolBlend {blendOption=Linear}
    legsDefinition [{
      bEnableKneeTwistCorrection=false
      bEnableRotationLimit=false
      footBoneForwardAxis=Y
      hingeRotationAxis=Z
      minRotationAngle=15
      numBonesInLimb=2
      twistOffsetCurveName=None
      fKFootBone {boneName=hand_l}
      iKFootBone {boneName=ik_hand_l}
     }]
   }
   n3 = Leg IK(ComponentPose=@n2, Alpha=Get RightHandIKAlpha() -> (RightHandIKAlpha:float)) {
    alphaBoolBlend {blendOption=Linear}
    legsDefinition [{
      bEnableKneeTwistCorrection=false
      bEnableRotationLimit=false
      footBoneForwardAxis=Y
      hingeRotationAxis=Z
      minRotationAngle=15
      numBonesInLimb=2
      twistOffsetCurveName=None
      fKFootBone {boneName=hand_r}
      iKFootBone {boneName=ik_hand_r}
     }]
   }
   n4 = Leg IK(ComponentPose=@n3, Alpha=Get LegIKAlpha() -> (LegIKAlpha:float)) {
    alphaBoolBlend {blendOption=Linear}
    legsDefinition [
     {
      bEnableKneeTwistCorrection=false
      bEnableRotationLimit=false
      footBoneForwardAxis=Y
      hingeRotationAxis=Z
      minRotationAngle=15
      numBonesInLimb=2
      twistOffsetCurveName=None
      fKFootBone {boneName=foot_l}
      iKFootBone {boneName=ik_foot_l}
     }
     {
      bEnableKneeTwistCorrection=false
      bEnableRotationLimit=false
      footBoneForwardAxis=Y
      hingeRotationAxis=Z
      minRotationAngle=15
      numBonesInLimb=2
      twistOffsetCurveName=None
      fKFootBone {boneName=foot_r}
      iKFootBone {boneName=ik_foot_r}
     }
    ]
   }
   Output Pose(Result=Component To Local(ComponentPose=@n4)) {layerGroup=VehicleOverride}
   AnimGraphNode_StateMachine_0 = RocketFingers() {is=title AnimGraphNode_StateMachine}
   AnimGraphNode_StateMachine_2 = Gunner() {is=title AnimGraphNode_StateMachine}
   n5 = Layered blend per bone(BasePose=@AnimGraphNode_StateMachine_2, BlendPoses_0=@AnimGraphNode_StateMachine_0) {bMeshSpaceRotationBlend=true, layerSetup [{branchFilters [{boneName=upperarm_l}]}]}
   AnimGraphNode_StateMachine_3 = ShotgunFingers() {is=title AnimGraphNode_StateMachine}
   AnimGraphNode_StateMachine_1 = BodyRecoil() {is=title AnimGraphNode_StateMachine}
   Save cached pose 'PostMaskedPose'(Pose=Apply Additive(
     Base=Layered blend per bone(BasePose=@n5, BlendPoses_0=@AnimGraphNode_StateMachine_3) {bMeshSpaceRotationBlend=true, layerSetup [{branchFilters [{boneName=upperarm_r}]}]}
     Additive=@AnimGraphNode_StateMachine_1
     Alpha=Get BodyRecoilAlphaValue() -> (BodyRecoilAlphaValue:float)
    ) {alphaBoolBlend {blendOption=Linear}}) {cacheName=PostMaskedPose}
   n6 = Get AimYaw() -> (AimYaw:float)
   n7 = Get AimPitch()
   Save cached pose 'IdlePose'(Pose=Apply Additive(
     Base=Blendspace Player 'OstrichVehicle_Gunner_Lean_BS_CMM'(X=@n6, Y=@n7) {blendSpace=/Game/Animation/Game/MainPlayer/Vehicles/OstrichVehicle/OstrichVehicle_Gunner_Lean_BS_CMM}
     Additive="Sequence Player (additive) 'OstrichVehicle_Gunner_Idle_CMM'"() {sequence=/Game/Animation/Game/MainPlayer/Vehicles/OstrichVehicle/OstrichVehicle_Gunner_Idle_CMM}
    ) {alphaBoolBlend {blendOption=Linear}}) {cacheName=IdlePose}
  ]}
 RocketFingers {
  entry=Default
  path=VehicleFinalPoseOverride.AnimGraphNode_StateMachine_0.RocketFingers
  ids=id
  nodes [
   AnimStateNode_0 = Default() {is=title AnimStateNode, bAlwaysResetOnEntry=true}
   AnimStateNode_2 = RocketFire() {is=title AnimStateNode}
   AnimStateNode_3 = RocketFireEnd() {is=title AnimStateNode}
   AnimStateTransitionNode_1 = AnimStateTransitionNode() {crossfadeDuration=0}
   AnimStateTransitionNode_2 = AnimStateTransitionNode()
   AnimStateTransitionNode_3 = AnimStateTransitionNode() {bAutomaticRuleBasedOnSequencePlayerInState=true}
  ]
  links [
   {
    dst=RocketFire
    rule=VehicleFinalPoseOverride.AnimGraphNode_StateMachine_0.RocketFingers.AnimStateTransitionNode_1.Transition
    src=Default
    via=AnimStateTransitionNode_1
   }
   {
    dst=RocketFireEnd
    rule=VehicleFinalPoseOverride.AnimGraphNode_StateMachine_0.RocketFingers.AnimStateTransitionNode_2.Transition
    src=RocketFire
    via=AnimStateTransitionNode_2
   }
   {
    dst=Default
    rule=VehicleFinalPoseOverride.AnimGraphNode_StateMachine_0.RocketFingers.AnimStateTransitionNode_3.Transition
    src=RocketFireEnd
    via=AnimStateTransitionNode_3
   }
  ]
 }
 Default {
  path=VehicleFinalPoseOverride.AnimGraphNode_StateMachine_0.RocketFingers.AnimStateNode_0.Default
  nodes [Output Animation Pose(Result=Use cached pose 'IdlePose'())]
 }
 RocketFire {
  path=VehicleFinalPoseOverride.AnimGraphNode_StateMachine_0.RocketFingers.AnimStateNode_2.RocketFire
  nodes [Output Animation Pose(Result=Sequence Player 'OstrichVehicle_Gunner_Rocket_Fire_CMM'() {bLoopAnimation=false, sequence=/Game/Animation/Game/MainPlayer/Vehicles/OstrichVehicle/OstrichVehicle_Gunner_Rocket_Fire_CMM})]
 }
 RocketFireEnd {
  path=VehicleFinalPoseOverride.AnimGraphNode_StateMachine_0.RocketFingers.AnimStateNode_3.RocketFireEnd
  nodes [Output Animation Pose(Result=Sequence Player 'OstrichVehicle_Gunner_Rocket_FireEnd_CMM'() {bLoopAnimation=false, sequence=/Game/Animation/Game/MainPlayer/Vehicles/OstrichVehicle/OstrichVehicle_Gunner_Rocket_FireEnd_CMM})]
 }
 Transition {
  path=VehicleFinalPoseOverride.AnimGraphNode_StateMachine_0.RocketFingers.AnimStateTransitionNode_1.Transition
  nodes [Result(bCanEnterTransition=Get bIsRocketFiring() -> (bIsRocketFiring:bool))]
 }
 Transition {
  path=VehicleFinalPoseOverride.AnimGraphNode_StateMachine_0.RocketFingers.AnimStateTransitionNode_2.Transition
  nodes [Result(bCanEnterTransition=NOT Boolean(A=Get bIsRocketFiring()))]
 }
 Transition {path=VehicleFinalPoseOverride.AnimGraphNode_StateMachine_0.RocketFingers.AnimStateTransitionNode_3.Transition, nodes [Result()]}
 BodyRecoil {
  entry=ShotgunBodyRecoil_A
  path=VehicleFinalPoseOverride.AnimGraphNode_StateMachine_1.BodyRecoil
  ids=id
  nodes [
   AnimStateNode_0 = ShotgunBodyRecoil_A() {is=title AnimStateNode}
   AnimStateNode_1 = ShotgunBodyRecoil_B() {is=title AnimStateNode}
   AnimStateTransitionNode_1 = AnimStateTransitionNode() {crossfadeDuration=0}
   AnimStateTransitionNode_2 = AnimStateTransitionNode() {crossfadeDuration=0}
  ]
  links [
   {
    dst=ShotgunBodyRecoil_B
    rule=VehicleFinalPoseOverride.AnimGraphNode_StateMachine_1.BodyRecoil.AnimStateTransitionNode_1.Transition
    src=ShotgunBodyRecoil_A
    via=AnimStateTransitionNode_1
   }
   {
    dst=ShotgunBodyRecoil_A
    rule=VehicleFinalPoseOverride.AnimGraphNode_StateMachine_1.BodyRecoil.AnimStateTransitionNode_2.Transition
    src=ShotgunBodyRecoil_B
    via=AnimStateTransitionNode_2
   }
  ]
 }
 ShotgunBodyRecoil_A {
  path=VehicleFinalPoseOverride.AnimGraphNode_StateMachine_1.BodyRecoil.AnimStateNode_0.ShotgunBodyRecoil_A
  nodes [
   n1 = "Sequence Player (additive) 'OstrichVehicle_Gunner_Shotgun_Fire_Additive_CMM'"() {bLoopAnimation=false, sequence=/Game/Animation/Game/MainPlayer/Vehicles/OstrichVehicle/OstrichVehicle_Gunner_Shotgun_Fire_Additive_CMM}
   Output Animation Pose(Result=@n1)
  ]
 }
 ShotgunBodyRecoil_B {
  path=VehicleFinalPoseOverride.AnimGraphNode_StateMachine_1.BodyRecoil.AnimStateNode_1.ShotgunBodyRecoil_B
  nodes [
   n1 = "Sequence Player (additive) 'OstrichVehicle_Gunner_Shotgun_Fire_Additive_CMM'"() {bLoopAnimation=false, sequence=&OstrichVehicle_Gunner_Shotgun_Fire_Additive_CMM}
   Output Animation Pose(Result=@n1)
  ]
 }
 Transition {
  path=VehicleFinalPoseOverride.AnimGraphNode_StateMachine_1.BodyRecoil.AnimStateTransitionNode_1.Transition
  nodes [Result(bCanEnterTransition=Get bIsMechBeginFired() -> (bIsMechBeginFired:bool))]
 }
 Transition {
  path=VehicleFinalPoseOverride.AnimGraphNode_StateMachine_1.BodyRecoil.AnimStateTransitionNode_2.Transition
  nodes [Result(bCanEnterTransition=NOT Boolean(A=Get bIsMechBeginFired()))]
 }
 Gunner {
  entry=Idle
  path=VehicleFinalPoseOverride.AnimGraphNode_StateMachine_2.Gunner
  ids=id
  nodes [
   AnimStateNode_0 = Idle()
   AnimStateNode_1 = React_W() {is=title AnimStateNode}
   AnimStateNode_2 = React_E() {is=title AnimStateNode}
   AnimStateNode_3 = React_S() {is=title AnimStateNode}
   AnimStateNode_4 = React_N() {is=title AnimStateNode}
   AnimStateTransitionNode_0 = AnimStateTransitionNode() {crossfadeDuration=0}
   AnimStateTransitionNode_1 = AnimStateTransitionNode() {bAutomaticRuleBasedOnSequencePlayerInState=true}
   AnimStateTransitionNode_2 = AnimStateTransitionNode() {crossfadeDuration=0}
   AnimStateTransitionNode_3 = AnimStateTransitionNode() {bAutomaticRuleBasedOnSequencePlayerInState=true}
   AnimStateTransitionNode_4 = AnimStateTransitionNode() {crossfadeDuration=0}
   AnimStateTransitionNode_5 = AnimStateTransitionNode() {bAutomaticRuleBasedOnSequencePlayerInState=true}
   AnimStateTransitionNode_6 = AnimStateTransitionNode() {crossfadeDuration=0}
   AnimStateTransitionNode_7 = AnimStateTransitionNode() {bAutomaticRuleBasedOnSequencePlayerInState=true}
   AnimStateTransitionNode_8 = AnimStateTransitionNode() {crossfadeDuration=0}
  ]
  links [
   {
    dst=React_E
    rule=VehicleFinalPoseOverride.AnimGraphNode_StateMachine_2.Gunner.AnimStateTransitionNode_2.Transition
    src=Idle
    via=AnimStateTransitionNode_2
   }
   {
    dst=React_N
    rule=VehicleFinalPoseOverride.AnimGraphNode_StateMachine_2.Gunner.AnimStateTransitionNode_6.Transition
    src=Idle
    via=AnimStateTransitionNode_6
   }
   {
    dst=React_S
    rule=VehicleFinalPoseOverride.AnimGraphNode_StateMachine_2.Gunner.AnimStateTransitionNode_4.Transition
    src=Idle
    via=AnimStateTransitionNode_4
   }
   {
    dst=React_S
    rule=VehicleFinalPoseOverride.AnimGraphNode_StateMachine_2.Gunner.AnimStateTransitionNode_8.Transition
    src=Idle
    via=AnimStateTransitionNode_8
   }
   {
    dst=React_W
    rule=VehicleFinalPoseOverride.AnimGraphNode_StateMachine_2.Gunner.AnimStateTransitionNode_0.Transition
    src=Idle
    via=AnimStateTransitionNode_0
   }
   {
    dst=Idle
    rule=VehicleFinalPoseOverride.AnimGraphNode_StateMachine_2.Gunner.AnimStateTransitionNode_3.Transition
    src=React_E
    via=AnimStateTransitionNode_3
   }
   {
    dst=Idle
    rule=VehicleFinalPoseOverride.AnimGraphNode_StateMachine_2.Gunner.AnimStateTransitionNode_7.Transition
    src=React_N
    via=AnimStateTransitionNode_7
   }
   {
    dst=Idle
    rule=VehicleFinalPoseOverride.AnimGraphNode_StateMachine_2.Gunner.AnimStateTransitionNode_5.Transition
    src=React_S
    via=AnimStateTransitionNode_5
   }
   {
    dst=Idle
    rule=VehicleFinalPoseOverride.AnimGraphNode_StateMachine_2.Gunner.AnimStateTransitionNode_1.Transition
    src=React_W
    via=AnimStateTransitionNode_1
   }
  ]
 }
 Idle {
  path=VehicleFinalPoseOverride.AnimGraphNode_StateMachine_2.Gunner.AnimStateNode_0.Idle
  nodes [Output Animation Pose(Result=Use cached pose 'IdlePose'())]
 }
 React_W {
  path=VehicleFinalPoseOverride.AnimGraphNode_StateMachine_2.Gunner.AnimStateNode_1.React_W
  nodes [Output Animation Pose(Result=Sequence Player 'OstrichVehicle_Gunner_React_W_CMM'() {bLoopAnimation=false, sequence=/Game/Animation/Game/MainPlayer/Vehicles/OstrichVehicle/OstrichVehicle_Gunner_React_W_CMM})]
 }
 React_E {
  path=VehicleFinalPoseOverride.AnimGraphNode_StateMachine_2.Gunner.AnimStateNode_2.React_E
  nodes [Output Animation Pose(Result=Sequence Player 'OstrichVehicle_Gunner_React_E_CMM'() {bLoopAnimation=false, sequence=/Game/Animation/Game/MainPlayer/Vehicles/OstrichVehicle/OstrichVehicle_Gunner_React_E_CMM})]
 }
 React_S {
  path=VehicleFinalPoseOverride.AnimGraphNode_StateMachine_2.Gunner.AnimStateNode_3.React_S
  nodes [Output Animation Pose(Result=Sequence Player 'OstrichVehicle_Gunner_React_S_CMM'() {bLoopAnimation=false, sequence=/Game/Animation/Game/MainPlayer/Vehicles/OstrichVehicle/OstrichVehicle_Gunner_React_S_CMM})]
 }
 React_N {
  path=VehicleFinalPoseOverride.AnimGraphNode_StateMachine_2.Gunner.AnimStateNode_4.React_N
  nodes [Output Animation Pose(Result=Sequence Player 'OstrichVehicle_Gunner_React_N_CMM'() {bLoopAnimation=false, sequence=/Game/Animation/Game/MainPlayer/Vehicles/OstrichVehicle/OstrichVehicle_Gunner_React_N_CMM})]
 }
 Transition {
  path=VehicleFinalPoseOverride.AnimGraphNode_StateMachine_2.Gunner.AnimStateTransitionNode_0.Transition
  nodes [Result(bCanEnterTransition=Break Fort Anim Input Velocity Impact!(FortAnimInput_VelocityImpact:"Fort Anim Input Velocity Impact Structure (by ref)"=Get GunnerVelocityImpactInput() -> (GunnerVelocityImpactInput:Fort Anim Input Velocity Impact Structure)) -> (
     DeltaVelocityThreshold:Vector=~
     ImpactScale:Vector=~
     ImpactLimitX:Input Range Structure
     ImpactLimitY:Input Range Structure
     ImpactLimitZ:Input Range Structure
     SpringInterpolatorX:Float RK4Spring Interpolator Structure
     SpringInterpolatorY:Float RK4Spring Interpolator Structure
     SpringInterpolatorZ:Float RK4Spring Interpolator Structure
     bTestVelocity:bool=~
     TestVelocity:Vector=~
     bIsForwardImpact:bool=~
     bIsBackwardImpact:bool=~
     bIsLeftImpact:bool=~
     bIsRightImpact:bool=~
     bIsUpImpact:bool=~
     bIsDownImpact:bool=~
    ) {is=title K2Node_BreakStruct}.bIsLeftImpact)]
 }
 Transition {path=VehicleFinalPoseOverride.AnimGraphNode_StateMachine_2.Gunner.AnimStateTransitionNode_1.Transition, nodes [Result()]}
 Transition {
  path=VehicleFinalPoseOverride.AnimGraphNode_StateMachine_2.Gunner.AnimStateTransitionNode_2.Transition
  nodes [Result(bCanEnterTransition=Break Fort Anim Input Velocity Impact(FortAnimInput_VelocityImpact=Get GunnerVelocityImpactInput()) -> (
     DeltaVelocityThreshold=~
     ImpactScale=~
     bTestVelocity=~
     TestVelocity=~
     bIsForwardImpact=~
     bIsBackwardImpact=~
     bIsLeftImpact=~
     bIsRightImpact=~
     bIsUpImpact=~
     bIsDownImpact=~
    ).bIsRightImpact)]
 }
 Transition {path=VehicleFinalPoseOverride.AnimGraphNode_StateMachine_2.Gunner.AnimStateTransitionNode_3.Transition, nodes [Result()]}
 Transition {
  path=VehicleFinalPoseOverride.AnimGraphNode_StateMachine_2.Gunner.AnimStateTransitionNode_4.Transition
  nodes [Result(bCanEnterTransition=Break Fort Anim Input Velocity Impact(FortAnimInput_VelocityImpact=Get GunnerVelocityImpactInput()) -> (
     DeltaVelocityThreshold=~
     ImpactScale=~
     bTestVelocity=~
     TestVelocity=~
     bIsForwardImpact=~
     bIsBackwardImpact=~
     bIsLeftImpact=~
     bIsRightImpact=~
     bIsUpImpact=~
     bIsDownImpact=~
    ).bIsBackwardImpact)]
 }
 Transition {path=VehicleFinalPoseOverride.AnimGraphNode_StateMachine_2.Gunner.AnimStateTransitionNode_5.Transition, nodes [Result()]}
 Transition {
  path=VehicleFinalPoseOverride.AnimGraphNode_StateMachine_2.Gunner.AnimStateTransitionNode_6.Transition
  nodes [Result(bCanEnterTransition=Break Fort Anim Input Velocity Impact(FortAnimInput_VelocityImpact=Get GunnerVelocityImpactInput()) -> (
     DeltaVelocityThreshold=~
     ImpactScale=~
     bTestVelocity=~
     TestVelocity=~
     bIsForwardImpact=~
     bIsBackwardImpact=~
     bIsLeftImpact=~
     bIsRightImpact=~
     bIsUpImpact=~
     bIsDownImpact=~
    ).bIsForwardImpact)]
 }
 Transition {path=VehicleFinalPoseOverride.AnimGraphNode_StateMachine_2.Gunner.AnimStateTransitionNode_7.Transition, nodes [Result()]}
 Transition {
  path=VehicleFinalPoseOverride.AnimGraphNode_StateMachine_2.Gunner.AnimStateTransitionNode_8.Transition
  nodes [Result(bCanEnterTransition=Break Fort Anim Input Velocity Impact(FortAnimInput_VelocityImpact=Get GunnerVelocityImpactInput()) -> (
     DeltaVelocityThreshold=~
     ImpactScale=~
     bTestVelocity=~
     TestVelocity=~
     bIsForwardImpact=~
     bIsBackwardImpact=~
     bIsLeftImpact=~
     bIsRightImpact=~
     bIsUpImpact=~
     bIsDownImpact=~
    ).bIsDownImpact)]
 }
 ShotgunFingers {
  entry=Default
  path=VehicleFinalPoseOverride.AnimGraphNode_StateMachine_3.ShotgunFingers
  ids=id
  nodes [
   AnimStateNode_0 = Default() {bAlwaysResetOnEntry=true}
   AnimStateNode_2 = ShotgunFire() {is=title AnimStateNode}
   AnimStateNode_3 = ShotgunFireEnd() {is=title AnimStateNode}
   AnimStateTransitionNode_1 = AnimStateTransitionNode() {crossfadeDuration=0}
   AnimStateTransitionNode_2 = AnimStateTransitionNode()
   AnimStateTransitionNode_3 = AnimStateTransitionNode() {bAutomaticRuleBasedOnSequencePlayerInState=true}
  ]
  links [
   {
    dst=ShotgunFire
    rule=VehicleFinalPoseOverride.AnimGraphNode_StateMachine_3.ShotgunFingers.AnimStateTransitionNode_1.Transition
    src=Default
    via=AnimStateTransitionNode_1
   }
   {
    dst=ShotgunFireEnd
    rule=VehicleFinalPoseOverride.AnimGraphNode_StateMachine_3.ShotgunFingers.AnimStateTransitionNode_2.Transition
    src=ShotgunFire
    via=AnimStateTransitionNode_2
   }
   {
    dst=Default
    rule=VehicleFinalPoseOverride.AnimGraphNode_StateMachine_3.ShotgunFingers.AnimStateTransitionNode_3.Transition
    src=ShotgunFireEnd
    via=AnimStateTransitionNode_3
   }
  ]
 }
 Default {
  path=VehicleFinalPoseOverride.AnimGraphNode_StateMachine_3.ShotgunFingers.AnimStateNode_0.Default
  nodes [Output Animation Pose(Result=Use cached pose 'IdlePose'())]
 }
 ShotgunFire {
  path=VehicleFinalPoseOverride.AnimGraphNode_StateMachine_3.ShotgunFingers.AnimStateNode_2.ShotgunFire
  nodes [
   n1 = Sequence Player 'OstrichVehicle_Gunner_Shotgun_Fire_CMM'() {bLoopAnimation=false, playRate=2, sequence=/Game/Animation/Game/MainPlayer/Vehicles/OstrichVehicle/OstrichVehicle_Gunner_Shotgun_Fire_CMM}
   Output Animation Pose(Result=@n1)
  ]
 }
 ShotgunFireEnd {
  path=VehicleFinalPoseOverride.AnimGraphNode_StateMachine_3.ShotgunFingers.AnimStateNode_3.ShotgunFireEnd
  nodes [Output Animation Pose(Result=Sequence Player 'OstrichVehicle_Gunner_Shotgun_FireEnd_CMM'() {bLoopAnimation=false, sequence=/Game/Animation/Game/MainPlayer/Vehicles/OstrichVehicle/OstrichVehicle_Gunner_Shotgun_FireEnd_CMM})]
 }
 Transition {
  path=VehicleFinalPoseOverride.AnimGraphNode_StateMachine_3.ShotgunFingers.AnimStateTransitionNode_1.Transition
  nodes [Result(bCanEnterTransition=Get bIsMechShotgunFired() -> (bIsMechShotgunFired:bool))]
 }
 Transition {
  path=VehicleFinalPoseOverride.AnimGraphNode_StateMachine_3.ShotgunFingers.AnimStateTransitionNode_2.Transition
  nodes [Result(bCanEnterTransition=NOT Boolean(A=Get bIsMechShotgunFired()))]
 }
 Transition {path=VehicleFinalPoseOverride.AnimGraphNode_StateMachine_3.ShotgunFingers.AnimStateTransitionNode_3.Transition, nodes [Result()]}
 VehicleFullBodyOverride {nodes [
   Input Pose() {name=InPoseFullBody}
   Output Pose() {layerGroup=VehicleOverride}
  ]}
 VehicleLowerBodyOverride {nodes [
   Input Pose() {name=InPoseLowerBodyDefault}
   Input Pose() {name=InPoseUpperBody}
   Output Pose() {layerGroup=VehicleOverride}
  ]}
 VehicleSplitBodyOverride {nodes [
   Input Pose() {name=InPoseSplitBody}
   Input Pose() {name=InPoseUpperAndLowerBody}
   Input Pose() {name=InPoseUpperAndLowerBodyRemoveRoot}
   Output Pose() {layerGroup=Vehicle Override}
  ]}
]
variables []
