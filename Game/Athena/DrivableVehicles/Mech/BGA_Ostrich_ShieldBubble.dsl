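; Gunner shield bubble actor for the Ostrich mech, with regeneration, break and respawn, and a timed Overcharge overshield.
; CurrentShield counts damage up toward MaxShield, breaking the shield there, while delayed timer ticks regenerate it toward zero.
; RepNotify functions drive boot, break and overcharge transitions through materials, timelines, particles, audio and rumble.
; Construction loads tuning from curve-table rows; Overcharge grants immunity or replacement health, pulses before expiring, then cools down.
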(event Custom|ShieldBootUp
  (bind _gunnershield (Variables|Default|GetGunnerShield))
  ;[1 1 1] Turn on and off shield
  (if (Variables|Overcharge|GetIsOvercharged)
    _
    (elif (Variables|MechGunnerShield|GetShieldActive)
      (if (Variables|MechGunnerShield|GetShieldDestroyed)
        (Variables|MechGunnerShield|SetShieldDestroyed false))
      (Game|Feedback|ClientPlayForceFeedback (Fortnite|GetFortPlayerControllerfromActor (Variables|MechGunnerShield|GetMechGunner)) (Variables|ForceFeedback|GetShieldActivateForceFeedback) "ShieldActivate" false false false)
      (bind _returnvalue (Audio|SpawnSoundAttached (Variables|Audio|GetShieldPowerupSound) _gunnershield))
      (Variables|Default|SetPowerUpShieldAudio _returnvalue)))
  (Collision|SetCollisionEnabled _gunnershield "QueryOnly")
  (bind (_direction _newtrack_0) (Timeline|ScaleUpShield Play)
    (:Update
      (Rendering|Material|SetScalarParameterValueonMaterials (Variables|Default|GetGunnerShield) "FadeDown" (Math|Float|Lerp 1.0 0.0 _newtrack_0))
      (Rendering|Components|Light|SetIntensity (Variables|Default|GetSpotLight) (Math|Float|Lerp 0.0 3e+01 _newtrack_0)))
    (:Finished
      (if (Variables|MechGunnerShield|GetShieldActive)
        (CallFunction|BubbleFlash false) ; Pulse when we hop in and fade in is complete. Can remove if we don't want.
      ))))

(event Custom|ShieldPowerDown
  ;[1 1 1] Turn on and off shield
  (if (Variables|Overcharge|GetIsOvercharged)
    (Components|Activation|SetActive (Variables|Default|GetP_Ostrich_ShieldReturning) false false)
    (CallFunction|SetOverchargeAudioEnabled false)
    (Components|Timeline|Stop (Variables|BGA_Ostrich_ShieldBubble|GetOverchargeShieldPulse))
    (Utilities|Time|ClearandInvalidateTimerbyHandle (Variables|Default|GetOverchargePulseDelay)))
  (Collision|SetCollisionEnabled (Variables|Default|GetGunnerShield) "NoCollision")
  (Timeline|ScaleUpShield Reverse))

(event Custom|BubbleFlash
  (params (Play%20Sound :type Boolean :default true))
  ;[1 1 1] Hit Flash
  (if (or (Variables|Overcharge|GetIsOvercharged) (Variables|MechGunnerShield|GetShieldDestroyed))
    _
    (elif Play%20Sound
      (Audio|SpawnSoundAttached :Sound (Variables|Audio|GetShieldImpactSound) :AttachToComponent (Variables|Default|GetGunnerShield) :AttachPointName "None" :LocationType "KeepRelativeOffset" :bStopWhenAttachedToDestroyed false :VolumeMultiplier 1.0 :PitchMultiplier (Math|Float|MapRangeUnclamped (- 1.0 (Variables|MechGunnerShield|GetNormalizedDamage)) 0.0 1.0 0.7 1.2) :StartTime 0.0 :bAutoDestroy true)
      (bind (_direction _newtrack_0) (Timeline|HitShieldVisuals PlayFromStart)
        (:Update
          (Rendering|Material|SetScalarParameterValueonMaterials (Variables|Default|GetGunnerShield) "emissiveMulti" (Math|Float|Lerp 1.0 5e+01 _newtrack_0))))
      (else
        _))))

(event Custom|BeginRegen
  ;[1 1 1] EVENT -> On Damage Server
  (if (not (or (Variables|Overcharge|GetIsOvercharged) (Variables|MechGunnerShield|GetShieldDestroyed)))
    (Variables|MechGunnerShield|SetRegenShield true)
    (bind _returnvalue (Utilities|Time|SetTimerbyEvent (EventDispatchers|CreateEvent :!event Custom|Regen) (Variables|ShieldRegen|GetRegenTickFrequency) true))
    (Variables|MechGunnerShield|SetRegenTimer _returnvalue)
    (bind _returnvalue_2 (Audio|SpawnSoundAttached (Variables|Audio|GetShieldPowerupRechargeSound) (Variables|Default|GetGunnerShield)))
    (Variables|Default|SetPowerUpRechargeAudio _returnvalue_2)))

(event EventOnDamageServer
  (params (Damage :type Float%20%28single-precision%29) (DamageTags :type Gameplay%20Tag%20Container%20Structure%20%28by%20ref%29) (Momentum :type Vector :default "0, 0, 0") (HitInfo :type Hit%20Result%20Structure%20%28by%20ref%29) (InstigatedBy :type Controller%20Object%20Reference) (DamageCauser :type Actor%20Object%20Reference) (EffectContext :type Gameplay%20Effect%20Context%20Handle%20Structure))
  ;[1 1 1] EVENT -> On Damage Server
  (Utilities|Time|ClearandInvalidateTimerbyHandle (Variables|MechGunnerShield|GetRegenTimer))
  (Variables|MechGunnerShield|SetRegenShield false)
  (Variables|MechGunnerShield|SetCurrentShield (+ Damage (Variables|MechGunnerShield|GetCurrentShield)))
  (if (>= (Variables|MechGunnerShield|GetCurrentShield) (Variables|MechGunnerShield|GetMaxShield))
    (Variables|MechGunnerShield|SetShieldDestroyed true)
    (else
      (Utilities|FlowControl|RetriggerableDelay (Variables|ShieldRegen|GetOnDamageRegenDelay))
      (if (not (or (Variables|Overcharge|GetIsOvercharged) (Variables|MechGunnerShield|GetShieldDestroyed)))
        (Variables|MechGunnerShield|SetRegenShield true)
        (bind _returnvalue (Utilities|Time|SetTimerbyEvent (EventDispatchers|CreateEvent :!event Custom|Regen) (Variables|ShieldRegen|GetRegenTickFrequency) true))
        (Variables|MechGunnerShield|SetRegenTimer _returnvalue)
        (bind _returnvalue_2 (Audio|SpawnSoundAttached (Variables|Audio|GetShieldPowerupRechargeSound) (Variables|Default|GetGunnerShield)))
        (Variables|Default|SetPowerUpRechargeAudio _returnvalue_2)))))

(event Custom|ReactivateShield
  (bind _gunnershield (Variables|Default|GetGunnerShield))
  ;[1 1 1] Activate Shield
  (if (> (Variables|MechGunnerShield|GetShouldRespawnFullHealth) 0.0)
    (Variables|MechGunnerShield|SetCurrentShield 0.0)
    (Rendering|SetVisibility (Variables|Default|GetGunnerShield) true false)
    (CallFunction|ShieldBootUp)
    (else
      (Variables|MechGunnerShield|SetCurrentShield (- (Variables|MechGunnerShield|GetMaxShield) (Variables|ShieldRegen|GetRegenHealPerTick)))
      (Rendering|SetVisibility _gunnershield true false)
      (CallFunction|BeginRegen)
      (Collision|SetCollisionEnabled _gunnershield "QueryOnly"))))

(event Custom|Regen
  (bind _regenhealpertick (Variables|ShieldRegen|GetRegenHealPerTick))
  (bind _currentshield (Variables|MechGunnerShield|GetCurrentShield))
  (bind _returnvalue (- _currentshield _regenhealpertick))
  ;[1 1 1] Regen Shield
  (if (<= _returnvalue 0.0)
    (Variables|MechGunnerShield|SetCurrentShield 0.0)
    (Utilities|Time|ClearandInvalidateTimerbyHandle (Variables|MechGunnerShield|GetRegenTimer))
    (Variables|MechGunnerShield|SetRegenShield false)
    (else
      (Variables|MechGunnerShield|SetCurrentShield _returnvalue))))

(event Custom|ShieldDestroy
  (bind _mech (Variables|MechGunnerShield|GetMech))
  ;[1 1 1] Deactivate Shield
  (Rendering|Components|Light|SetIntensity (Variables|Default|GetSpotLight) 0.0)
  (Collision|SetCollisionEnabled (Variables|Default|GetGunnerShield) "NoCollision")
  (Rendering|SetVisibility (Variables|Default|GetGunnerShield) false false)
  (Utilities|IsValid _mech
    (:Is%20Valid
      (Utilities|FlowControl|Delay 0.15) ; delay so shield doesn't show up after exiting
      (if (Gameplay|GetHasGunner _mech)
        (Components|Activation|Activate (Variables|Default|GetP_Ostrich_ShieldReturning) true)
        (bind _returnvalue (Utilities|Time|SetTimerbyEvent (delegate Custom|ResetShieldBool) (Variables|MechGunnerShield|GetRespawnShieldDelay)))
        (Variables|MechGunnerShield|SetResetShieldTimer _returnvalue)
        (Effects|SpawnEmitterAttached :EmitterTemplate "/Game/Effects/Fort_Effects/Gameplay/Ostrich/P_Ostrich_ShieldBreak.P_Ostrich_ShieldBreak" :AttachToComponent (Variables|Default|GetGunnerShield) :AttachPointName "None" :Scale "1.000000,1.000000,1.000000" :LocationType "SnapToTarget" :bAutoDestroy true :PoolingMethod "None" :bAutoActivate true)
        (else
          _))
      (Rendering|Material|SetScalarParameterValueonMaterials (Mesh|GetMechMesh (Variables|MechGunnerShield|GetMech)) "DamagePerchent" 1.0)
      (Audio|SpawnSoundAttached (Variables|Audio|GetShieldBreakSound) (Variables|Default|GetGunnerShield))
      (if (Variables|MechGunnerShield|GetShieldActive)
        (Variables|Default|GetShieldBrokenLoopAudio
          (:then)
          (:else
            (bind _returnvalue_2 (Audio|SpawnSoundAttached (Variables|Audio|GetShieldBrokenLoopSound) (Variables|Default|GetGunnerShield)))
            (Variables|Default|SetShieldBrokenLoopAudio _returnvalue_2)))))))

(event Custom|ResetShieldBool
  ;[1 1 1] Deactivate Shield
  (Variables|MechGunnerShield|SetShieldDestroyed false)
  (bind _shieldbrokenloopaudio (Variables|Default|GetShieldBrokenLoopAudio)
    (:then
      (Audio|Components|Audio|FadeOut (Variables|Default|GetShieldBrokenLoopAudio) 0.25 0.0 "Linear" :!class /Script/Engine.AudioComponent))
    (:else
      (Audio|Components|Audio|Stop _shieldbrokenloopaudio))))

(event Custom|BeginOverchargeShield
  ;[1 1 1] Overcharge Begin
  (Rendering|Material|SetMaterial (Variables|Default|GetGunnerShield) 0 "/Game/Athena/DrivableVehicles/Mech/Materials/M_Mech_Shield_OverShieldOnly.M_Mech_Shield_OverShieldOnly")
  (bind _returnvalue (Utilities|Time|SetTimerbyEvent (delegate Custom|ResetIsOvercharged) (Variables|Overcharge|GetOverchargeDuration)))
  (Variables|Overcharge|SetOverchargeDurationTimer _returnvalue)
  (bind _returnvalue_2 (Utilities|Time|SetTimerbyEvent (EventDispatchers|CreateEvent :!event Custom|OvershieldBeginExpirePulse) (- (Variables|Overcharge|GetOverchargeDuration) 3.0)))
  (Variables|Default|SetOverchargePulseDelay _returnvalue_2)
  (Components|Activation|Deactivate (Variables|Default|GetP_Ostrich_ShieldReturning))
  (Variables|Overcharge|SetCanOvercharge false)
  (Variables|MechGunnerShield|SetRegenShield false)
  (Variables|MechGunnerShield|SetShieldDestroyed false)
  (Variables|MechGunnerShield|SetShieldActive true)
  (Utilities|Time|ClearandInvalidateTimerbyHandle (Variables|MechGunnerShield|GetRegenTimer))
  (Utilities|Time|ClearandInvalidateTimerbyHandle (Variables|MechGunnerShield|GetResetShieldTimer))
  (Game|Feedback|ClientPlayForceFeedback (Fortnite|GetFortPlayerControllerfromActor (Variables|MechGunnerShield|GetMechGunner)) (Variables|ForceFeedback|GetOverchargeActivateForceFeedback) "OverchargeActivate" false false false)
  (CallFunction|SetOverchargeAudioEnabled true)
  (if (> (Variables|Overcharge|GetOverchargeDamageImmune) 0.0)
    (Variables|Actor|SetbCanBeDamaged false)
    (Variables|MechGunnerShield|SetCurrentShield 0.0)
    (Components|Activation|Deactivate (Variables|Default|GetP_Ostrich_ShieldReturning))
    (else
      (Variables|MechGunnerShield|SetCurrentShield 0.0)
      (Variables|MechGunnerShield|SetMaxShield (Variables|Overcharge|GetOverchargeHealthIfNotImmune)))))

(event Custom|ResetIsOvercharged
  (bind _gunnershield (Variables|Default|GetGunnerShield))
  ;[1 1 1] Overcharge Begin
  (Rendering|Material|SetMaterial _gunnershield 0 "/Game/Athena/DrivableVehicles/Mech/Materials/M_Mech_Shield.M_Mech_Shield")
  (Rendering|Material|SetScalarParameterValueonMaterials _gunnershield "FadeDown" 0.0)
  (Variables|Overcharge|SetIsOvercharged false)
  (Variables|Actor|SetbCanBeDamaged true)
  (bind _returnvalue (Utilities|Time|SetTimerbyEvent (delegate Custom|ResetOverCharge) (Variables|Overcharge|GetOverchargeCooldown)))
  (Variables|Overcharge|SetOverchargeCooldownTimer _returnvalue))

(event Custom|ResetOverCharge
  ;[1 1 1] Overcharge Begin
  (Variables|Overcharge|SetCanOvercharge true))

(event Custom|EndOverchargeShield
  ;[1 1 1] Overcharge End
  (Variables|Actor|SetbCanBeDamaged true)
  (Variables|MechGunnerShield|SetMaxShield (Variables|MechGunnerShield|GetMaxShield))
  (Variables|MechGunnerShield|SetShieldDestroyed true)
  (Game|Feedback|ClientPlayForceFeedback (Fortnite|GetFortPlayerControllerfromActor (Variables|MechGunnerShield|GetMechGunner)) (Variables|ForceFeedback|GetShieldDeactivateForceFeedback) "OverchargeDeactivate" false false false))

(event Custom|OvershieldBeginExpirePulse
  ;[1 1 1] Overcharge Pulse Begin.
  (bind (_direction _pulseparam _pulseaudiopitchcurve _pulseaudiovolumecurve) (Timeline|OverchargeShieldPulse PlayFromStart)
    (:Update
      (Rendering|Material|SetScalarParameterValueonMaterials (Variables|Default|GetSM_Ostrich_Shield_Overshield_Top_Electrical) "OvershieldPulse" _pulseparam))
    (:Finished
      (CallFunction|SetOverchargeAudioEnabled false))
    (:PulseEvent
      (if (Variables|Overcharge|GetIsOvercharged)
        (bind _returnvalue (Audio|SpawnSoundAttached :Sound (Variables|Audio|GetOverchargeShieldPulseSound) :AttachToComponent (Variables|Default|GetGunnerShield) :AttachPointName "None" :LocationType "KeepRelativeOffset" :bStopWhenAttachedToDestroyed false :VolumeMultiplier _pulseaudiovolumecurve :PitchMultiplier _pulseaudiopitchcurve :StartTime 0.0 :bAutoDestroy true))
        (Variables|Default|SetOverchargeShieldPulseAudio _returnvalue)
        (else
          (Audio|Components|Audio|Stop (Variables|Default|GetOverchargeShieldPulseAudio)))))))

(event EventEndPlay
  (params (EndPlayReason :type EEndPlayReason%20Enum :default "Destroyed"))
  (Variables|MechGunnerShield|SetEndSounds true))

(fn IsShieldActive
  (return (and (Variables|MechGunnerShield|GetShieldActive) (not (Variables|MechGunnerShield|GetShieldDestroyed)))))

(fn OnRep_CurrentShield
  (bind _mech (Variables|MechGunnerShield|GetMech))
  (bind _output_get (Variables|MechGunnerShield|SetNormalizedDamage (/ (Variables|MechGunnerShield|GetCurrentShield) (Variables|MechGunnerShield|GetMaxShield))))
  (Rendering|Material|SetScalarParameterValueonMaterials (Variables|Default|GetGunnerShield) "DamagePerchent" _output_get) ; Increase Static based on Current Health
  (Utilities|IsValid _mech
    (:Is%20Valid
      (Rendering|Material|SetScalarParameterValueonMaterials (Mesh|GetMechMesh _mech) "DamagePerchent" _output_get)
      (if (Variables|MechGunnerShield|GetRegenShield)
        _
        (elif (> (Variables|MechGunnerShield|GetCurrentShield) 0.0)
          (CallFunction|BubbleFlash true))))
    (:Is%20Not%20Valid
      (Actor|DestroyActor) ; If mech is not valid, it is dead in which case we should also be dead
    )))

(fn OnRep_EndSounds
  (if (Variables|MechGunnerShield|GetEndSounds)
    (Audio|Components|Audio|Stop (Variables|Default|GetOverchargeAudioLoop))
    (Components|Timeline|Stop (Variables|BGA_Ostrich_ShieldBubble|GetOverchargeShieldPulse))
    (Variables|MechGunnerShield|SetEndSounds false)))

(fn OnRep_ForceCancelOvercharge
  (Utilities|Time|ClearandInvalidateTimerbyHandle (Variables|Overcharge|GetOverchargeDurationTimer))
  (Utilities|Time|ClearandInvalidateTimerbyHandle (Variables|Overcharge|GetOverchargeCooldownTimer))
  (Variables|Actor|SetbCanBeDamaged true)
  (Rendering|Material|SetScalarParameterValueonMaterials (Variables|Default|GetGunnerShield) "OverShield" 0.0)
  (Components|Activation|Deactivate (Variables|Default|GetP_Ostrich_OverShield))
  (Rendering|SetVisibility (Variables|Default|GetSM_Ostrich_Shield_Overshield_Top_Electrical) false false)
  (Rendering|Material|SetScalarParameterValueonMaterials (Mesh|GetMechMesh (Variables|MechGunnerShield|GetMech)) "OverShield" 0.0)
  (Variables|Actor|SetbCanBeDamaged true) ; Failsafe
  (Variables|MechGunnerShield|SetShieldDestroyed true)
  (Variables|MechGunnerShield|SetShieldActive false)
  (CallFunction|ResetIsOvercharged)
  (bind _shieldbrokenloopaudio (Variables|Default|GetShieldBrokenLoopAudio)
    (:then
      (Audio|Components|Audio|FadeOut _shieldbrokenloopaudio 0.25 0.0 "Linear" :!class /Script/Engine.AudioComponent))
    (:else
      _))
  (Audio|Components|Audio|Stop (Variables|Default|GetOverchargeAudioLoop))
  (Components|Timeline|Stop (Variables|BGA_Ostrich_ShieldBubble|GetOverchargeShieldPulse)))

(fn OnRep_IsOvercharged
  (bind _p_ostrich_overshield (Variables|Default|GetP_Ostrich_OverShield))
  (if (Variables|Overcharge|GetIsOvercharged)
    (Variables|Overcharge|SetCanOvercharge false)
    (CallFunction|BeginOverchargeShield)
    (Rendering|Material|SetScalarParameterValueonMaterials (Variables|Default|GetGunnerShield) "OverShield" 1.0)
    (Components|Activation|Activate _p_ostrich_overshield false)
    (Rendering|SetVisibility (Variables|Default|GetSM_Ostrich_Shield_Overshield_Top_Electrical) true false)
    (Rendering|Material|SetScalarParameterValueonMaterials (Mesh|GetMechMesh (Variables|MechGunnerShield|GetMech)) "OverShield" 1.0)
    (else
      (CallFunction|EndOverchargeShield)
      (Rendering|Material|SetScalarParameterValueonMaterials (Variables|Default|GetGunnerShield) "OverShield" 0.0)
      (Components|Activation|Deactivate _p_ostrich_overshield)
      (Rendering|SetVisibility (Variables|Default|GetSM_Ostrich_Shield_Overshield_Top_Electrical) false false)
      (Rendering|Material|SetScalarParameterValueonMaterials (Mesh|GetMechMesh (Variables|MechGunnerShield|GetMech)) "OverShield" 0.0)
      (Variables|Actor|SetbCanBeDamaged true)
      (Rendering|Material|SetScalarParameterValueonMaterials (Variables|Default|GetSM_Ostrich_Shield_Overshield_Top_Electrical) "OvershieldPulse" 0.0))))

(fn OnRep_RegenShield)

(fn OnRep_ShieldActive
  (if (Variables|MechGunnerShield|GetShieldActive)
    (CallFunction|ShieldBootUp)
    (else
      (CallFunction|ShieldPowerDown))))

(fn OnRep_ShieldDestroyed
  (if (Variables|MechGunnerShield|GetShieldDestroyed)
    (CallFunction|ShieldDestroy)
    (else
      (CallFunction|ReactivateShield))))

(fn SetOverchargeAudioEnabled
  (params (Enabled :type Boolean))
  (if Enabled
    (Variables|Default|GetOverchargeAudioLoop
      (:then)
      (:else
        (bind _returnvalue (Audio|SpawnSoundAttached (Variables|Audio|GetOverchargeLoopSound) (Variables|Default|GetDefaultSceneRoot)))
        (bind _output_get (Variables|Default|SetOverchargeAudioLoop _returnvalue))
        (Audio|Components|Audio|FadeIn _output_get 1.0 1.0 0.0 "Linear" :!class /Script/Engine.AudioComponent)
        (bind _returnvalue_2 (Audio|SpawnSoundAttached (Variables|Audio|GetOverchargeSound) (Variables|Default|GetDefaultSceneRoot)))
        (Variables|Default|SetOverchargeAudio _returnvalue_2)))
    (else
      (bind _overchargeaudioloop (Variables|Default|GetOverchargeAudioLoop)
        (:then
          (Audio|Components|Audio|FadeOut _overchargeaudioloop 0.25 0.0 "Linear" :!class /Script/Engine.AudioComponent)
          (Variables|Default|SetOverchargeAudioLoop 0)
          (Components|Timeline|Stop (Variables|BGA_Ostrich_ShieldBubble|GetOverchargeShieldPulse)))
        (:else)))))

(fn ConstructionScript
  (Collision|SetCollisionEnabled (Variables|Default|GetGunnerShield) "NoCollision")
  (Variables|MechGunnerShield|SetMaxShield (ScalableFloat|GetValueatLevel (Variables|CurveTableRows|GetMaxShieldRow)))
  (Variables|MechGunnerShield|SetRespawnShieldDelay (ScalableFloat|GetValueatLevel (Variables|CurveTableRows|GetRespawnShieldDelayRow)))
  (Variables|ShieldRegen|SetRegenHealPerTick (ScalableFloat|GetValueatLevel (Variables|CurveTableRows|GetRegenHealPerTickRow)))
  (Variables|ShieldRegen|SetRegenTickFrequency (ScalableFloat|GetValueatLevel (Variables|CurveTableRows|GetRegenTickFrequencyRow)))
  (Variables|ShieldRegen|SetOnDamageRegenDelay (ScalableFloat|GetValueatLevel (Variables|CurveTableRows|GetOnDamageRegenDelayRow)))
  (Variables|Overcharge|SetOverchargeDamageImmune (ScalableFloat|GetValueatLevel (Variables|CurveTableRows|GetOverchargeDamageImmuneRow)))
  (Variables|Overcharge|SetOverchargeHealthIfNotImmune (ScalableFloat|GetValueatLevel (Variables|CurveTableRows|GetOverchargeHealthIfNotImmuneRow)))
  (Variables|Overcharge|SetOverchargeDuration (ScalableFloat|GetValueatLevel (Variables|CurveTableRows|GetOverchargeDurationRow)))
  (Variables|MechGunnerShield|SetShouldRespawnFullHealth (ScalableFloat|GetValueatLevel (Variables|CurveTableRows|GetShouldRespawnFullHealthRow)))
  (Variables|Overcharge|SetOverchargeCooldown (ScalableFloat|GetValueatLevel (Variables|CurveTableRows|GetOverchargeCooldownRow))))
