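; Rocket projectile for the Ostrich mech that flies with DrunkArtillery homing data and explodes on impact.
; It also detonates when a fuse timer expires or, on authority, when it passes its homing target.
; Hit components tagged NoStick double its velocity; other hits set it to stop simulating.
; Cosmetic-only logic spins the mesh in flight, then plays water or generic explosion effects, sound and a camera-shake GameplayCue.
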(event EventBeginPlay
  ;[1 1 1] Begin Play
  (Projectile|SetHomingData 0 0 (Utilities|Struct|MakeProjectileHomingData "DrunkArtillery" 175.0 6e+02 1.0) :!class /Script/FortniteGame.FortProjectileMovementInterface)
  (Utilities|Time|SetTimerbyFunctionName self "FuseTimerMax" 15)
  (Networking|CanExecuteCosmeticEvents
    (:True
      (Components|Activation|Activate (Variables|Default|Get%20RocketTrailPS) false)
      (Rendering|SetRenderCustomDepth (Variables|Default|Get%20RocketTrailPS) true))))

(event ProjectileMovement|EventOnHomingPassedTarget
  (bind _projectilemovementcomponent (Variables|Movement|GetProjectileMovementComponent))
  (bind _updatedcomponent (Variables|/Script/Engine.MovementComponent|GetUpdatedComponent _projectilemovementcomponent))
  (bind _returnvalue (Transformation|GetWorldLocation _updatedcomponent))
  ;[1 1 1] EVENT -> Homing Passed Target
  (Networking|SwitchHasAuthority
    (:Authority
      (Variables|Effects|SetStopLocation _returnvalue)
      (Delivery|DoExplosionatLocation _returnvalue))))

(event EventTick
  (params (DeltaSeconds :type Float%20%28single-precision%29))
  ;[1 1 1] On Tick, spin the rocket and particle trail
  (Variables|Default|SetTick%20Delta DeltaSeconds)
  (Networking|CanExecuteCosmeticEvents
    (:True
      (if (Components|Activation|IsActive (Variables|Default|Get%20RocketTrailPS))
        (Variables|Effects|SetRocketSpin (Math|Rotator|MakeRotator 0.0 (* (Variables|Default|GetTick%20Delta) (Variables|Effects|GetRocketSpinRate))))
        (Transformation|AddLocalRotation (Variables|Default|GetSM_OstrichRocket) (Variables|Effects|GetRocketSpin) false false)))))

(event EventOnTouched
  (params (OtherActor :type Actor%20Object%20Reference) (OtherComp :type Primitive%20Component%20Object%20Reference) (HitResult :type Hit%20Result%20Structure%20%28by%20ref%29) (bIsOverlap :type Boolean))
  (bind (_bblockinghit _binitialoverlap _time _distance _location _impactpoint _normal _impactnormal _physmat _hitactor _hitcomponent) (Collision|BreakHitResult HitResult))
  ;[1 1 1] Bounce off "NoBounce"
  (if (Components|ComponentHasTag _hitcomponent "NoStick")
    (Variables|/Script/Engine.MovementComponent|SetVelocity (* (Variables|/Script/Engine.MovementComponent|GetVelocity (Variables|Movement|GetProjectileMovementComponent)) 2.0) (Variables|Movement|GetProjectileMovementComponent))
    (else
      (Variables|Projectile|SetbStopSimulatingOnHit true))))

(event EventOnStop
  (params (Hit :type Hit%20Result%20Structure%20%28by%20ref%29))
  (bind (_bblockinghit _binitialoverlap _time _distance _location _impactpoint _normal _impactnormal _physmat) (Collision|BreakHitResult Hit))
  ;[1 1 1] Impact and Explosion Stuff
  (Variables|Effects|SetDecalLocation _impactpoint)
  (Variables|Effects|SetStopLocation (+ _impactpoint (* _impactnormal (Variables|Effects|GetExplosion%20Impact%20Offset))))
  (Variables|Effects|SetStopRotZ (Math|Rotator|MakeRotfromZ _impactnormal))
  (Variables|Default|SetSurfaceType (Variables|/Script/PhysicsCore.PhysicalMaterial|GetSurfaceType _physmat))
  (if (not (Delivery|IsBeingKilled))
    (Delivery|DoExplosion)))

(event Custom|FuseTimerMax
  ;[1 1 1] Impact and Explosion Stuff
  (Variables|Effects|SetTimerMaxReached? true)
  (Variables|Effects|SetStopLocation (Transformation|GetWorldLocation (Variables|Default|GetSM_OstrichRocket)))
  (if (not (Delivery|IsBeingKilled))
    (Delivery|DoExplosion)))

(event EventOnExploded
  (params (HitActors :type Array%20of%20Actor%20Object%20References) (HitResults :type Array%20of%20Hit%20Result%20Structures))
  (bind _stoplocation (Variables|Effects|GetStopLocation))
  (bind _stoprotz (Variables|Effects|GetStopRotZ))
  ;[1 1 1] Impact and Explosion Stuff
  (Networking|CanExecuteCosmeticEvents
    (:True
      (Development|SetHiddeninGame (Variables|Default|GetSM_OstrichRocket) true false)
      (Components|Activation|SetActive (Variables|Default|Get%20RocketTrailPS) false false)
      ;;[1 1 1] SingleLineCheckForWorldObjectsSurfaceType
      (if (Utilities|Enum|Equal(Enum) (Variables|Default|GetSurfaceType) "SurfaceType8")
        (bind _returnvalue (Niagara|SpawnSystematLocation "/Game/Effects/Explosion_Update/NS_GrenadeLauncher_Explosion_Asteria_Water.NS_GrenadeLauncher_Explosion_Asteria_Water" _stoplocation _stoprotz))
        (Rendering|SetRenderCustomDepth _returnvalue true)
        (Audio|PlaySoundatLocation (Variables|Sounds|GetRocket_Projectile_Explosion_Water_Sound) (Variables|Effects|GetStopLocation))
        (else
          (bind _returnvalue_2 (Effects|SpawnEmitteratLocation (Variables|Effects|GetExplosion_Generic_PS) _stoplocation _stoprotz "1.000000,1.000000,1.000000" true "AutoRelease"))
          (Rendering|SetRenderCustomDepth _returnvalue_2 true)
          (Audio|PlaySoundatLocation (Variables|Sounds|GetRocket_Projectile_Explosion_Sound) (Variables|Effects|GetStopLocation))))
      (Rendering|SetVisibility (Variables|Default|GetSM_OstrichRocket) false false)
      ;[1 1 1] Shake that camera
      (GameplayCue|ExecuteGameplayCue%28Burst%29Local self (Variables|Effects|GetFeedbackCue) (Ability|GameplayCue|MakeGameplayCueParameters 0.0)) ; Cam shake + controller vibration
      (Audio|Components|Audio|FadeOut (Variables|Default|GetRocket_AudioComponent) 1 0.0 "Linear" :!class /Script/Engine.AudioComponent))))

(event EventDestroyed
  ;[1 1 1] Impact and Explosion Stuff
  (Networking|CanExecuteCosmeticEvents
    (:True
      (Components|Activation|SetActive (Variables|Default|Get%20RocketTrailPS) false false))))

(fn ConstructionScript
  (Effects|Components|ParticleSystem|SetTemplate (Variables|Default|GetRibbon_Trail_PSC) (Variables|Effects|GetRibbon_Trail_PS))
  (Effects|Components|ParticleSystem|SetTemplate (Variables|Default|GetExplosion_Generic_PSC) (Variables|Effects|GetExplosion_Generic_PS))
  (Effects|Components|ParticleSystem|SetTemplate (Variables|Default|GetExplosion_Flesh_Damage_PSC) (Variables|Effects|GetExplosion_Flesh_Damage_PS)))
