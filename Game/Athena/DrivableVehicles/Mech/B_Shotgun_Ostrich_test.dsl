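; Weapon Blueprint for the Ostrich mech's shotgun and rockets, handling cosmetic feedback for firing and loading.
; Fire and rocket events play force feedback, with sounds, emitters and a sound indicator on rocket launch, plus a settle emitter.
; Rocket-load sound pitch scales with rockets loaded against magazine ammo count.
; A reload-started delegate, bound on BeginPlay and unbound on EndPlay, forwards shotgun reloads to the instigator's Ostrich vehicle.
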(event EventOnPlayWeaponFireFX
  (params (bPersistentFire :type Boolean) (bSecondaryFire :type Boolean))
  (|CallParentFunction bPersistentFire bSecondaryFire)
  (Game|Feedback|ClientPlayForceFeedback (Fortnite|GetFortPlayerControllerfromActor (Game|GetInstigator)) (Variables|Default|GetShotgunFireForceFeedback) "FireShotgun" false false false))

(event EventOnLoadedRockets
  (params (NumRocketsLoaded :type Integer))
  (Game|Feedback|ClientPlayForceFeedback (Fortnite|GetFortPlayerControllerfromActor (Game|GetInstigator)) (Variables|Default|GetRocketLoadForceFeedback) "LoadMissile" false false false)
  (Audio|SpawnSoundAttached :Sound (Variables|Sound|GetSound_LoadMissile) :AttachToComponent (Variables|Visual|GetWeaponMesh) :AttachPointName "None" :LocationType "KeepRelativeOffset" :bStopWhenAttachedToDestroyed false :VolumeMultiplier 1.0 :PitchMultiplier (Math|Float|MapRangeUnclamped (Math|Conversions|ToFloat(Integer) NumRocketsLoaded) 0.0 (Math|Conversions|ToFloat(Integer) (FortWeapon|GetMagazineAmmoCount)) 0.9 1.1) :StartTime 0.0 :bAutoDestroy true))

(event EventOnRocketFired
  (params (AttachToMesh :type Scene%20Component%20Object%20Reference) (AttachToSocket :type Name :default "None"))
  (bind _returnvalue (Transformation|GetActorLocation))
  (bind _sound_rocketfire (Variables|Sound|GetSound_RocketFire))
  (Effects|SpawnEmitterAttached "/Game/Effects/Fort_Effects/Gameplay/Ostrich/P_Ostrich_Rocket_MF.P_Ostrich_Rocket_MF" AttachToMesh AttachToSocket)
  (Audio|PlaySoundatLocation _sound_rocketfire _returnvalue)
  (SoundIndicator|ShowSoundIndicator (Variables|SoundIndicator|GetSoundIndicatorComponent) _returnvalue _sound_rocketfire 1.0 0)
  (Game|Feedback|ClientPlayForceFeedback (Fortnite|GetFortPlayerControllerfromActor (Game|GetInstigator)) (Variables|Default|GetRocketFireForceFeedback) "FireMissile" false false false))

(event EventOnRocketLauncherSettle
  (params (AttachToMesh :type Scene%20Component%20Object%20Reference) (AttachToSocket :type Name :default "None"))
  (Effects|SpawnEmitterAttached "/Game/Effects/Fort_Effects/Gameplay/Ostrich/P_Ostrich_Rocket_Reload.P_Ostrich_Rocket_Reload" AttachToMesh AttachToSocket))

(event EventOnRocketFiringStarted
  (params (AttachToMesh :type Scene%20Component%20Object%20Reference) (AttachToSocket :type Name :default "None")))

(event EventBeginPlay
  (|CallParentFunction)
  (EventDispatchers|BindEventtoOnReloadStarted (delegate Custom|PlayReloadMontage)))

(event Custom|PlayReloadMontage
  (params (ReloadTime :type Float%20%28single-precision%29) (ReloadType :type EFortWeaponReloadType%20Enum :default "ReloadWholeClip"))
  (bind _asfort_player_pawn (Utilities|Casting|CastToFortPlayerPawn (Game|GetInstigator)))
  (bind _asfort_character_vehicle_ostrich (Utilities|Casting|CastToFortCharacterVehicle_Ostrich (Vehicle|GetVehicleActor _asfort_player_pawn :!class /Script/FortniteGame.FortPlayerPawn)))
  (Gameplay|OnStartShotgunReload _asfort_character_vehicle_ostrich))

(event EventEndPlay
  (params (EndPlayReason :type EEndPlayReason%20Enum :default "Destroyed"))
  (|CallParentFunction EndPlayReason)
  (EventDispatchers|UnbindEventfromOnReloadStarted (delegate Custom|PlayReloadMontage)))

(fn ConstructionScript
  (|CallParentFunction))
