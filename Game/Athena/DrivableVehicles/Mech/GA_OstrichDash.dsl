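; Gameplay ability for the drivable Ostrich mech's dash, activated by an event whose effect context supplies the direction.
; Unless the mech is charging a jump or kicking, it pushes a dash camera, updates dash state, notifies HUD cooldown and plays rumble.
; Airborne dashes launch via AirDash, ending on movement-mode change or after a delay.
; Ground dashes apply curve-driven root motion force, popping the camera after a delay and ending when the task finishes.
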(event Custom|MovementModeChangedDelegate
  (params (Character :type Character%20Object%20Reference) (PrevMovementMode :type EMovementMode%20Enum :default "MOVE_None") (PreviousCustomMode :type Byte))
  ;[1 1 1] Launch if in air after charged Jumping
  (Ability|EndAbility))

(event Ability|EventActivateAbilityFromEvent
  (params (EventData :type Gameplay%20Event%20Data%20Structure%20%28by%20ref%29))
  (bind _mech (Variables|Default|GetMech))
  (bind _mech_2 (Variables|Default|GetMech))
  (bind _mech_3 (Variables|Default|GetMech))
  ;[1 1 1] EVENT -> Activate Ability From Event
  (CallFunction|InitializeAbility EventData)
  (CharacterVehicle|StopEmoting _mech)
  (Gameplay|OnDashStart _mech)
  (if (CallFunction|CanUseDash)
    (Camera|PushCameraMode (Variables|Default|GetDashCameraMode))
    (bind _result (Math|Float|IncrementFloat (Variables|/Game/Athena/DrivableVehicles/Mech/TestMechVehicle.TestMechVehicle_C|GetDashing _mech_3))
      (:%20%20
        (Variables|/Game/Athena/DrivableVehicles/Mech/TestMechVehicle.TestMechVehicle_C|SetDashing _result _mech_3) ; Trigger GC on Mech
        (CallFunction|NotifyHudCooldown)
        (if (Vehicle|VehicleIsinAir _mech_3)
          ;[1 1 1] Launch if in air after charged Jumping
          (Gameplay|AirDash (Variables|Default|GetMech) (Variables|Default|GetDashDirection) (ScalableFloat|GetValueatLevel (Variables|Default|GetDashLaunchSpeedInAir)) (ScalableFloat|GetValueatLevel (Variables|Default|GetDashLaunchZInAir)))
          (Variables|/Game/Athena/DrivableVehicles/Mech/TestMechVehicle.TestMechVehicle_C|SetDashed true _mech_2)
          (Game|Feedback|ClientPlayForceFeedback (Fortnite|GetFortPlayerControllerfromActor _mech_2) (Variables|Default|GetDashRumble) "Dash" false false false)
          (Character|AssignMovementModeChangedDelegate (Variables|Default|GetMech) (delegate Custom|MovementModeChangedDelegate)) ; When we land end ability to reset camera
          (Utilities|FlowControl|Delay (ScalableFloat|GetValueatLevel (Variables|Default|GetDashCooldownResetsCamera))) ; Need delay so we ensure ability ends before another is activated
          (Ability|EndAbility)
          (else
            ;[1 1 1] Root Motion Dash if we are on Ground
            (bind _asynctaskproxy (Ability|Tasks|ApplyRootMotionConstantForce "None" (Variables|Default|GetDashDirection) (ScalableFloat|GetValueatLevel (Variables|Default|GetDashGroundStrength)) (ScalableFloat|GetValueatLevel (Variables|Default|GetDashGroundDuration)) true "/Game/Athena/DrivableVehicles/Mech/OstrichDashCurve.OstrichDashCurve" "ClampVelocity" "0, 0, 0" 5e+04)
              (:then
                (Variables|Default|SetMoveTask _asynctaskproxy)
                (Game|Feedback|ClientPlayForceFeedback (Fortnite|GetFortPlayerControllerfromActor (Variables|Default|GetMech)) (Variables|Default|GetDashRumble) "Dash" false false false)
                (Utilities|FlowControl|Delay (ScalableFloat|GetValueatLevel (Variables|Default|GetDashCameraDuration)))
                (Camera|PopCameraMode))
              (:OnFinish
                (Ability|EndAbility))) ; Apply Root Motioin
          ))))
    (else
      (Ability|EndAbility))))

(event Ability|EventOnEndAbility
  (params (bWasCancelled :type Boolean))
  ;[1 1 1] EVENT -> End Ability
  (bind _asfort_character_vehicle_ostrich (Utilities|Casting|CastToFortCharacterVehicle_Ostrich (Ability|GetOwningActorfromActorInfo)))
  (Gameplay|OnDashFinished _asfort_character_vehicle_ostrich))

(fn CanUseDash
  (return (not (or (Gameplay|GetIsChargingJump (Variables|Default|GetMech)) (Variables|/Game/Athena/DrivableVehicles/Mech/TestMechVehicle.TestMechVehicle_C|GetKicking (Variables|Default|GetMech))))))

(fn InitializeAbility
  (params (GameplayEventData :type Gameplay%20Event%20Data%20Structure%20%28by%20ref%29))
  (Setup|SetupGenericAbility true
    (:Output
      (Variables|Default|SetDashDirection (Ability|EffectContext|GetEffectDirection (Utilities|Struct|BreakGameplayEventData GameplayEventData)))
      (bind _astest_mech_vehicle (Utilities|Casting|CastToTestMechVehicle (Ability|GetAvatarActorfromActorInfo)))
      (Variables|Default|SetMech _astest_mech_vehicle)
      (return))))

(fn NotifyHudCooldown
  (EventDispatchers|CallOnUIGameplayCue (Variables|Default|GetMech) "GameplayCue.Athena.Boost" "Executed" (Ability|GameplayCue|MakeGameplayCueParameters 0.0 (Ability|GetCooldownTimeRemaining))))
