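; Gameplay ability for the drivable Ostrich mech's kick stomp, serving as a timed shell with no stomp or damage logic here.
; On activation it stops mech emoting, runs generic ability setup, then ends after a short delay; a comment mentions a ground root-motion dash, but none is implemented.
; SpawnFX attaches boost, aftermath and eye emitters to mech mesh sockets on thighs, feet and startup lights, though nothing here calls it.
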(event Ability|EventActivateAbility
  (CharacterVehicle|StopEmoting (Variables|Default|GetMech))
  (Setup|SetupGenericAbility true false
    (:Output
      (Utilities|FlowControl|Delay 0.5)
      ;[1 1 1] Root Motion Dash if we are on Ground
      (Ability|EndAbility))))

(fn SpawnFX
  (params (Vehicle :type Fort%20Character%20Vehicle%20Object%20Reference))
  (bind _returnvalue (Mesh|GetMechMesh Vehicle))
  (Effects|SpawnEmitterAttached :EmitterTemplate (Variables|FX|GetP_BoostFin) :AttachToComponent _returnvalue :AttachPointName "FX_Boost_Thigh_L" :Scale "1.000000,1.000000,1.000000" :LocationType "SnapToTarget" :bAutoDestroy true :PoolingMethod "None" :bAutoActivate true)
  (Effects|SpawnEmitterAttached :EmitterTemplate (Variables|FX|GetP_BoostFin) :AttachToComponent _returnvalue :AttachPointName "FX_Boost_Thigh_R" :Scale "1.000000,1.000000,1.000000" :LocationType "SnapToTarget" :bAutoDestroy true :PoolingMethod "None" :bAutoActivate true)
  (Effects|SpawnEmitterAttached :EmitterTemplate (Variables|FX|GetP_BoostFin) :AttachToComponent _returnvalue :AttachPointName "FX_Boost_Foot_L" :Scale "1.000000,1.000000,1.000000" :LocationType "SnapToTarget" :bAutoDestroy true :PoolingMethod "None" :bAutoActivate true)
  (Effects|SpawnEmitterAttached :EmitterTemplate (Variables|FX|GetP_BoostFin) :AttachToComponent _returnvalue :AttachPointName "FX_Boost_Foot_R" :Scale "1.000000,1.000000,1.000000" :LocationType "SnapToTarget" :bAutoDestroy true :PoolingMethod "None" :bAutoActivate true)
  (Effects|SpawnEmitterAttached :EmitterTemplate (Variables|FX|GetP_Boost_Aftermath_LeftFoot) :AttachToComponent _returnvalue :AttachPointName "FX_Boost_Foot_L" :Location "4.000000,-2.000000,-1.000000" :Scale "1.000000,1.000000,1.000000" :LocationType "SnapToTarget" :bAutoDestroy true :PoolingMethod "None" :bAutoActivate true)
  (Effects|SpawnEmitterAttached :EmitterTemplate (Variables|FX|GetP_Boost_Aftermath_RightFoot) :AttachToComponent _returnvalue :AttachPointName "FX_Boost_Foot_R" :Location "4.000000,2.000000,-1.000000" :Scale "1.000000,1.000000,1.000000" :LocationType "SnapToTarget" :bAutoDestroy true :PoolingMethod "None" :bAutoActivate true)
  (Effects|SpawnEmitterAttached :EmitterTemplate (Variables|FX|GetP_Boost_Aftermath_LeftLeg) :AttachToComponent _returnvalue :AttachPointName "FX_Boost_Thigh_L" :Scale "1.000000,1.000000,1.000000" :LocationType "SnapToTarget" :bAutoDestroy true :PoolingMethod "None" :bAutoActivate true)
  (Effects|SpawnEmitterAttached :EmitterTemplate (Variables|FX|GetP_Boost_Aftermath_RightLeg) :AttachToComponent _returnvalue :AttachPointName "FX_Boost_Thigh_R" :Scale "1.000000,1.000000,1.000000" :LocationType "SnapToTarget" :bAutoDestroy true :PoolingMethod "None" :bAutoActivate true)
  (Effects|SpawnEmitterAttached :EmitterTemplate (Variables|FX|GetP_Boost_Eye) :AttachToComponent _returnvalue :AttachPointName "FX_StartUp_Light_L" :Scale "1.000000,1.000000,1.000000" :LocationType "SnapToTarget" :bAutoDestroy true :PoolingMethod "None" :bAutoActivate true)
  (Effects|SpawnEmitterAttached :EmitterTemplate (Variables|FX|GetP_Boost_Eye) :AttachToComponent _returnvalue :AttachPointName "FX_StartUp_Light_R" :Scale "1.000000,1.000000,1.000000" :LocationType "SnapToTarget" :bAutoDestroy true :PoolingMethod "None" :bAutoActivate true))
