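; Gameplay ability for the Ostrich mech's charged super jump, triggered by a gameplay event carrying charge and direction.
; It aborts while kicking, only adds upward velocity while dashing, and otherwise scales launch by charge, moving forward only with directional input.
; High charge server-launches nearby enemy players on foot, crediting the driver as fall instigator.
; Data-driven thresholds optionally enable capsule tracing along the launch path to damage weaker buildings via gameplay effect.
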(event Custom|KnockPlayers
  ;[1 1 1] Knock back nearby players when we jump
  (bind _outactors (Collision|SphereOverlapActors (Transformation|GetActorLocation (Variables|MechJump|GetOstrich)) (Variables|JumpVariables|GetKnockRadius) (Variables|JumpVariables|GetKnockObjects) "/Script/FortniteGame.FortPawn"))
  (for _array_element _outactors
    (bind _asfort_player_pawn (Utilities|Casting|CastToFortPlayerPawn _array_element))
    (if (Vehicle|IsinVehicle _asfort_player_pawn)
      _
      (elif (Team|OnSameTeam _asfort_player_pawn (Variables|/Game/Athena/DrivableVehicles/Mech/TestMechVehicle.TestMechVehicle_C|GetDriverPawn (Variables|MechJump|GetOstrich)))
        _
        (else
          (Utilities|IsServer
            (:true
              (Character|LaunchCharacter _asfort_player_pawn (Math|Vector|MakeVector (.x (* (Math|Vector|Normalize (- (Transformation|GetActorLocation _asfort_player_pawn) (Transformation|GetActorLocation (Variables|MechJump|GetOstrich)))) (Variables|JumpVariables|GetKnockPlayerMagnitude))) (.y (* (Math|Vector|Normalize (- (Transformation|GetActorLocation _asfort_player_pawn) (Transformation|GetActorLocation (Variables|MechJump|GetOstrich)))) (Variables|JumpVariables|GetKnockPlayerMagnitude))) (Math|Float|Max(Float) (+ (.z (* (Math|Vector|Normalize (- (Transformation|GetActorLocation _asfort_player_pawn) (Transformation|GetActorLocation (Variables|MechJump|GetOstrich)))) (Variables|JumpVariables|GetKnockPlayerMagnitude))) 5e+01) 8e+02)) true true)
              (bind _asfort_player_pawn_athena (Utilities|Casting|CastToFortPlayerPawnAthena _asfort_player_pawn))
              (Class|FortPlayerPawnAthena|SetFallInstigator :self _asfort_player_pawn_athena :NewFallInstigator (Game|GetInstigatorController (Vehicle|GetDriver (Variables|MechJump|GetOstrich)))))))))))

(event Ability|EventActivateAbilityFromEvent
  (params (EventData :type Gameplay%20Event%20Data%20Structure%20%28by%20ref%29))
  (bind (_contexthandle _eventmagnitude) (Utilities|Struct|BreakGameplayEventData EventData))
  (bind (_returnvalue_x _returnvalue_y) (Math|Vector|BreakVector (Ability|EffectContext|GetEffectDirection _contexthandle)))
  ;[1 1 1] EVENT -> Activate Ability From Event
  (Setup|SetupGenericAbility true
    (:Output
      (bind _astest_mech_vehicle (Utilities|Casting|CastToTestMechVehicle (Ability|GetAvatarActorfromActorInfo)))
      (bind _output_get (Variables|MechJump|SetOstrich _astest_mech_vehicle))
      (if (Variables|/Game/Athena/DrivableVehicles/Mech/TestMechVehicle.TestMechVehicle_C|GetKicking _output_get)
        (Ability|EndAbility)
        (elif (Gameplay|GetIsDashing _astest_mech_vehicle)
          ;[1 1 1] Launch while dashing
          (bind (_returnvalue_x_2 _returnvalue_y_2 _returnvalue_z) (Math|Vector|BreakVector (Transformation|GetVelocity _astest_mech_vehicle)))
          (Pawn|Character|LaunchCharacterJump _astest_mech_vehicle (Math|Vector|MakeVector _returnvalue_x_2 _returnvalue_y_2 (+ _returnvalue_z (ScalableFloat|GetValueatLevel (Variables|RowLookups|GetJumpZAdditiveWhileDashing)))) false false true true) ; Launch Character if Dashing so movement is smooth.
          (Ability|EndAbility)
          (Networking|ForceNetUpdate (Ability|GetOwningActorfromActorInfo))
          (else
            ;[1 1 1] Set Variables for Jump
            (bind _output_get_2 (Variables|JumpVariables|SetJumpChargePercent _eventmagnitude))
            (Variables|JumpVariables|SetJumpXYMultiplier (Math|Float|SelectFloat 0.5 0.0 (or (> (Math|Float|Absolute(Float) _returnvalue_x) 0.1) (> (Math|Float|Absolute(Float) _returnvalue_y) 0.1)))) ; Scale Forward Motion
            (Variables|/Game/Athena/DrivableVehicles/Mech/TestMechVehicle.TestMechVehicle_C|SetChargeJumping _output_get_2 (Variables|MechJump|GetOstrich))
            (if (>= _eventmagnitude 0.75)
              (CallFunction|KnockPlayers) ; Taking out the destruction for now.
            )
            (CallFunction|Jump)))))))

(event Custom|Jump
  (bind _jumplaunchstrength (Variables|RowLookups|GetJumpLaunchStrength))
  (bind _returnvalue (ScalableFloat|GetValueatLevel _jumplaunchstrength))
  (bind _ostrich (Variables|MechJump|GetOstrich))
  (bind _launchvelocity (CallFunction|CalcLaunchVelocity _ostrich _returnvalue))
  ;[1 1 1] Jump!
  (if (and (>= (Variables|JumpVariables|GetJumpChargePercent) (ScalableFloat|GetValueatLevel (Variables|RowLookups|GetRequiredJumpChargePercentToDestroy))) (> (ScalableFloat|GetValueatLevel (Variables|RowLookups|GetShouldDestroyOnJump)) 0.0))
    (bind _outhits (Collision|MultiCapsuleTraceForObjects (Transformation|GetActorLocation (Variables|MechJump|GetOstrich)) (+ (Transformation|GetActorLocation (Variables|MechJump|GetOstrich)) (* _launchvelocity (Variables|JumpVariables|GetTraceDistanceMult))) 5e+02 5e+02 (Utilities|Array|MakeArray "ObjectTypeQuery1" "ObjectTypeQuery2" "ObjectTypeQuery6")))
    (for _array_element _outhits
      (bind _asbuilding_actor (Utilities|Casting|CastToBuildingActor (.HitActor (Collision|BreakHitResult _array_element))))
      (if (CallFunction|ShouldDamageBuilding _asbuilding_actor 5e+03)
        (GameplayEffects|ApplyGameplayEffectToTarget :self (Ability|GetAbilitySystemComponent (Variables|/Game/Athena/DrivableVehicles/Mech/TestMechVehicle.TestMechVehicle_C|GetDriverPawn (Variables|MechJump|GetOstrich))) :GameplayEffectClass (Variables|Default|GetGE_DestroyBuildingDamage) :Target (Ability|GetAbilitySystemComponent (.HitActor (Collision|BreakHitResult _array_element))) :Level 0.0)))
    (else
      _))
  (Pawn|Character|LaunchCharacterJump _ostrich _launchvelocity true true true true) ; Launch Character if Dashing so movement is smooth.
  (Ability|EndAbility)
  (Networking|ForceNetUpdate (Ability|GetOwningActorfromActorInfo)))

(fn CalcLaunchVelocity
  (params (Vehicle :type Fort%20Vehicle%20Interface%20Interface) (LaunchStrength :type Float%20%28double-precision%29))
  (bind (_returnvalue_x _returnvalue_y) (Math|Vector|BreakVector (Vehicle|GetVehicleForwardVector Vehicle)))
  (return (* (Math|Vector|MakeVector (* _returnvalue_x (* (Variables|JumpVariables|GetJumpXYMultiplier) LaunchStrength)) (* _returnvalue_y (* (Variables|JumpVariables|GetJumpXYMultiplier) LaunchStrength)) LaunchStrength) (Variables|JumpVariables|GetJumpChargePercent))))

(fn CanBeDestroyed
  (params (Building :type Building%20Actor%20Object%20Reference))
  (return (and (> (Health|GetMaxHealth Building :!class /Script/FortniteGame.BuildingActor) 0.0) (<= (Health|GetMaxHealth Building :!class /Script/FortniteGame.BuildingActor) 9e+03))))

(fn GC_Params
  (return (Ability|GameplayCue|MakeGameplayCueParameters 1.0 1.0)))

(fn ShouldDamageBuilding
  (params (Building :type Building%20Actor%20Object%20Reference) (MaxHealth :type Float%20%28double-precision%29))
  (return (Math|Boolean|ANDBoolean (> (Health|GetMaxHealth Building :!class /Script/FortniteGame.BuildingActor) 0.0) (<= (Health|GetMaxHealth Building :!class /Script/FortniteGame.BuildingActor) MaxHealth) (not (Actor|ActorHasTag Building "NoVehicleDamage")) (Variables|/Script/Engine.Actor|GetbCanBeDamaged Building))))
