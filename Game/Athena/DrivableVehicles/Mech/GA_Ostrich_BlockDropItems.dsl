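; Passive gameplay ability on the Ostrich mech, intended to block the driver and gunner from dropping items while in the mech.
; On activation it takes the avatar actor from the actor info, checks validity, and casts its ability system component to the Fort subclass.
; It then blocks abilities matching the item-drop block tag variable on that component.
; An author note says the ability failed to apply properly with an empty event graph.
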(event Ability|EventActivateAbility
  (bind _returnvalue (Ability|GetActorInfo))
  (bind (_owneractor _avataractor) (Utilities|Struct|BreakGameplayAbilityActorInfo _returnvalue))
  ;[1 0 0] This ability is a passive on the mech itself which blocks driver and gunner from dropping while in Mech
  ;;[1 1 1] Apparently need logic within the BP to actually get this to work. If Event Graph is empty it wasn't applying the ability properly.
  (Utilities|IsValid _avataractor
    (:Is%20Valid
      (bind _asfort_ability_system_component (Utilities|Casting|CastToFortAbilitySystemComponent (Ability|GetAbilitySystemComponent _avataractor)))
      (Tags|BlockAbilitiesWithTags _asfort_ability_system_component (Variables|Default|GetBlockItemDropTag)))))
