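; Gameplay ability for the Ostrich mech that loads a single rocket into the pilot's current weapon.
; On activation it runs SetupGenericAbility, casts the avatar's current weapon to the Ostrich ranged weapon, calls OnServerLoadSingleRocket on it, then ends the ability.
; A SpawnFX function attaches boost fin, aftermath and eye emitters to mech mesh sockets, but nothing in this file calls it.
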(event Ability|EventActivateAbility
  (Setup|SetupGenericAbility true false
    (:Output
      (bind _asplayer_pawn_athena (Utilities|Casting|CastToPlayerPawn_Athena (Ability|GetAvatarActorfromActorInfo)))
      (bind _asfort_weapon_ranged_ostrich (Utilities|Casting|CastToFortWeaponRanged_Ostrich (Variables|/Script/FortniteGame.FortPawn|GetCurrentWeapon _asplayer_pawn_athena)))
      (Gameplay|OnServerLoadSingleRocket _asfort_weapon_ranged_ostrich)
      (Ability|EndAbility))))

(fn SpawnFX
  (params (Vehicle :type Fort%20Character%20Vehicle%20Object%20Reference))
  (bind _returnvalue (Mesh|GetMechMesh Vehicle))
  (Effects|SpawnEmitterAttached :EmitterTemplate (Variables|FX|GetP_BoostFin) :AttachToComponent _returnvalue :AttachPointName "FX_Boost_Thigh_L" :Scale "1.000000,1.000000,1.000000" :LocationType "SnapToTarget" :bAutoDestroy true :PoolingMethod "None" :bAutoActivate true)
  (Effects|SpawnEmitterAttached :EmitterTemplate (Variables|FX|GetP_BoostFin) :AttachToComponent _returnvalue :AttachPointName "FX_Boost_Thigh_R" :Scale "1.000000,1.000000,1.000000" :LocationType "SnapToTarget" :bAutoDestroy true :PoolingMethod "None" :bAutoActivate true)
  (Effects|SpawnEmitterAttached :EmitterTemplate (Variables|FX|GetP_BoostFin) :AttachToComponent _returnvalue :AttachPointName "FX_Boost_Foot_L" :Scale "1.000000,1.000000,1.000000" :LocationType "SnapToTarget" :bAutoDestroy true :PoolingMethod "None" :bAutoActivate true)
  (Effects|SpawnEmitterAttached :EmitterTemplate (Variables|FX|GetP_BoostFin) :AttachToComponent _returnvalue :AttachPointName "FX_Boost_Foot_R" :Scale "1.000000,1.000000,1.000000" :LocationType "SnapToTarget" :bAutoDestroy true :PoolingMethod "None" :bAutoActivate true)
  (Effects|SpawnEmitterAttached :EmitterTemplate (Variables|FX|GetP_Boost_Aftermath_LeftFoot) :AttachToComponent _returnvalue :AttachPointName "FX_Boost_Foot_L" :Location "4.000000,-2.000000,-1.000000" :Scale "1.000000,1.000000,1.000000" :LocationType "SnapToTarget" :bAutoDestroy true :PoolingMethod "None" :bAutoActivate true)
  (Effects|SpawnEmitterAttached :EmitterTemplate (Variables|FX|GetP_Boost_Aftermath_RightFoot) :AttachToComponent _returnvalue :AttachPointName "FX_Boost_Foot_R" :Location "4.000000,2.000000,-1.000000" :Scale "1.000000,1.000000,1.000000" :LocationType "SnapToTarget" :bAutoDestroy true :PoolingMethod "None" :bAutoActivate true)
  (Effects|SpawnEmitterAttached :EmitterTemplate (Variables|FX|GetP_Boost_Aftermath_LeftLeg) :AttachToComponent _returnvalue :AttachPointName "FX_Boost_Thigh_L" :Scale "1.000000,1.000000,1.000000" :LocationType "SnapToTarget" :bAutoDestroy true :PoolingMethod "None" :bAutoActivate true)
  (Effects|SpawnEmitterAttached :EmitterTemplate (Variables|FX|GetP_Boost_Aftermath_RightLeg) :AttachToComponent _returnvalue :AttachPointName "FX_Boost_Thigh_R" :Scale "1.000000,1.000000,1.000000" :LocationType "SnapToTarget" :bAutoDestroy true :PoolingMethod "None" :bAutoActivate true)
  (Effects|SpawnEmitterAttached :EmitterTemplate (Variables|FX|GetP_Boost_Eye) :AttachToComponent _returnvalue :AttachPointName "FX_StartUp_Light_L" :Scale "1.000000,1.000000,1.000000" :LocationType "SnapToTarget" :bAutoDestroy true :PoolingMethod "None" :bAutoActivate true)
  (Effects|SpawnEmitterAttached :EmitterTemplate (Variables|FX|GetP_Boost_Eye) :AttachToComponent _returnvalue :AttachPointName "FX_StartUp_Light_R" :Scale "1.000000,1.000000,1.000000" :LocationType "SnapToTarget" :bAutoDestroy true :PoolingMethod "None" :bAutoActivate true))
