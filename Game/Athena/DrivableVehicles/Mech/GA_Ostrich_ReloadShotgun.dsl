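; Gameplay ability that reloads the shotgun on the Ostrich mech vehicle.
; On activation it finds the vehicle from the avatar player pawn and calls OnStartShotgunReload on it.
; With authority only, it starts a two-second timer to ServerReload, which calls LoadShotgunAmmo on the pawn's current Ostrich ranged weapon and ends the ability.
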(event Ability|EventActivateAbility
  (bind _asfort_player_pawn (Utilities|Casting|CastToFortPlayerPawn (Ability|GetAvatarActorfromActorInfo)))
  (bind _asfort_character_vehicle_ostrich (Utilities|Casting|CastToFortCharacterVehicle_Ostrich (Vehicle|GetVehicleActor _asfort_player_pawn :!class /Script/FortniteGame.FortPlayerPawn)))
  (Gameplay|OnStartShotgunReload _asfort_character_vehicle_ostrich)
  (if (Networking|HasAuthority _asfort_player_pawn)
    (Utilities|Time|SetTimerbyEvent (EventDispatchers|CreateEvent :!event ServerReload) 2.0)))

(fn ServerReload
  (bind _asfort_player_pawn (Utilities|Casting|CastToFortPlayerPawn (Ability|GetAvatarActorfromActorInfo)))
  (bind _asfort_weapon_ranged_ostrich (Utilities|Casting|CastToFortWeaponRanged_Ostrich (Variables|/Script/FortniteGame.FortPawn|GetCurrentWeapon _asfort_player_pawn)))
  (Gameplay|LoadShotgunAmmo _asfort_weapon_ranged_ostrich)
  (Ability|EndAbility))
