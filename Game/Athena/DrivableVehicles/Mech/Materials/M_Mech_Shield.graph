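# Translucent, two-sided unlit material for the mech's shield, with color driven by damage and overshield parameters.
# Emissive combines fake cracked reflections, scene-normal glow and intersection edges, stepping between health colors.
# Opacity layers Fresnel, near and depth fades, cloud noise and a sphere-mask erode with bright rim, adding animated tile flicker as damage rises.
# High damage triggers panning-noise vertex wobble; feature-level, quality and shading-path switches simplify costly effects.
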Material /Game/Athena/DrivableVehicles/Mech/Materials/M_Mech_Shield
graph {
 properties {blendMode=BLEND_Translucent, pixelDepthOffsetMode=PDOM_Legacy, shadingModel=MSM_Unlit, twoSided=true, usedShadingModels=Unlit}
 nodes [
  n1 = TextureSample(UVs=ComponentMask(None=ReflectionVectorWS() {bNormalizeCustomWorldNormal=true}) {g=true, r=true}) {samplerType=SAMPLERTYPE_Grayscale, texture=/Game/Athena/WinterArt/Textures/Cracks}
  n2 = ScalarParameter() {defaultValue=1, parameterName=ReflectPower}
  n3 = Multiply(A=ScalarParameter() {parameterName=DamagePerchent}, B=Constant() {r=2})
  n4 = LinearInterpolate(
   A=VectorParameter() {parameterName=Cube Map Multiplier, defaultValue=<rgba 15 1 0 0>}.RGB
   B=VectorParameter() {parameterName=Cube Map Multiplier Low Health, defaultValue=<rgba 5 4.1066 0 1>}.RGB
   Alpha=Divide(A=Round(None=@n3))
  )
  n5 = Multiply(A=Fresnel() {baseReflectFraction=0.02}, B=Multiply(A=Power(Base=@n1.G, Exp=@n2), B=@n4) {constB=1.5})
  n6 = ComponentMask(None=MaterialFunctionCall() {materialFunction=/Engine/Functions/GetPostProcessFunctions/GetAmbientCubemapTint}.AmbientCubemapTint) {b=true, g=true, r=true}
  n7 = SceneTexture() {sceneTextureId=PPI_WorldNormal}
  n8 = ScalarParameter() {defaultValue=500, parameterName=FadeInDistance}
  n9 = Multiply(
   A=Fresnel(Normal=ComponentMask(None=@n7.Color) {b=true, g=true, r=true}) {baseReflectFraction=0, exponent=1.5}
   B=OneMinus(None=DepthFade(FadeDistance=@n8) {fadeDistanceDefault=500})
  )
  n10 = Constant()
  n11 = ShadingPathSwitch(Default=@n9, Deferred=@n9, Forward=@n10, Mobile=@n9)
  n12 = Multiply(
   A=FeatureLevelSwitch(Default=@n11, ES2=@n10, ES3_1=@n10, SM4_REMOVED=@n11)
   B=Multiply(A=ScalarParameter() {defaultValue=15, parameterName=GlassIntensity}, B=ScalarParameter() {parameterName=OverShield})
  )
  n13 = MaterialFunctionCall!(
   Distance Bias=Constant()
   Distance Divisor=ScalarParameter() {defaultValue=50, parameterName=Intersection Edge Divisor}
   Scene Depth
   Pixel Depth
  ) {materialFunction=/Game/Effects/Fort_Effects/Materials/_Functions/MF_DistanceBehindTranslucency_01}
  n14 = ScalarParameter() {defaultValue=6, parameterName=Intersection Edge Exponent}
  n15 = VectorParameter() {parameterName=ColorHighHealth, defaultValue=<rgba 1 0.1 0 1>}
  n16 = Round(None=Multiply(A=ScalarParameter() {parameterName=DamagePerchent}, B=Constant() {r=2}))
  n17 = LinearInterpolate(
   A=LinearInterpolate(A=@n15.RGB, B=VectorParameter() {parameterName=ColorLowHealth, defaultValue=<rgba 1 0.7223 0 1>}.RGB, Alpha=Divide(A=@n16))
   B=@n15.RGB
   Alpha=ScalarParameter() {parameterName=OverShield}
  )
  n18 = Add(
   A=Multiply(
    A=Power(Base=@n13.Normalized Inverted, Exp=@n14)
    B=VectorParameter() {parameterName=Intersection Edge Color, defaultValue=<rgba 5 5 5 1>}.RGB
   )
   B=@n17
  ) {desc=+ intersection edge}
  n19 = Multiply(A=@n17, B=ScalarParameter() {defaultValue=1, parameterName=emissiveMulti})
  Add(A=Desaturation(None=Multiply(A=@n5, B=@n6), Fraction=Constant()), B=Add(A=Multiply(A=@n12, B=@n4), B=Multiply(A=@n18, B=@n19)))
  Constant() {r=1}
  n20 = Subtract(A=ScalarParameter() {parameterName=DamagePerchent}) {constB=0.75}
  Multiply(A=@n20, B=Constant() {r=100})
  n21 = Constant()
  n22 = ScalarParameter() {defaultValue=1, parameterName=AddFlickerNoise}
  n23 = ScalarParameter() {parameterName=DamagePerchent}
  n24 = Constant() {r=0.5}
  n25 = Saturate(None=Multiply(A=Subtract(A=@n23, B=@n24), B=Constant() {r=100}))
  n26 = LinearInterpolate(
   A=ScalarParameter() {defaultValue=0.1, parameterName=Opacity}
   B=ScalarParameter() {defaultValue=1, parameterName=Opacitysides}
   Alpha=Fresnel() {baseReflectFraction=0}
  )
  n27 = PixelDepth()
  n28 = Constant() {r=-100}
  n29 = Saturate(None=Divide(A=Add(A=@n27.R, B=@n28), B=ScalarParameter() {defaultValue=25, parameterName=NearFade}))
  n30 = Add(A=TextureCoordinate() {uTiling=2, vTiling=2}, B=Constant2Vector() {g=-1.25, r=-0.5})
  n31 = TextureSample(UVs=Panner(Coordinate=MaterialFunctionCall(Vector or UVs=@n30) {materialFunction=&VectorToRadialValue}.Radial Coordinates) {speedX=-0.15, speedY=0.35}) {texture=/Game/Effects/Fort_Effects/Textures/Tile/T_CloudScape_08}
  n32 = Constant() {r=1}
  n33 = SphereMask(
   A=TextureCoordinate() {uTiling=2}
   B=ComponentMask(None=VectorParameter() {parameterName=PannerEdge, defaultValue=<rgba 0.55 0.6 0 0>}.RGB) {g=true, r=true}
   Radius=ScalarParameter() {defaultValue=0.5, parameterName=Param_1}
   Hardness=ScalarParameter() {defaultValue=0.5, parameterName=PannerEdge}
  )
  n34 = ComponentMask(None=VectorParameter() {parameterName=CircleMiddle, defaultValue=<rgba 0.55 0.6 0 0>}.RGB) {g=true, r=true}
  n35 = ScalarParameter() {defaultValue=1, parameterName=FadeDown}
  n36 = Saturate(None=SphereMask(A=TextureCoordinate() {uTiling=2}, B=@n34, Radius=@n35, Hardness=ScalarParameter() {defaultValue=0.9, parameterName=ErodeMask_Harness}))
  n37 = ScalarParameter() {defaultValue=0.001, parameterName=Subtract_EdgeSize}
  n38 = TextureCoordinate() {uTiling=2}
  n39 = Subtract(A=@n35, B=ScalarParameter() {defaultValue=0.01, parameterName=Subtract_EdgeSize_Inner})
  n40 = ScalarParameter() {defaultValue=0.9, parameterName=ErodeMask_Harness_Edge}
  n41 = Multiply(
   A=SphereMask(
    A=TextureCoordinate() {uTiling=2}
    B=@n34
    Radius=Subtract(A=@n35, B=@n37)
    Hardness=ScalarParameter() {defaultValue=0.9, parameterName=ErodeMask_Harness_Edge}
   )
   B=OneMinus(None=SphereMask(A=@n38, B=@n34, Radius=@n39, Hardness=@n40))
  )
  n42 = ScalarParameter() {defaultValue=50, parameterName=Edge_Mult}
  n43 = Multiply(
   A=Multiply(A=Multiply(A=@n26, B=@n29), B=LinearInterpolate(A=@n31.G, B=@n32, Alpha=@n33))
   B=Add(A=OneMinus(None=@n36), B=Multiply(A=@n41, B=@n42))
  )
  n44 = DepthFade(Opacity=@n43, FadeDistance=ScalarParameter() {defaultValue=25, parameterName=DepthFade_Distance})
  n45 = FeatureLevelSwitch(Default=@n44, ES2=@n43, ES3_1=@n43, SM4_REMOVED=@n44)
  n46 = MaterialFunctionCall(Speed=ScalarParameter() {defaultValue=5, parameterName=Shield_FadeIn_Speed}) {materialFunction=&DebugTimeSine}
  n47 = Multiply(A=Subtract(A=ScalarParameter() {parameterName=DamagePerchent}) {constB=0.75}, B=Constant() {r=100})
  n48 = LinearInterpolate(
   A=@n45
   B=Multiply(A=@n45, B=Saturate(None=@n46.Result))
   Alpha=LinearInterpolate(A=Constant(), B=ScalarParameter() {defaultValue=1, parameterName=Shield_FadeIn}, Alpha=Saturate(None=@n47))
  )
  n49 = Multiply(
   A=ComponentMask(None=TextureCoordinate()) {g=true, r=true}
   B=ScalarParameter() {defaultValue=1.65, group=EQ Shape, parameterName=X_Nr_Bands}
  ) {constB=10}
  n50 = ScalarParameter() {defaultValue=0.25, parameterName="Rotation Angle (0-1) (S)"}
  n51 = TextureSample(UVs=Panner(Coordinate=MaterialFunctionCall(UVs=@n49, "Rotation Angle (0-1)"=@n50) {materialFunction=&CustomRotator}.Rotated Values) {speedX=0.06, speedY=0.02}) {samplerType=SAMPLERTYPE_LinearColor, texture=/Game/Global/Textures/Default/Tileable_Patterns/DiscoTiles_M}
  n52 = MaterialFunctionCall(Speed=ScalarParameter() {defaultValue=0.6, parameterName="Speed2 (S)"}) {materialFunction=&DebugTimeSine}
  n53 = MaterialFunctionCall(Speed=ScalarParameter() {defaultValue=0.15, parameterName="Speed (S)"}) {materialFunction=&DebugTimeSine}
  n54 = Constant() {r=0.5}
  n55 = Constant() {r=1}
  n56 = Multiply(A=ComponentMask(None=TextureCoordinate()) {g=true, r=true}, B=ScalarParameter() {defaultValue=2.5, group=EQ Shape, parameterName=Y_Nr_Bands}) {constB=10}
  n57 = Panner(Coordinate=OneMinus(None=Frac(None=@n56))) {speedX=-0.04, speedY=0.1}
  n58 = ComponentMask(None=Multiply(A=TextureSample(UVs=@n57) {samplerType=SAMPLERTYPE_Masks, texture=/Game/Global/Textures/Default/Tileable_Patterns/DiscoFloors_M}.R) {constB=0.5}) {g=true, r=true}
  n59 = MaterialFunctionCall(Speed=ScalarParameter() {defaultValue=0.6, parameterName="Speed2 (S)"}) {materialFunction=&DebugTimeSine}
  n60 = AppendVector(
   A=@n59.Result
   B=MaterialFunctionCall(Speed=ScalarParameter() {defaultValue=1, parameterName="Speed3 (S)"}) {materialFunction=&DebugTimeSine}.Result
  )
  n61 = LinearInterpolate(A=Constant() {r=0.5}, B=Constant() {r=1}, Alpha=@n59.Result)
  n62 = ScalarParameter() {defaultValue=0.8, parameterName=Hardness}
  n63 = ComponentMask(None=TextureCoordinate()) {g=true, r=true}
  n64 = ScalarParameter() {defaultValue=2.7, group=EQ Shape, parameterName=Z_Nr_Bands}
  n65 = OneMinus(None=Frac(None=Multiply(A=@n63, B=@n64) {constB=10}))
  n66 = Multiply(A=TextureSample(UVs=Panner(Coordinate=@n65) {speedX=0.05, speedY=-0.02}) {samplerType=SAMPLERTYPE_LinearColor, texture=/Game/Weapons/WeaponSkins/Textures/DiscoTiles1_M}.RGB) {constB=0.5}
  n67 = MaterialFunctionCall(Speed=ScalarParameter() {defaultValue=0.6, parameterName="Speed2 (S)"}) {materialFunction=&DebugTimeSine}
  n68 = MaterialFunctionCall(Speed=ScalarParameter() {defaultValue=1, parameterName="Speed3 (S)"}) {materialFunction=&DebugTimeSine}
  n69 = Constant() {r=0.5}
  n70 = Constant() {r=1}
  n71 = LinearInterpolate(
   A=SphereMask(
    A=ComponentMask(None=@n51.RGB) {g=true, r=true}
    B=AppendVector(A=@n52.Result, B=@n53.Result)
    Radius=LinearInterpolate(A=@n54, B=@n55, Alpha=@n52.Result)
    Hardness=ScalarParameter() {defaultValue=0.8, parameterName=Hardness}
   )
   B=Multiply(A=ScalarParameter() {defaultValue=3, parameterName=Param_2}, B=SphereMask(A=@n58, B=@n60, Radius=@n61, Hardness=@n62))
   Alpha=SphereMask(
    A=ComponentMask(None=@n66) {g=true, r=true}
    B=AppendVector(A=@n67.Result, B=@n68.Result)
    Radius=LinearInterpolate(A=@n69, B=@n70, Alpha=@n67.Result)
    Hardness=ScalarParameter() {defaultValue=0.8, parameterName=Hardness}
   )
  )
  Multiply(
   A=LinearInterpolate(A=@n48, B=Multiply(A=@n48, B=@n71), Alpha=LinearInterpolate(A=@n21, B=@n22, Alpha=@n25))
   B=ScalarParameter() {defaultValue=1, parameterName=OverShieldPulse}
  )
  ScalarParameter() {parameterName=OverShield}
  ScalarParameter() {parameterName=DamagePerchent}
  ScalarParameter() {defaultValue=1, parameterName=FadeDown}
  n72 = Constant3Vector()
  n73 = PixelDepth()
  n74 = Constant() {r=-100}
  n75 = Saturate(None=Divide(A=Add(A=@n73.R, B=@n74), B=ScalarParameter() {defaultValue=4000, parameterName=WPO_Scale_DistanceFade}))
  n76 = Multiply(
   A=TextureCoordinate()
   B=LinearInterpolate(
    A=ScalarParameter() {defaultValue=3, parameterName=WP_TestureSize_Close}
    B=ScalarParameter() {defaultValue=1, parameterName=WP_TestureSize_Far}
    Alpha=@n75
   )
  )
  n77 = TextureSample!(UVs=Panner(Coordinate=@n76) {speedY=3}, Tex, Level, Apply View MipBias) {mipValueMode=TMVM_MipLevel, samplerType=SAMPLERTYPE_LinearColor, texture=&T_NoisePack_LinearColor}
  n78 = TextureSample(UVs=Panner(Coordinate=@n76) {speedX=-1, speedY=-2}) {mipValueMode=TMVM_MipLevel, samplerType=SAMPLERTYPE_LinearColor, texture=&T_NoisePack_LinearColor}
  n79 = TextureSample(UVs=Panner(Coordinate=@n76) {speedX=1, speedY=1.25}) {mipValueMode=TMVM_MipLevel, samplerType=SAMPLERTYPE_LinearColor, texture=&T_NoisePack_LinearColor}
  n80 = Constant3Vector() {constant=<rgba 0 1 0 0>}
  n81 = TextureCoordinate() {uTiling=2, vTiling=2}
  n82 = Add(
   A=AppendVector(A=AppendVector(A=@n77.R, B=@n78.G), B=@n79.B)
   B=Multiply(A=Transform(None=@n80) {transformType=TRANSFORM_Local}, B=ComponentMask(None=@n81) {g=true})
  )
  n83 = LinearInterpolate(
   A=Multiply(A=VertexNormalWS(), B=VectorParameter() {parameterName=WPO_Mult_Close, defaultValue=<rgba 20 20 20 1>}.RGB)
   B=Multiply(A=VertexNormalWS(), B=VectorParameter() {parameterName=WPO_Mult_Far, defaultValue=<rgba 50 50 50 1>}.RGB)
   Alpha=@n75
  )
  n84 = TextureCoordinate() {uTiling=2}
  n85 = ComponentMask(None=VectorParameter() {parameterName=CircleMiddle, defaultValue=<rgba 0.55 0.6 0 0>}.RGB) {g=true, r=true}
  n86 = ScalarParameter() {defaultValue=0.45, parameterName=WPO_Radius}
  n87 = ScalarParameter() {defaultValue=0.9, parameterName=ErodeMask_Harness}
  n88 = LinearInterpolate(
   A=@n72
   B=LinearInterpolate(A=@n72, B=Multiply(A=@n82, B=@n83), Alpha=SphereMask(A=@n84, B=@n85, Radius=@n86, Hardness=@n87))
   Alpha=Saturate(None=@n20)
  )
  n89 = VertexNormalWS()
  ShadingPathSwitch(Default=QualitySwitch(Default=FeatureLevelSwitch(Default=@n88, ES2=@n89, ES3_1=@n89, SM4_REMOVED=@n88), Low=@n89), Forward=@n89)
 ]
}
