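# Translucent, unlit, two-sided material for the mech's dome overshield, drawing an animated, glowing smoky energy surface.
# Panning smoke noise and a sine-pulsed, view-angle curve ramp drive bump-offset noise layers, with an object-space layer dropped on low-quality, mobile and forward paths.
# That noise feeds overshield tints, combined with health colors stepped by DamagePerchent.
# Opacity combines Fresnel edges, near-camera fade and an OverShieldPulse multiplier.
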Material /Game/Athena/DrivableVehicles/Mech/Materials/M_Mech_Shield_OverShieldOnly
graph {
 properties {
  blendMode=BLEND_Translucent
  pixelDepthOffsetMode=PDOM_Legacy
  previewMesh=/Game/Athena/DrivableVehicles/Mech/Meshes/Mech_Dome
  shadingModel=MSM_Unlit
  twoSided=true
  usedShadingModels=Unlit
 }
 nodes [
  n1 = Multiply(A=ScalarParameter() {parameterName=DamagePerchent}, B=Constant() {r=2})
  n2 = LinearInterpolate(
   A=VectorParameter() {parameterName=Cube Map Multiplier, defaultValue=<rgba 10 0.5 0 0>}.RGB
   B=VectorParameter() {parameterName=Cube Map Multiplier Low Health, defaultValue=<rgba 5 4.1066 0 1>}.RGB
   Alpha=Divide(A=Round(None=@n1))
  )
  n3 = TextureCoordinate()
  n4 = Panner(Coordinate=@n3) {speedX=0.4, speedY=0.6}
  n5 = Panner(Coordinate=@n3) {speedX=-0.4, speedY=-0.5}
  n6 = VertexNormalWS()
  n7 = CameraVectorWS()
  n8 = CurveAtlasRowParameter(CurveTime=Abs(None=DotProduct(A=@n6, B=@n7))) {
   atlas=&AthenaCurveAtlas01
   curve=/Game/Athena/Environments/Nexus/Materials/FXshader/Curve_Ramps
   defaultValue=79
  }
  n9 = ScalarParameter() {defaultValue=0.5, parameterName=OverShield__Sine_Speed}
  n10 = Add(
   A=Subtract(
    A=TextureSample(UVs=@n4) {texture=&T_Inky_Smoke_Tile}.R
    B=TextureSample(UVs=@n5) {texture=&T_Inky_Smoke_Tile}.G
   )
   B=Multiply(
    A=Subtract(A=@n8.R)
    B=MaterialFunctionCall(Speed=@n9) {materialFunction=&DebugTimeSine}.Result
   )
  )
  n11 = ScalarParameter() {defaultValue=1, parameterName=Hight}
  n12 = Add(B=BumpOffset(Coordinate=TextureCoordinate(), Height=@n10, HeightRatioInput=@n11))
  n13 = Multiply(A=@n12) {constB=1.618}
  n14 = Multiply(
   A=TextureSample(UVs=Panner(Coordinate=@n12) {speedX=-0.6, speedY=0.6}) {texture=&T_Inky_Smoke_Tile}.R
   B=TextureSample(UVs=Panner(Coordinate=@n13) {speedX=1, speedY=0.6}) {texture=&T_Inky_Smoke_Tile}.R
  )
  n15 = WorldPosition()
  n16 = Divide(A=ComponentMask(None=TransformPosition(None=@n15.XYZ) {transformSourceType=TRANSFORMPOSSOURCE_World}) {g=true, r=true}, B=ObjectRadius())
  n17 = TextureSample(UVs=Panner(Coordinate=BumpOffset(Coordinate=@n16, Height=@n10, HeightRatioInput=@n11)) {speedY=1}) {texture=&T_Inky_Smoke_Tile}
  n18 = LinearInterpolate(A=@n14, B=@n17.R, Alpha=@n17.R)
  n19 = ShadingPathSwitch(Default=QualitySwitch(Default=FeatureLevelSwitch(Default=@n18, ES2=@n14, ES3_1=@n14, SM4_REMOVED=@n18), Low=@n14), Forward=@n14)
  n20 = Multiply(A=@n19, B=ScalarParameter() {defaultValue=3, parameterName=NoiseScale})
  n21 = Power(
   Base=Multiply(A=@n20, B=VectorParameter() {parameterName=Tint, defaultValue=<rgba 1 1 1 0.5>}.RGB)
   Exp=ScalarParameter() {defaultValue=1, parameterName=Exp}
  )
  n22 = Saturate(None=@n19)
  n23 = Multiply(A=Multiply(A=Multiply(A=@n22, B=@n8.B)) {constB=5}, B=VectorParameter() {parameterName=TextureTint, defaultValue=<rgba 1 2 1 1>}.RGB)
  n24 = Add(
   A=VectorParameter() {parameterName=ColorHighHealth, defaultValue=<rgba 1 0 0 1>}.RGB
   B=Multiply(A=VectorParameter() {parameterName=Color_OverShield, defaultValue=<rgba 5 0.75 0 1>}.RGB, B=Add(A=@n21, B=@n23))
  )
  n25 = Multiply(A=@n24, B=ScalarParameter() {defaultValue=1, parameterName=emissiveMulti})
  Add(A=@n2, B=Add(A=Multiply(A=@n2, B=@n20), B=Multiply(A=@n24, B=@n25)))
  n26 = LinearInterpolate(
   A=ScalarParameter() {defaultValue=0.02, parameterName=Opacity}
   B=ScalarParameter() {defaultValue=4, parameterName=Opacitysides}
   Alpha=Fresnel() {baseReflectFraction=0}
  )
  n27 = PixelDepth()
  n28 = Constant() {r=-100}
  n29 = Saturate(None=Divide(A=Add(A=@n27.R, B=@n28), B=ScalarParameter() {defaultValue=25, parameterName=NearFade}))
  Multiply(A=Multiply(A=Multiply(A=@n26, B=@n29), B=Constant() {r=1}), B=ScalarParameter() {defaultValue=1, parameterName=OverShieldPulse})
 ]
}
