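# Translucent, two-sided mesh-particle material for the Ostrich mech's shield-break effect.
# Emissive combines crack-textured Fresnel reflections, scene-normal glow near geometry, intersection edges and health/overshield colors, with cheaper ES2/ES3.1 paths.
# Opacity blends by Fresnel, fades near camera and at intersections, and is modulated by radially panned clouds.
# The Erode dynamic parameter drives alpha erosion through a tiled crack texture, letting particles break the shield apart.
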Material /Game/Athena/DrivableVehicles/Mech/Materials/M_Ostrich_ShieldBreak
graph {
 properties {bUsedWithMeshParticles=true, blendMode=BLEND_Translucent, pixelDepthOffsetMode=PDOM_Legacy, twoSided=true, usedShadingModels=Default Lit}
 nodes [
  n1 = TextureSample(UVs=ComponentMask(None=ReflectionVectorWS() {bNormalizeCustomWorldNormal=true}) {g=true, r=true}) {samplerType=SAMPLERTYPE_Grayscale, texture=/Game/Athena/WinterArt/Textures/Cracks}
  n2 = ScalarParameter() {defaultValue=1, parameterName=ReflectPower}
  n3 = ParticleColor()
  n4 = Multiply(A=Fresnel() {baseReflectFraction=0.02}, B=Multiply(A=Power(Base=@n1.G, Exp=@n2), B=@n3.RGB) {constB=1.5})
  n5 = ComponentMask(None=MaterialFunctionCall() {materialFunction=/Engine/Functions/GetPostProcessFunctions/GetAmbientCubemapTint}.AmbientCubemapTint) {b=true, g=true, r=true}
  n6 = SceneTexture() {sceneTextureId=PPI_WorldNormal}
  n7 = ScalarParameter() {defaultValue=500, parameterName=FadeInDistance}
  n8 = Multiply(
   A=Fresnel(Normal=ComponentMask(None=@n6.Color) {b=true, g=true, r=true}) {baseReflectFraction=0, exponent=1.5}
   B=OneMinus(None=DepthFade(FadeDistance=@n7) {fadeDistanceDefault=500})
  )
  n9 = Constant()
  n10 = ShadingPathSwitch(Default=@n8, Deferred=@n8, Forward=@n9, Mobile=@n8)
  n11 = ScalarParameter() {defaultValue=15, parameterName=GlassIntensity}
  n12 = ScalarParameter() {parameterName=OverShield}
  n13 = Multiply(A=Multiply(A=FeatureLevelSwitch(Default=@n10, ES2=@n9, ES3_1=@n9, SM4_REMOVED=@n10), B=Multiply(A=@n11, B=@n12)), B=@n3.RGB)
  n14 = Constant()
  n15 = ScalarParameter() {defaultValue=50, parameterName=Intersection Edge Divisor}
  n16 = Multiply(
   A=Power(
    Base=MaterialFunctionCall!(Distance Bias=@n14, Distance Divisor=@n15, Scene Depth, Pixel Depth) {materialFunction=/Game/Effects/Fort_Effects/Materials/_Functions/MF_DistanceBehindTranslucency_01}.Normalized Inverted
    Exp=ScalarParameter() {defaultValue=6, parameterName=Intersection Edge Exponent}
   )
   B=VectorParameter() {parameterName=Intersection Edge Color, defaultValue=<rgba 5 5 5 1>}.RGB
  )
  n17 = VectorParameter() {parameterName=ColorHighHealth, defaultValue=<rgba 1 0.1 0 1>}
  n18 = LinearInterpolate(
   A=LinearInterpolate(
    A=@n17.RGB
    B=VectorParameter() {parameterName=ColorLowHealth, defaultValue=<rgba 0.1 0 0.5 1>}.RGB
    Alpha=ScalarParameter() {parameterName=DamagePerchent}
   )
   B=Add(A=@n17.RGB)
   Alpha=ScalarParameter() {parameterName=OverShield}
  )
  n19 = ScalarParameter() {defaultValue=1, parameterName=emissiveMulti}
  n20 = Add(
   A=FeatureLevelSwitch(Default=@n13, ES2=@n3.RGB, ES3_1=@n3.RGB, SM4_REMOVED=@n13)
   B=Multiply(A=Add(A=@n16, B=@n18) {desc=+ intersection edge}, B=Multiply(A=@n18, B=@n19))
  )
  n21 = Add(A=Desaturation(None=Multiply(A=@n4, B=@n5), Fraction=Constant()), B=@n20)
  MaterialFunctionCall(Emissive=FeatureLevelSwitch(Default=@n21, ES2=@n20, ES3_1=@n20, SM4_REMOVED=@n21)) {materialFunction=&MF_FX_EmissiveAdjustments}
  n22 = Add(A=PixelDepth().R, B=Constant() {r=-100})
  n23 = ScalarParameter() {defaultValue=25, parameterName=NearFade}
  n24 = Multiply(
   A=LinearInterpolate(
    A=ScalarParameter() {defaultValue=0.1, parameterName=Opacity}
    B=ScalarParameter() {defaultValue=1, parameterName=Opacitysides}
    Alpha=Fresnel() {baseReflectFraction=0}
   )
   B=Saturate(None=Divide(A=@n22, B=@n23))
  )
  n25 = MaterialFunctionCall(Vector or UVs=Add(A=TextureCoordinate() {uTiling=2, vTiling=2}, B=Constant2Vector() {g=-1.25, r=-0.5})) {materialFunction=&VectorToRadialValue}
  n26 = VectorParameter() {parameterName=PannerEdge, defaultValue=<rgba 0.55 0.6 0 0>}
  n27 = LinearInterpolate(
   A=TextureSample(UVs=Panner(Coordinate=@n25.Radial Coordinates) {speedX=-0.15, speedY=0.35}) {texture=/Game/Effects/Fort_Effects/Textures/Tile/T_CloudScape_08}.G
   B=Constant() {r=1}
   Alpha=SphereMask(
    A=TextureCoordinate() {uTiling=2}
    B=ComponentMask(None=@n26.RGB) {g=true, r=true}
    Radius=ScalarParameter() {defaultValue=0.5, parameterName=Param_1}
    Hardness=ScalarParameter() {defaultValue=0.5, parameterName=PannerEdge}
   )
  )
  n28 = Constant() {r=1}
  n29 = FeatureLevelSwitch(Default=@n27, ES2=@n28, ES3_1=@n28, SM4_REMOVED=@n27)
  n30 = TextureSample(UVs=TextureCoordinate() {uTiling=3, vTiling=3}) {samplerType=SAMPLERTYPE_Grayscale, texture=&Cracks}
  n31 = ScalarParameter() {defaultValue=4, parameterName=ErodePower}
  Multiply(
   A=DepthFade(Opacity=Multiply(A=@n24, B=@n29), FadeDistance=ScalarParameter() {defaultValue=25, parameterName=DepthFade_Distance})
   B=MaterialFunctionCall!(Dynamic Param=DynamicParameter() {paramNames [Erode,Param2,Param3,Param4]}.Erode, Alpha=Power(Base=@n30.G, Exp=@n31)) {materialFunction=/Game/Effects/Fort_Effects/Materials/_Functions/Erode_Alpha}.Result
  )
 ]
}
