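# Translucent unlit glow material for the drivable Mech's skeletal mesh, showing dash cooldown, overshield and damage state.
# A UV gradient plus sphere mask separates the dash gauge, which fills along V and shifts to GlowFull as cooldown completes, from other glowing parts.
# OverShield and DamagePerchent blend the colors toward shielded, low or gone.
# An emissive texture masks the result, modulated by layered sine flicker and panning smoke, which mobile replaces with a constant.
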Material /Game/Athena/DrivableVehicles/Mech/Materials/Mech_Glow
graph {
 properties {
  bUsedWithMeshDeformer=true
  bUsedWithSkeletalMesh=true
  blendMode=BLEND_Translucent
  pixelDepthOffsetMode=PDOM_Legacy
  shadingModel=MSM_Unlit
  usedShadingModels=Unlit
 }
 nodes [
  n1 = ComponentMask(None=VectorParameter() {parameterName=TileTex02, defaultValue=<rgba 1 1 0 1>}.RGB) {g=true, r=true}
  Add(A=DynamicParameter() {defaultValue=<rgba 1 5 1 1>, paramNames [RandTile,ColorMult,TileMult,Param4]}.RandTile, B=@n1)
  n2 = TextureCoordinate()
  n3 = Time()
  n4 = ScalarParameter() {parameterName=DashGameTime}
  n5 = Saturate(None=Divide(A=Subtract(A=@n3, B=@n4), B=ScalarParameter() {defaultValue=5, parameterName=DashCooldownDuration}))
  n6 = Floor(None=Add(A=ComponentMask(None=@n2) {g=true}, B=@n5))
  n7 = Add(
   A=TextureSample(UVs=TextureCoordinate() {uTiling=2}) {samplerSource=SSM_Clamp_WorldGroupSettings, texture=/Game/Tools/Turntable/Textures/T_ThinkBase_Teleporter_Gradient}.G
   B=SphereMask(
    A=TextureCoordinate()
    B=Constant2Vector() {g=0.25, r=0.75}
    Radius=ScalarParameter() {defaultValue=0.25, parameterName=MaskOutRound_Radius}
    Hardness=ScalarParameter() {defaultValue=1, parameterName=MaskOutRound_Hardness}
   )
  )
  n8 = VectorParameter() {parameterName=GlowOtherParts, defaultValue=<rgba 4 0.25 0 1>}
  n9 = LinearInterpolate(A=VectorParameter() {parameterName=Glow, defaultValue=<rgba 4 3.3862 0 1>}.RGB, B=@n8.RGB, Alpha=@n7)
  n10 = ShadingPathSwitch(
   Default=VectorParameter() {parameterName=GlowOverShield, defaultValue=<rgba 100 5 0 1>}.RGB
   Mobile=VectorParameter() {parameterName=GlowOverShield_Mobile, defaultValue=<rgba 3 0.15 0 1>}.RGB
  ) {desc=Migrated from FeatureLevelSwitch}
  n11 = ScalarParameter() {parameterName=OverShield}
  n12 = Round(None=Multiply(A=ScalarParameter() {parameterName=DamagePerchent}, B=Constant() {r=2}))
  n13 = Subtract(A=ScalarParameter() {parameterName=DamagePerchent}) {constB=0.9999}
  n14 = LinearInterpolate(
   A=LinearInterpolate(
    A=LinearInterpolate(A=@n9, B=@n10, Alpha=@n11)
    B=VectorParameter() {parameterName=GlowLow, defaultValue=<rgba 2 1.7003 0 1>}.RGB
    Alpha=Divide(A=@n12)
   )
   B=VectorParameter() {parameterName=GlowGone, defaultValue=<rgba 0 0 0 1>}.RGB
   Alpha=Saturate(None=Ceil(None=@n13))
  )
  n15 = LinearInterpolate(
   A=@n14
   B=LinearInterpolate(A=VectorParameter() {parameterName=GlowFull, defaultValue=<rgba 4 0 0 1>}.RGB, B=@n14, Alpha=@n7)
   Alpha=Saturate(None=Power(Base=@n5) {constExponent=4})
  )
  n16 = LinearInterpolate(Alpha=ScalarParameter() {defaultValue=1, parameterName=RandomOffsetPerClient_Mesh})
  n17 = Add(A=Time() {bOverride_Period=true, period=15}, B=@n16)
  n18 = Add(A=Time() {bOverride_Period=true, period=5.235}, B=@n16)
  n19 = Add(A=Time() {bOverride_Period=true, period=3}, B=@n16)
  n20 = LinearInterpolate(
   A=LinearInterpolate(Alpha=Sine(None=@n17) {period=2.6}) {constA=0.7}
   B=LinearInterpolate(Alpha=Sine(None=@n18) {period=1.57}) {constA=0.5}
   Alpha=LinearInterpolate(A=Constant() {r=0.6}, B=Constant() {r=1}, Alpha=Sine(None=@n19) {period=1.24}) {constA=0.15, constB=0.85}
  )
  n21 = TextureCoordinate()
  n22 = Multiply(A=@n21, B=@n1) {constB=1.7}
  n23 = VectorParameter() {parameterName=PannerSpeed, defaultValue=<rgba 0 -0.5 0 1>}
  n24 = Panner(Coordinate=@n22, Speed=ComponentMask(None=Multiply(B=@n23.RGB) {constA=0.5}) {g=true, r=true}) {speedY=-1}
  n25 = ScalarParameter() {defaultValue=0.25, parameterName="Rotation Angle (0-1) (S)"}
  n26 = Panner(Coordinate=@n22, Speed=ComponentMask(None=@n23.RGB) {g=true, r=true}) {speedY=-1}
  n27 = ScalarParameter() {defaultValue=0.25, parameterName="Rotation Angle (0-1) (S)"}
  n28 = Multiply(
   A=TextureSample(UVs=MaterialFunctionCall(UVs=@n24, "Rotation Angle (0-1)"=@n25) {materialFunction=&CustomRotator}.Rotated Values) {texture=&T_Inky_Smoke_Tile_Chromatic}.G
   B=TextureSample(UVs=MaterialFunctionCall(UVs=@n26, "Rotation Angle (0-1)"=@n27) {materialFunction=&CustomRotator}.Rotated Values) {texture=&T_Inky_Smoke_Tile}.G
  )
  n29 = Clamp(None=ShadingPathSwitch(Default=Add(A=@n20, B=@n28), Mobile=Constant() {r=1}) {desc=Migrated from FeatureLevelSwitch}) {maxDefault=5, minDefault=0.5}
  Multiply(
   A=LinearInterpolate(A=Saturate(None=@n6), Alpha=Round(None=@n7))
   B=Multiply(A=Multiply(A=@n15, B=@n29), B=TextureSample() {texture=/Game/Athena/DrivableVehicles/Mech/Textures/T_Mech_Emmive}.G)
  )
  Multiply(A=DynamicParameter() {defaultValue=<rgba 1 5 1 1>, paramNames [RandTile,ColorMult,TileMult,Param4]}.TileMult, B=@n21)
  ShadingPathSwitch(Default=@n8.RGB, Mobile=Saturate(None=@n8.RGB)) {desc=Migrated from FeatureLevelSwitch}
 ]
}
