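# Emote animation Blueprint for the Ostrich mech vehicle, running on the Fortnite player skeleton.
# Each update copies the 2D movement flag from the mech mesh's anim instance, selecting FullBodyInMotion or FullBody slots over a cached dual-wield pose, followed by an UpperBody slot.
# PropScale then enlarges hand, prop-root and head bones.
# A Copy Bone Delta on the attach bone multiplies translation and scale by PropScale's largest component.
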AnimBlueprint /Game/Athena/DrivableVehicles/Mech/Meshes/Mech_Emote_AnimBP
target_skeleton=&Fortnite_M_Avg_Player_Skeleton
graphs [
 AnimGraph {nodes [
   n1 = Blend Poses by bool(
    BlendPose_0=Slot 'FullBodyInMotion'(Source=Use cached pose 'Base Pose'()) {slotName=FullBodyInMotion}
    BlendPose_1=Slot 'FullBody'(Source=Use cached pose 'Base Pose'()) {slotName=FullBody}
    bActiveValue=Get Is Moving 2D() -> (Is Moving 2D:bool)
   ) {blendType=Linear}
   n2 = Get PropScale() -> (PropScale:Vector)
   n3 = "Transform (Modify) Bone - Bone: hand_l"(
    ComponentPose=Local To Component(LocalPose=Slot 'UpperBody'(Source=@n1) {slotName=UpperBody})
    Alpha=1
    Translation="0,0,0"
    Rotation="0,0,0"
    Scale=@n2
   ) {scaleMode=BMM_Replace, alphaBoolBlend {blendOption=Linear}, boneToModify {boneName=hand_l}}
   n4 = "Transform (Modify) Bone - Bone: VB head_fx"(
    ComponentPose="Transform (Modify) Bone - Bone: VB root_prop"(
     ComponentPose="Transform (Modify) Bone - Bone: hand_r"(ComponentPose=@n3, Alpha=1, Translation="0,0,0", Rotation="0,0,0", Scale=@n2) {scaleMode=BMM_Replace, alphaBoolBlend {blendOption=Linear}, boneToModify {boneName=hand_r}}
     Alpha=1
     Translation="0,0,0"
     Rotation="0,0,0"
     Scale=@n2
    ) {scaleMode=BMM_Replace, alphaBoolBlend {blendOption=Linear}, boneToModify {boneName=VB root_prop}}
    Alpha=1
    Translation="0,0,0"
    Rotation="0,0,0"
    Scale=@n2
   ) {scaleMode=BMM_Replace, alphaBoolBlend {blendOption=Linear}, boneToModify {boneName=VB head_fx}}
   n5 = "Transform (Modify) Bone - Bone: head"(
    ComponentPose="Transform (Modify) Bone - Bone: head"(
     ComponentPose="Transform (Modify) Bone - Bone: attach"(ComponentPose=@n4, Alpha=1, Translation="0,0,0", Rotation="0,0,0", Scale="1,1,1") {scaleMode=BMM_Additive, alphaBoolBlend {blendOption=Linear}, boneToModify {boneName=attach}}
     Alpha=1
     Translation="0,0,0"
     Rotation="0,0,0"
     Scale="1,1,1"
    ) {scaleMode=BMM_Replace, alphaBoolBlend {blendOption=Linear}, boneToModify {boneName=head}}
    Alpha=1
    Translation="0,0,0"
    Rotation="0,0,0"
    Scale=@n2
   ) {scaleMode=BMM_Replace, alphaBoolBlend {blendOption=Linear}, boneToModify {boneName=head}}
   n6 = GetMaxElement!(A:Vector=@n2) -> (ReturnValue:double)
   Output Pose(Result=Component To Local(ComponentPose=Copy Bone Delta - Source Bone: attach - Target Bone: attach!(
      ComponentPose:Component Space Pose Link Structure=@n5
      Alpha:float=1
      bCopyTranslation:bool=true
      bCopyRotation:bool=False
      bCopyScale:bool=true
      TranslationMultiplier:float=@n6
      ScaleMultiplier:float=@n6
     ) -> (Pose:Component Space Pose Link Structure) {
      is=title AnimGraphNode_CopyBoneDelta
      copyMode=Copy
      alphaBoolBlend {blendOption=Linear}
      sourceBone {boneName=attach}
      targetBone {boneName=attach}
     })) {layerGroup=None}
   Save cached pose 'Base Pose'(Pose=Sequence Player 'DualWield_NonTargeted_Pose_CMM'() {sequence=/Game/Animation/Game/MainPlayer/Poses/Medium/Male/DualWield/DualWield_NonTargeted_Pose_CMM}) {cacheName=Base Pose}
  ]}
 EventGraph {nodes [
   Event BlueprintUpdateAnimation()
   > n1 = Cast To FortCharacterVehicle_Ostrich!(execute, Object=TryGetPawnOwner()) -> (then, CastFailed, AsFort Character Vehicle Ostrich:Fort Character Vehicle Ostrich Object Reference) {is=title K2Node_DynamicCast}
   > n2 = Cast To FortOstrichAnimInstance!(execute, Object=GetAnimInstance(self=Get Mesh(self=@n1))) -> (then, CastFailed, AsFort Ostrich Anim Instance:Fort Ostrich Anim Instance Object Reference) {is=title K2Node_DynamicCast}
   > Set Is Moving 2D(Is Moving 2D:bool=Get bIsMoving2D!(self:Fort Ostrich Anim Instance Object Reference=@n2) -> (bIsMoving2D)) -> (Output_Get:bool)
   Set Is Moving 2D(execute=@n2.CastFailed, Is Moving 2D=false)
  ]}
]
variables [
 Is Moving 2D {type=boolean}
 PropScale {type=Vector, default=<3.5 3.5 3.5>}
]
