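# Control Rig for parts of the drivable Mech vehicle, keeping the left and right leg pistons aligned.
# From begin execution, four chained AimBone units point each piston's top and bottom bones at each other, using the thighs for roll and mirrored axes per side, without propagating to children.
# Separate control-driven rotation chains for the left paw thigh and calf, clavicles and upper arms are not connected to execution.
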ControlRigBlueprint /Game/Athena/DrivableVehicles/Mech/Meshes/Mech_Veh_ControRig_Parts
graphs [
 RigVMFunctionLibrary {links [], nodes []}
 RigVMModel {nodes [
   n1 = &RigUnit_BeginExecution()
   /Script/ControlRig.RigUnit_AimBone(
    Bone=dyn_piston_top_l
    DebugSettings=(bEnabled=False,Scale=10,WorldOffset=(Rotation=(X=0,Y=0,Z=0,W=1),Translation=(X=0,Y=0,Z=0),Scale3D=(X=1,Y=1,Z=1)))
    Primary=(Weight=1,Axis=(X=1,Y=0,Z=0),Target=(X=0,Y=0,Z=0),Kind=Location,Space="dyn_piston_bott_l")
    Secondary=(Weight=1,Axis=(X=0,Y=0,Z=1),Target=(X=0,Y=0,Z=1),Kind=Direction,Space="thigh_l")
    Weight=1
    bPropagateToChildren=False
    ExecutePin=@n1.ExecutePin
   ) {is=struct RigVMUnitNode}
   > &RigUnit_AimBone(
    Bone=dyn_piston_bott_l
    DebugSettings=(bEnabled=False,Scale=10,WorldOffset=(Rotation=(X=0,Y=0,Z=0,W=1),Translation=(X=0,Y=0,Z=0),Scale3D=(X=1,Y=1,Z=1)))
    Primary=(Weight=1,Axis=(X=-1,Y=0,Z=0),Target=(X=0,Y=0,Z=0),Kind=Location,Space="dyn_piston_top_l")
    Secondary=(Weight=1,Axis=(X=0,Y=0,Z=1),Target=(X=0,Y=0,Z=1),Kind=Direction,Space="thigh_l")
    Weight=1
    bPropagateToChildren=False
   )
   > &RigUnit_AimBone(
    Bone=dyn_piston_top_r
    DebugSettings=(bEnabled=False,Scale=10,WorldOffset=(Rotation=(X=0,Y=0,Z=0,W=1),Translation=(X=0,Y=0,Z=0),Scale3D=(X=1,Y=1,Z=1)))
    Primary=(Weight=1,Axis=(X=-1,Y=0,Z=0),Target=(X=0,Y=0,Z=0),Kind=Location,Space="dyn_piston_bott_r")
    Secondary=(Weight=1,Axis=(X=0,Y=0,Z=1),Target=(X=0,Y=0,Z=1),Kind=Direction,Space="thigh_r")
    Weight=1
    bPropagateToChildren=False
   )
   > &RigUnit_AimBone(
    Bone=dyn_piston_bott_r
    DebugSettings=(bEnabled=False,Scale=10,WorldOffset=(Rotation=(X=0,Y=0,Z=0,W=1),Translation=(X=0,Y=0,Z=0),Scale3D=(X=1,Y=1,Z=1)))
    Primary=(Weight=1,Axis=(X=1,Y=0,Z=0),Target=(X=0,Y=0,Z=0),Kind=Location,Space="dyn_piston_top_r")
    Secondary=(Weight=1,Axis=(X=0,Y=0,Z=1),Target=(X=0,Y=0,Z=1),Kind=Direction,Space="thigh_r")
    Weight=1
    bPropagateToChildren=False
   )
   Paw Thigh L = /Script/ControlRig.RigUnit_Control(
    Filter=(TranslationFilter=(bX=True,bY=True,bZ=True),RotationFilter=(bX=True,bY=True,bZ=True),ScaleFilter=(bX=True,bY=True,bZ=True))
    InitTransform=(Rotation=(X=0,Y=0,Z=0,W=1),Translation=(X=0,Y=0,Z=0),Scale3D=(X=1,Y=1,Z=1))
    Base=/Script/ControlRig.RigUnit_GetInitialBoneTransform(Bone=paw_thigh_l, Space=GlobalSpace) {is=struct RigVMUnitNode}.Transform
   ) {is=struct RigVMUnitNode}
   /Script/ControlRig.RigUnit_SetBoneRotation(Bone=paw_thigh_l, ExecutePin=(), Space=GlobalSpace, Weight=1, bPropagateToChildren=True, Rotation=@"Paw Thigh L".Result.Rotation) {is=struct RigVMUnitNode}
   Paw Calf L = &RigUnit_Control(
    Filter=(TranslationFilter=(bX=True,bY=True,bZ=True),RotationFilter=(bX=True,bY=True,bZ=True),ScaleFilter=(bX=True,bY=True,bZ=True))
    InitTransform=(Rotation=(X=0,Y=0,Z=0,W=1),Translation=(X=0,Y=0,Z=0),Scale3D=(X=1,Y=1,Z=1))
    Base=&RigUnit_GetInitialBoneTransform(Bone=paw_calf_l, Space=GlobalSpace).Transform
   )
   > &RigUnit_SetBoneRotation(Bone=paw_calf_l, Space=GlobalSpace, Weight=1, bPropagateToChildren=True, Rotation=@"Paw Calf L".Result.Rotation)
   Clav L = &RigUnit_Control(
    Filter=(TranslationFilter=(bX=True,bY=True,bZ=True),RotationFilter=(bX=True,bY=True,bZ=True),ScaleFilter=(bX=True,bY=True,bZ=True))
    InitTransform=(Rotation=(X=0,Y=0,Z=0,W=1),Translation=(X=0,Y=0,Z=0),Scale3D=(X=1,Y=1,Z=1))
    Base=&RigUnit_GetInitialBoneTransform(Bone=clavicle_l, Space=GlobalSpace).Transform
   )
   &RigUnit_SetBoneRotation(Bone=clavicle_l, ExecutePin=(), Space=GlobalSpace, Weight=1, bPropagateToChildren=True, Rotation=@"Clav L".Result.Rotation)
   Upperarm L = &RigUnit_Control(
    Filter=(TranslationFilter=(bX=True,bY=True,bZ=True),RotationFilter=(bX=True,bY=True,bZ=True),ScaleFilter=(bX=True,bY=True,bZ=True))
    InitTransform=(Rotation=(X=0,Y=0,Z=0,W=1),Translation=(X=0,Y=0,Z=0),Scale3D=(X=1,Y=1,Z=1))
    Base=&RigUnit_GetInitialBoneTransform(Bone=upperarm_l, Space=GlobalSpace).Transform
   )
   > &RigUnit_SetBoneRotation(Bone=upperarm_l, Space=GlobalSpace, Weight=1, bPropagateToChildren=True, Rotation=@"Upperarm L".Result.Rotation)
   Clav_R = &RigUnit_Control(
    Filter=(TranslationFilter=(bX=True,bY=True,bZ=True),RotationFilter=(bX=True,bY=True,bZ=True),ScaleFilter=(bX=True,bY=True,bZ=True))
    InitTransform=(Rotation=(X=0,Y=0,Z=0,W=1),Translation=(X=0,Y=0,Z=0),Scale3D=(X=1,Y=1,Z=1))
    Base=&RigUnit_GetInitialBoneTransform(Bone=clavicle_r, Space=GlobalSpace).Transform
   )
   > &RigUnit_SetBoneRotation(Bone=clavicle_r, Space=GlobalSpace, Weight=1, bPropagateToChildren=True, Rotation=@Clav_R.Result.Rotation)
   Upperarm R = &RigUnit_Control(
    Filter=(TranslationFilter=(bX=True,bY=True,bZ=True),RotationFilter=(bX=True,bY=True,bZ=True),ScaleFilter=(bX=True,bY=True,bZ=True))
    InitTransform=(Rotation=(X=0,Y=0,Z=0,W=1),Translation=(X=0,Y=0,Z=0),Scale3D=(X=1,Y=1,Z=1))
    Base=&RigUnit_GetInitialBoneTransform(Bone=upperarm_r, Space=GlobalSpace).Transform
   )
   > &RigUnit_SetBoneRotation(Bone=upperarm_r, Space=GlobalSpace, Weight=1, bPropagateToChildren=True, Rotation=@"Upperarm R".Result.Rotation)
  ]}
]
hierarchy [
 root {loc=<0 0 0>, rot=<0 0 0 1>}
 attach {parent=root, loc=<0 0 0>, rot=<0 0 0 1>}
 pelvis {parent=root, loc=<0 -12.2176 317.0198>, rot=<0.0127 0.707 -0.0127 -0.707>}
 spine_01 {parent=pelvis, loc=<25.2494 -2.3897 0>, rot=<0 0 0.034 -0.9994>}
 spine_02 {parent=spine_01, loc=<27.1289 -0.1309 0>, rot=<0 0 0.0238 -0.9997>}
 spine_03 {parent=spine_02, loc=<22.4851 0 0>, rot=<0 0 0.0149 -0.9999>}
 spine_04 {parent=spine_03, loc=<30.311 0 0>, rot=<0 0 -0.0132 -0.9999>}
 spine_05 {parent=spine_04, loc=<37.4648 0 0>, rot=<0 0 -0.0015 -1>}
 clavicle_l {parent=spine_05, loc=<16.3013 -0.5483 -8.6879>, rot=<0.0465 -0.7289 -0.009 -0.6829>}
 dyn_clav_fin_l {parent=clavicle_l, loc=<78.5323 7.3771 37.8754>, rot=<0.0148 0 0 -0.9999>}
 upperarm_l {parent=clavicle_l, loc=<85.1604 9.6907 13.803>, rot=<0.0231 -0.3234 -0.009 -0.9459>}
 lowerarm_l {parent=upperarm_l, loc=<99.703 0 -0.0001>, rot=<0.0391 -0.0099 -0.2453 -0.9686>}
 hand_l {parent=lowerarm_l, loc=<83.6084 0 0.0001>, rot=<-0.7081 -0.0507 0.0975 -0.6975>}
 index_metacarpal_l {parent=hand_l, loc=<6.1858 0.0684 -11.0735>, rot=<0.0158 -0.0124 -0.0173 -0.9996>}
 index_01_l {parent=index_metacarpal_l, loc=<31.0535 0.1337 -0.1961>, rot=<-0.0044 0.0485 0.0924 -0.9945>}
 index_02_l {parent=index_01_l, loc=<12.5651 -1.1236 0.1691>, rot=<-0.002 0.0027 0.3225 -0.9466>}
 index_03_l {parent=index_02_l, loc=<9.5687 -0.4603 -0.0943>, rot=<-0.002 0.0067 0.2558 -0.9667>}
 middle_metacarpal_l {parent=hand_l, loc=<8.1764 -0.6454 -2.2385>, rot=<0.006 0.0168 -0.0329 -0.9993>}
 middle_01_l {parent=middle_metacarpal_l, loc=<27.4123 -0.9156 0.0484>, rot=<-0.0208 0.0042 0.0858 -0.9961>}
 middle_02_l {parent=middle_01_l, loc=<15.01 0.5243 0.4396>, rot=<-0.0218 0.0258 0.3512 -0.9357>}
 middle_03_l {parent=middle_02_l, loc=<9.7555 0.1998 0.5137>, rot=<0.0082 0.0188 0.1708 -0.9851>}
 pinky_metacarpal_l {parent=hand_l, loc=<6.5902 -3.6486 12.6498>, rot=<-0.0397 0.0814 -0.069 -0.9935>}
 pinky_01_l {parent=pinky_metacarpal_l, loc=<26.8265 -1.8073 -0.4091>, rot=<-0.0654 -0.0035 0.2242 -0.9723>}
 pinky_02_l {parent=pinky_01_l, loc=<10.5775 0.334 -0.3903>, rot=<0.0037 0.0011 0.2854 -0.9584>}
 pinky_03_l {parent=pinky_02_l, loc=<8.2476 0.115 -0.0291>, rot=<-0.0002 0.0098 0.1445 -0.9895>}
 ring_metacarpal_l {parent=hand_l, loc=<8.2578 -1.3561 5.1021>, rot=<-0.0284 0.0722 -0.0458 -0.9959>}
 ring_01_l {parent=ring_metacarpal_l, loc=<26.2974 -2.3182 -0.5409>, rot=<-0.0489 -0.0252 0.1892 -0.9804>}
 ring_02_l {parent=ring_01_l, loc=<13.5328 1.7383 -0.3841>, rot=<0 0 0.2806 -0.9598>}
 ring_03_l {parent=ring_02_l, loc=<9.0352 0.4266 0.0744>, rot=<0 0 0.1189 -0.9929>}
 thumb_01_l {parent=hand_l, loc=<4.9075 -1.1938 -9.1992>, rot=<0.6546 -0.4042 0.0664 -0.6354>}
 thumb_02_l {parent=thumb_01_l, loc=<22.1978 -0.1677 0.1501>, rot=<-0.0037 0.0213 0.298 -0.9543>}
 thumb_03_l {parent=thumb_02_l, loc=<8.3137 1.3775 -0.3393>, rot=<0.0118 -0.013 0.0066 -0.9998>}
 weapon_l {parent=hand_l, loc=<-0.7385 -12.7608 1.4256>, rot=<0 0 0 -1>}
 lowerarm_twist_01_l {parent=lowerarm_l, loc=<66.7953 0 0>, rot=<0 0 0 1>}
 lowerarm_twist_02_l {parent=lowerarm_l, loc=<35.9881 0 0.0001>, rot=<0 0 0 1>}
 weapon_base_r {parent=lowerarm_l, loc=<17.9316 -7.9807 3.264>, rot=<0.6502 -0.6814 0.2525 0.2215>}
 weapon_barrel_guide_r_01 {parent=weapon_base_r, loc=<-21.5909 -178.9143 -10.2543>, rot=<-0.024 -0.0015 0.0088 -0.9997>}
 weapon_barrel_guide_r_02 {parent=weapon_base_r, loc=<-19.7696 -242.2001 -10.442>, rot=<-0.024 -0.0015 0.0088 -0.9997>}
 weapon_barrel_r {parent=weapon_base_r, loc=<0.2542 -93.5458 -8.6383>, rot=<0 0 0 -1>}
 weapon_barrel_support_bott_r {parent=weapon_base_r, loc=<-4.9195 -41.5805 23.9765>, rot=<0 0 0 -1>}
 weapon_barrel_support_r {parent=weapon_base_r, loc=<-28.6054 -94.4612 -9.0839>, rot=<-0.024 -0.0015 0.0088 -0.9997>}
 weapon_barrels_detail_r {parent=weapon_base_r, loc=<0.0967 -123.2523 -29.7369>, rot=<0 0 0 -1>}
 weapon_outer_r_01 {parent=weapon_base_r, loc=<26.5751 33.9965 -4.2893>, rot=<0.0088 -0.0034 0.0171 -0.9998>}
 weapon_outer_r_02 {parent=weapon_outer_r_01, loc=<17.7222 -72.4141 -2.2076>, rot=<-0.0103 0.01 -0.0267 -0.9995>}
 weapon_reload_case_end_r {parent=weapon_base_r, loc=<35.6712 -192.6758 28.2912>, rot=<-0.024 -0.0015 0.0088 -0.9997>}
 weapon_reload_case_r {parent=weapon_base_r, loc=<13.4572 -138.0588 8.7007>, rot=<-0.024 -0.0015 0.0088 -0.9997>}
 weapon_rocket_r_01 {parent=weapon_reload_case_r, loc=<9.1665 5.1737 -5.1782>, rot=<0.7013 -0.6825 0.1561 -0.1343>}
 weapon_rocket_r_02 {parent=weapon_reload_case_r, loc=<17.6544 5.4163 5.5634>, rot=<0.7013 -0.6825 0.1561 -0.1343>}
 weapon_rocket_r_03 {parent=weapon_reload_case_r, loc=<27.0695 5.3762 15.5223>, rot=<0.7013 -0.6825 0.1561 -0.1343>}
 weapon_rocket_r_04 {parent=weapon_reload_case_r, loc=<28.1472 5.9859 27.9022>, rot=<0.7013 -0.6825 0.1561 -0.1343>}
 upperarm_twist_01_l {parent=upperarm_l, loc=<0.0153 0 0.0001>, rot=<0 0 0 1>}
 dyn_shoulder_armor_l {parent=upperarm_twist_01_l, loc=<7.935 0 24.6531>, rot=<0 0 0 1>}
 upperarm_twist_02_l {parent=upperarm_l, loc=<51.6039 0 0>, rot=<0 0 0 1>}
 clavicle_r {parent=spine_05, loc=<16.3016 -0.5483 8.6878>, rot=<0.7289 0.0465 -0.6829 0.009>}
 dyn_clav_fin_r {parent=clavicle_r, loc=<-78.5323 -7.377 -37.8752>, rot=<0.0148 0 0 -0.9999>}
 upperarm_r {parent=clavicle_r, loc=<-85.1602 -9.6907 -13.8031>, rot=<0.0231 -0.3234 -0.009 -0.9459>}
 lowerarm_r {parent=upperarm_r, loc=<-99.7034 0 0>, rot=<0.0391 -0.0099 -0.2453 -0.9686>}
 hand_r {parent=lowerarm_r, loc=<-83.6088 0 0.0001>, rot=<-0.7081 -0.0507 0.0975 -0.6975>}
 index_metacarpal_r {parent=hand_r, loc=<-6.1853 -0.0676 11.0736>, rot=<0.0158 -0.0124 -0.0173 -0.9996>}
 index_01_r {parent=index_metacarpal_r, loc=<-31.0534 -0.1337 0.1961>, rot=<-0.0044 0.0485 0.0924 -0.9945>}
 index_02_r {parent=index_01_r, loc=<-12.5656 1.1235 -0.1693>, rot=<-0.002 0.0027 0.3225 -0.9466>}
 index_03_r {parent=index_02_r, loc=<-9.5687 0.4602 0.0943>, rot=<-0.002 0.0067 0.2558 -0.9667>}
 middle_metacarpal_r {parent=hand_r, loc=<-8.1765 0.6458 2.2385>, rot=<0.006 0.0168 -0.0329 -0.9993>}
 middle_01_r {parent=middle_metacarpal_r, loc=<-27.4117 0.9155 -0.0483>, rot=<-0.0208 0.0042 0.0858 -0.9961>}
 middle_02_r {parent=middle_01_r, loc=<-15.0104 -0.5237 -0.4396>, rot=<-0.0218 0.0258 0.3512 -0.9357>}
 middle_03_r {parent=middle_02_r, loc=<-9.7554 -0.1995 -0.5137>, rot=<0.0082 0.0188 0.1708 -0.9851>}
 pinky_metacarpal_r {parent=hand_r, loc=<-6.5906 3.6486 -12.6499>, rot=<-0.0397 0.0814 -0.069 -0.9935>}
 pinky_01_r {parent=pinky_metacarpal_r, loc=<-26.8257 1.8077 0.4092>, rot=<-0.0654 -0.0035 0.2242 -0.9723>}
 pinky_02_r {parent=pinky_01_r, loc=<-10.5775 -0.3336 0.3903>, rot=<0.0037 0.0011 0.2854 -0.9584>}
 pinky_03_r {parent=pinky_02_r, loc=<-8.247 -0.116 0.0291>, rot=<-0.0002 0.0098 0.1445 -0.9895>}
 ring_metacarpal_r {parent=hand_r, loc=<-8.2574 1.3567 -5.102>, rot=<-0.0284 0.0722 -0.0458 -0.9959>}
 ring_01_r {parent=ring_metacarpal_r, loc=<-26.2981 2.3184 0.5411>, rot=<-0.0489 -0.0252 0.1892 -0.9804>}
 ring_02_r {parent=ring_01_r, loc=<-13.5318 -1.7386 0.384>, rot=<0 0 0.2806 -0.9598>}
 ring_03_r {parent=ring_02_r, loc=<-9.0355 -0.427 -0.0744>, rot=<0 0 0.1189 -0.9929>}
 thumb_01_r {parent=hand_r, loc=<-4.9077 1.1943 9.1991>, rot=<0.6546 -0.4042 0.0664 -0.6354>}
 thumb_02_r {parent=thumb_01_r, loc=<-22.1974 0.1674 -0.1499>, rot=<-0.0037 0.0213 0.298 -0.9543>}
 thumb_03_r {parent=thumb_02_r, loc=<-8.3137 -1.3775 0.3393>, rot=<0.0118 -0.013 0.0066 -0.9998>}
 weapon_r {parent=hand_r, loc=<0.7388 12.7612 -1.4256>, rot=<0 0 0 1>}
 lowerarm_twist_01_r {parent=lowerarm_r, loc=<-66.7956 0 0>, rot=<0 0 0 1>}
 lowerarm_twist_02_r {parent=lowerarm_r, loc=<-35.9883 0 0>, rot=<0 0 0 1>}
 weapon_base_l {parent=lowerarm_r, loc=<-17.932 7.9809 -3.2636>, rot=<0.5856 -0.6609 0.3522 0.3101>}
 weapon_barrel_bott_l {parent=weapon_base_l, loc=<-2.0641 122.0372 -16.9475>, rot=<-0.0045 -0.0279 0.0342 -0.999>}
 weapon_barrel_guide_l_01 {parent=weapon_base_l, loc=<17.5612 233.8459 0.078>, rot=<-0.0041 -0.018 0.0268 -0.9995>}
 weapon_barrel_guide_l_02 {parent=weapon_base_l, loc=<18.8579 255.9191 0.3021>, rot=<-0.0055 -0.0194 0.0217 -0.9996>}
 weapon_barrel_support_l {parent=weapon_base_l, loc=<20.674 134.2488 -0.886>, rot=<-0.0045 -0.0279 0.0342 -0.999>}
 weapon_barrel_top_l {parent=weapon_base_l, loc=<-3.1897 121.7889 14.0296>, rot=<-0.0045 -0.0279 0.0342 -0.999>}
 weapon_barrels_base_l {parent=weapon_base_l, loc=<-4.8862 112.1884 -1.7074>, rot=<-0.0538 0.031 0.0449 -0.9971>}
 weapon_outer_l_01 {parent=weapon_base_l, loc=<-37.8536 37.9537 1.9872>, rot=<0 0.0267 -0.0239 -0.9994>}
 weapon_outer_l_02 {parent=weapon_outer_l_01, loc=<1.5505 89.1677 -0.6181>, rot=<-0.0061 -0.0076 0.0473 -0.9988>}
 weapon_shotgunshell_01 {parent=weapon_outer_l_02, loc=<-7.9245 15.4323 7.9959>, rot=<0.0047 0 0.0096 -0.9999>}
 weapon_shotgunshell_02 {parent=weapon_outer_l_02, loc=<-0.1416 15.3637 12.5141>, rot=<0.0047 0 0.0096 -0.9999>}
 weapon_shotgunshell_03 {parent=weapon_outer_l_02, loc=<8.3941 15.2661 14.8933>, rot=<0.0047 0 0.0096 -0.9999>}
 weapon_shotgunshell_04 {parent=weapon_outer_l_02, loc=<16.9071 15.1426 14.2476>, rot=<0.0047 0 0.0096 -0.9999>}
 weapon_tube_l {parent=weapon_base_l, loc=<18.6633 38.4109 -24.1844>, rot=<-0.0016 0.0086 0.031 -0.9995>}
 upperarm_twist_01_r {parent=upperarm_r, loc=<-0.0154 0 0>, rot=<0 0 0 -1>}
 dyn_shoulder_armor_r {parent=upperarm_twist_01_r, loc=<-7.9349 0 -24.6531>, rot=<0 0 0 -1>}
 upperarm_twist_02_r {parent=upperarm_r, loc=<-51.6041 0 0.0001>, rot=<0 0 0 -1>}
 dyn_eye_l {parent=spine_05, loc=<-20.8059 69.959 -74.174>, rot=<0.0105 -0.707 -0.0105 -0.707>}
 dyn_eye_r {parent=spine_05, loc=<-20.8057 69.959 74.174>, rot=<0.707 0.0105 -0.707 0.0105>}
 dyn_gunner_joystick_l {parent=spine_05, loc=<63.8345 39.9566 -24.0422>, rot=<0.0105 -0.707 -0.0105 -0.707>}
 dyn_gunner_joystick_r {parent=spine_05, loc=<63.8345 39.9567 24.0422>, rot=<0.707 0.0105 -0.707 0.0105>}
 gunner_pivot {parent=spine_05, loc=<-34.9092 7.3702 0>, rot=<0.0105 -0.707 -0.0105 -0.707>}
 neck_01 {parent=spine_05, loc=<42.3252 0.5719 0>, rot=<0 0 -0.2003 -0.9797>}
 neck_02 {parent=neck_01, loc=<17.069 0 0>, rot=<0 0 0.0504 -0.9987>}
 head {parent=neck_02, loc=<20.1693 0 0>, rot=<0 0 0.115 -0.9934>}
 thigh_l {parent=pelvis, loc=<4.123 -45.4953 -56.607>, rot=<0.0012 0.0572 0.0155 -0.9982>}
 calf_l {parent=thigh_l, loc=<-128.541 0 0>, rot=<0.0006 -0.0161 -0.0391 -0.9991>}
 calf_twist_01_l {parent=calf_l, loc=<-117.8907 -0.3184 0.1307>, rot=<0.0237 0.0095 -0.0035 -0.9997>}
 calf_twist_02_l {parent=calf_l, loc=<-61.8958 -0.3183 0.1307>, rot=<0.0237 0.0095 -0.0035 -0.9997>}
 foot_l {parent=calf_l, loc=<-140.1039 0.3547 1.4128>, rot=<0.0418 -0.0455 0.0343 -0.9975>}
 ball_l {parent=foot_l, loc=<-20.2734 45.5057 2.24>, rot=<0.0001 -0.0001 -0.7063 -0.7079>}
 thigh_twist_01_l {parent=thigh_l, loc=<-71.19 0 0>, rot=<0.0013 0 0 -1>}
 thigh_r {parent=pelvis, loc=<4.1236 -45.4953 56.607>, rot=<0.0572 -0.0012 0.9982 0.0155>}
 calf_r {parent=thigh_r, loc=<128.5417 0 0>, rot=<0.0006 -0.0161 -0.0391 -0.9991>}
 calf_twist_01_r {parent=calf_r, loc=<117.891 0.3184 -0.1307>, rot=<0.0237 0.0095 -0.0035 -0.9997>}
 calf_twist_02_r {parent=calf_r, loc=<61.8961 0.3183 -0.1307>, rot=<0.0237 0.0095 -0.0035 -0.9997>}
 foot_r {parent=calf_r, loc=<140.1043 -0.3546 -1.4129>, rot=<0.0418 -0.0455 0.0343 -0.9975>}
 ball_r {parent=foot_r, loc=<20.2734 -45.5058 -2.24>, rot=<0.0001 -0.0001 -0.7063 -0.7079>}
 thigh_twist_01_r {parent=thigh_r, loc=<71.19 0 0>, rot=<0.0013 0 0 -1>}
 paw_root {parent=pelvis, loc=<-316.3786 23.5661 0>, rot=<-0.0127 -0.707 0.0127 -0.707>}
 paw_thigh_l {parent=paw_root, loc=<56.6069 -57.536 322.7698>, rot=<-0.211 -0.6391 -0.2047 -0.7107>}
 paw_calf_l {parent=paw_thigh_l, loc=<128.7803 0 0>, rot=<0 0 0.7255 -0.6882>}
 paw_foot_l {parent=paw_calf_l, loc=<147.2977 -0.0001 0>, rot=<0 0 -0.6765 -0.7364>}
 paw_ball_l {parent=paw_foot_l, loc=<95.1785 0 0>, rot=<-0.034 -0.0325 -0.3331 -0.9417>}
 paw_toe_l {parent=paw_ball_l, loc=<49.8678 0 0>, rot=<-0.0006 0.0011 0.0157 -0.9999>}
 paw_toetip_l {parent=paw_toe_l, loc=<73.1574 0.0001 0>, rot=<0 0 -0.2282 -0.9736>}
 dyn_piston_bott_l {parent=paw_calf_l, loc=<122.441 -1.5595 0.9205>, rot=<-0.6604 -0.1625 -0.1304 -0.7215>}
 dyn_piston_top_l {parent=paw_thigh_l, loc=<79.0346 -12.1892 8.2613>, rot=<-0.5723 0.3673 0.4337 -0.5911>}
 paw_thigh_r {parent=paw_root, loc=<-56.607 -57.536 322.7698>, rot=<0.7107 -0.2047 0.6391 -0.211>}
 paw_calf_r {parent=paw_thigh_r, loc=<-128.7803 0 0>, rot=<0 0 0.7255 -0.6882>}
 paw_foot_r {parent=paw_calf_r, loc=<-147.2977 -0.0001 0>, rot=<0 0 -0.6765 -0.7364>}
 paw_ball_r {parent=paw_foot_r, loc=<-95.1785 0 -0.0001>, rot=<-0.034 -0.0325 -0.3331 -0.9417>}
 paw_toe_r {parent=paw_ball_r, loc=<-49.8678 0 0>, rot=<-0.0006 0.0011 0.0157 -0.9999>}
 paw_toetip_r {parent=paw_toe_r, loc=<-73.1574 0 -0.0001>, rot=<0 0 -0.2282 -0.9736>}
 dyn_piston_bott_r {parent=paw_calf_r, loc=<-122.441 1.5593 -0.9205>, rot=<-0.6604 -0.1625 -0.1304 -0.7215>}
 dyn_piston_top_r {parent=paw_thigh_r, loc=<-79.0347 12.1892 -8.2613>, rot=<-0.5723 0.3673 0.4337 -0.5911>}
 ik_foot_root {parent=root, loc=<0 0 0>, rot=<0 0 0 1>}
 ik_foot_l {parent=ik_foot_root, loc=<81.3312 -69.4114 55.7738>, rot=<-0.0357 -0.7045 -0.0284 0.7082>}
 ik_foot_r {parent=ik_foot_root, loc=<-81.3313 -69.4115 55.7733>, rot=<-0.7082 -0.0284 0.7045 -0.0357>}
 VB_ik_foot_root_paw_ball_l {parent=ik_foot_root, loc=<81.3667 -69.2965 55.8616>, rot=<0.3937 0.5874 0.3704 0.6023>}
 VB_ik_foot_root_paw_ball_r {parent=ik_foot_root, loc=<-81.3668 -69.2965 55.8617>, rot=<-0.6023 0.3704 -0.5874 0.3937>}
 ik_hand_root {parent=root, loc=<0 0 0>, rot=<0 0 0 1>}
 ik_hand_gun {parent=ik_hand_root, loc=<-222.8553 13.9866 364.7599>, rot=<-0.6086 -0.1793 -0.4614 0.6201>}
 ik_hand_l {parent=ik_hand_gun, loc=<227.2657 352.3337 151.2248>, rot=<-0.5099 -0.7905 -0.3393 0>}
 ik_hand_r {parent=ik_hand_gun, loc=<0 0 0>, rot=<0 0 0 -1>}
]
