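# Physics asset for secondary motion on the drivable Mech vehicle's antenna, torso hose, shoulder fins and exhaust flaps.
# Capsule bodies form short constraint chains rooted at spine_05: a two-segment antenna, a tightly limited hose, and left and right fin-to-flap pairs.
# The fins rotate freely on tiny placeholder-sized capsules, while each flap hinges on a single limited swing axis.
# The antenna locks twist, with its tip nearly rigid relative to its base.
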PhysicsAsset /Game/Athena/DrivableVehicles/Mech/Meshes/Mech_Veh_Dyn
bodies {
 dyn_antenna_01 [{shapeType=Capsule, length=30, radius=2, center=<22 0 0>, rotation=<rot 90 0 0>}]
 dyn_antenna_02 [{shapeType=Capsule, length=30, radius=2, center=<22 0 0>, rotation=<rot 90 0 0>}]
 dyn_clav_fin_l [{shapeType=Capsule, length=1, radius=1}]
 dyn_clav_fin_r [{shapeType=Capsule, length=1, radius=1}]
 dyn_exhaust_flap_l [{shapeType=Capsule, length=40, radius=12, shapeName=dyn_exhaust_flap_l_capsule, center=<0 13 0>, rotation=<rot 0 -90 0>}]
 dyn_exhaust_flap_r [{shapeType=Capsule, length=40, radius=12, shapeName=dyn_exhaust_flap_r_capsule, center=<0 -13 0>, rotation=<rot 0 -90 0>}]
 dyn_torso_hose_01 [{shapeType=Capsule, length=24.6691, radius=8.4888, center=<42 -0.4917 0>, rotation=<rot 0 179.328 -84.7471>}]
 spine_05 [{shapeType=Capsule, length=1, radius=1}]
}
constraints [
 {bone1Name=dyn_antenna_02, bone2Name=dyn_antenna_01, swing1LimitDegrees=1, swing2LimitDegrees=1, twistLimitDegrees=45, twistMotion=Locked}
 {bone1Name=dyn_antenna_01, bone2Name=spine_05, swing1LimitDegrees=45, swing2LimitDegrees=45, twistLimitDegrees=45, twistMotion=Locked}
 {bone1Name=dyn_torso_hose_01, bone2Name=spine_05, swing1LimitDegrees=5, swing2LimitDegrees=20, twistLimitDegrees=10}
 {
  bone1Name=dyn_clav_fin_l
  bone2Name=spine_05
  swing1LimitDegrees=45
  swing1Motion=Free
  swing2LimitDegrees=45
  swing2Motion=Free
  twistLimitDegrees=45
  twistMotion=Free
 }
 {
  bone1Name=dyn_exhaust_flap_l
  bone2Name=dyn_clav_fin_l
  swing1LimitDegrees=45
  swing2LimitDegrees=45
  swing2Motion=Locked
  twistLimitDegrees=45
  twistMotion=Locked
 }
 {
  bone1Name=dyn_clav_fin_r
  bone2Name=spine_05
  swing1LimitDegrees=45
  swing1Motion=Free
  swing2LimitDegrees=45
  swing2Motion=Free
  twistLimitDegrees=45
  twistMotion=Free
 }
 {
  bone1Name=dyn_exhaust_flap_r
  bone2Name=dyn_clav_fin_r
  swing1LimitDegrees=45
  swing2LimitDegrees=45
  swing2Motion=Locked
  twistLimitDegrees=45
  twistMotion=Locked
 }
]
