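# Physics asset for the ammo belt on the Mech vehicle's left weapon, giving its bones collision bodies and joints for dynamic motion.
# Ten capsule shell bodies form a serial chain from a small bullet base capsule, with tight twist, wide second swing and first swing loosening near the base's far end.
# Capsules and boxes cover the weapon base, outer section and barrel housing, linked to each other and the bullet base by free-rotating constraints.
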PhysicsAsset /Game/Athena/DrivableVehicles/Mech/Meshes/Mech_Veh_Dyn_Belt
bodies {
 dyn_bullet_base [{shapeType=Capsule, length=0.505, radius=0.505, rotation=<rot 0 -90 0>}]
 dyn_shell_01 [{shapeType=Capsule, length=17.4006, radius=6.2991, rotation=<rot 0 0 90>}]
 dyn_shell_010 [{shapeType=Capsule, length=17.4632, radius=6.4001, rotation=<rot 0 0 90>}]
 dyn_shell_02 [{shapeType=Capsule, length=17.3851, radius=5.3656, rotation=<rot 0 0 90>}]
 dyn_shell_03 [{shapeType=Capsule, length=17.3819, radius=5.3323, rotation=<rot 0 0 90>}]
 dyn_shell_04 [{shapeType=Capsule, length=17.3787, radius=5.3317, rotation=<rot 0 0 90>}]
 dyn_shell_05 [{shapeType=Capsule, length=17.3754, radius=5.3311, rotation=<rot 0 0 90>}]
 dyn_shell_06 [{shapeType=Capsule, length=17.3722, radius=5.3305, rotation=<rot 0 0 90>}]
 dyn_shell_07 [{shapeType=Capsule, length=17.369, radius=5.3432, rotation=<rot 0 0 90>}]
 dyn_shell_08 [{shapeType=Capsule, length=17.3657, radius=5.3852, rotation=<rot 0 0 90>}]
 dyn_shell_09 [{shapeType=Capsule, length=17.3632, radius=5.4273, rotation=<rot 0 0 90>}]
 weapon_barrels_base_l [
  {shapeType=Box, extentX=46.625, extentY=160.625, extentZ=52.9375, center=<9.5027 7.1021 -27.9208>, rotation=<rot 0 2.2359 0>}
  {shapeType=Box, extentX=53.5, extentY=33.25, extentZ=13.4375, center=<-29.7895 36.2815 -29.3836>, rotation=<rot 0 2.3906 0>}
 ]
 weapon_base_l [{shapeType=Capsule, length=27.5645, radius=12.8907, center=<0 58.1196 0.0012>, rotation=<rot 0 0 90>}]
 weapon_outer_l_01 [{shapeType=Capsule, length=176, radius=18, shapeName=weapon_outer_l_01_capsule, center=<10 73 0>, rotation=<rot 0 0 90>}]
}
constraints [
 {bone1Name=dyn_shell_01, bone2Name=dyn_bullet_base, swing1LimitDegrees=5, swing2LimitDegrees=40, twistLimitDegrees=5}
 {bone1Name=dyn_shell_02, bone2Name=dyn_shell_01, swing1LimitDegrees=10, swing2LimitDegrees=60, twistLimitDegrees=5}
 {bone1Name=dyn_shell_03, bone2Name=dyn_shell_02, swing1LimitDegrees=15, swing2LimitDegrees=60, twistLimitDegrees=5}
 {bone1Name=dyn_shell_04, bone2Name=dyn_shell_03, swing1LimitDegrees=20, swing2LimitDegrees=60, twistLimitDegrees=5}
 {bone1Name=dyn_shell_05, bone2Name=dyn_shell_04, swing1LimitDegrees=20, swing2LimitDegrees=60, twistLimitDegrees=5}
 {bone1Name=dyn_shell_06, bone2Name=dyn_shell_05, swing1LimitDegrees=20, swing2LimitDegrees=60, twistLimitDegrees=5}
 {bone1Name=dyn_shell_07, bone2Name=dyn_shell_06, swing1LimitDegrees=20, swing2LimitDegrees=60, twistLimitDegrees=5}
 {bone1Name=dyn_shell_08, bone2Name=dyn_shell_07, swing1LimitDegrees=20, swing2LimitDegrees=60, twistLimitDegrees=5}
 {bone1Name=dyn_shell_09, bone2Name=dyn_shell_08, swing1LimitDegrees=20, swing2LimitDegrees=60, twistLimitDegrees=5}
 {bone1Name=dyn_shell_010, bone2Name=dyn_shell_09, swing1LimitDegrees=20, swing2LimitDegrees=60, twistLimitDegrees=5}
 {
  bone1Name=weapon_outer_l_01
  bone2Name=weapon_base_l
  swing1LimitDegrees=45
  swing1Motion=Free
  swing2LimitDegrees=45
  swing2Motion=Free
  twistLimitDegrees=45
  twistMotion=Free
 }
 {
  bone1Name=weapon_barrels_base_l
  bone2Name=weapon_base_l
  swing1LimitDegrees=45
  swing1Motion=Free
  swing2LimitDegrees=45
  swing2Motion=Free
  twistLimitDegrees=45
  twistMotion=Free
 }
 {
  bone1Name=dyn_bullet_base
  bone2Name=weapon_outer_l_01
  swing1LimitDegrees=45
  swing1Motion=Free
  swing2LimitDegrees=45
  swing2Motion=Free
  twistLimitDegrees=45
  twistMotion=Free
 }
]
