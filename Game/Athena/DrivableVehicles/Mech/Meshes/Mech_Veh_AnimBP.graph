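# Animation Blueprint for the drivable Mech (OstrichVehicle), covering driver locomotion, boot-up and shutdown, jump/dash and gunner aiming.
# Cached poses layer state machines, arm slots, aim, recoil and shotgun shells.
# Orientation, speed and slope warping plus Leg IK adjust the legs.
# Final blends switch in emote-mesh copying and cosmetic preview, then Control Rigs, trail and RigidBody nodes drive parts, secondary motion and ammo feed.
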AnimBlueprint /Game/Athena/DrivableVehicles/Mech/Meshes/Mech_Veh_AnimBP
target_skeleton=&/Mech_Veh_Skeleton
graphs [
 AnimGraph {nodes [
   n1 = Control Rig(Source=Sequence Player 'OstrichVehicle_Idle_CosmeticPreivew_Pose'() {sequence=/Game/Animation/Game/Vehicles/OstrichVehicle/OstrichVehicle_Idle_CosmeticPreivew_Pose}) {
    alphaBoolBlend {blendOption=Linear}
    controlRigAssetReference {blueprintRigClass=&/Mech_Veh_ControRig_Parts.Mech_Veh_ControRig_Parts_C}
   }
   n2 = Get StatusStartPositionAlphaValue() -> (StatusStartPositionAlphaValue:float)
   AnimGraphNode_StateMachine_4 = RocketOnCoolDown() {is=title AnimGraphNode_StateMachine}
   n3 = Use cached pose 'PostAdditivePose'()
   n4 = "Transform (Modify) Bone - Bone: weapon_shotgunshell_01"!(
    ComponentPose=Local To Component(LocalPose=Layered blend per bone(BasePose=@n3, BlendPoses_0=@AnimGraphNode_StateMachine_4) {layerSetup [{branchFilters [{boneName=weapon_outer_r_01}]}]})
    Alpha=Get ShellOneAlphaValue() -> (ShellOneAlphaValue:float)
    Scale="0.01,0.01,0.01"
   ) -> (Pose) {
    scaleMode=BMM_Replace
    scaleSpace=BCS_BoneSpace
    alphaBoolBlend {blendOption=Linear}
    boneToModify {boneName=weapon_shotgunshell_01}
    scale=<0.01 0.01 0.01>
   }
   n5 = "Transform (Modify) Bone - Bone: weapon_shotgunshell_02"!(ComponentPose=@n4, Alpha=Get ShellTwoAlphaValue() -> (ShellTwoAlphaValue:float), Scale="0.01,0.01,0.01") -> (Pose) {
    scaleMode=BMM_Replace
    scaleSpace=BCS_BoneSpace
    alphaBoolBlend {blendOption=Linear}
    boneToModify {boneName=weapon_shotgunshell_02}
    scale=<0.01 0.01 0.01>
   }
   n6 = "Transform (Modify) Bone - Bone: weapon_shotgunshell_03"!(ComponentPose=@n5, Alpha=Get ShellThreeAlphaValue() -> (ShellThreeAlphaValue:float), Scale="0.01,0.01,0.01") -> (Pose) {
    scaleMode=BMM_Replace
    scaleSpace=BCS_BoneSpace
    alphaBoolBlend {blendOption=Linear}
    boneToModify {boneName=weapon_shotgunshell_03}
    scale=<0.01 0.01 0.01>
   }
   n7 = "Transform (Modify) Bone - Bone: weapon_shotgunshell_04"!(ComponentPose=@n6, Alpha=Get ShellFourAlphaValue() -> (ShellFourAlphaValue:float), Scale="0.01,0.01,0.01") -> (Pose) {
    scaleMode=BMM_Replace
    scaleSpace=BCS_BoneSpace
    alphaBoolBlend {blendOption=Linear}
    boneToModify {boneName=weapon_shotgunshell_04}
    scale=<0.01 0.01 0.01>
   }
   n8 = Copy Bone - Source Bone: VB ik_foot_root_paw_ball_l - Target Bone: ik_foot_l!(
    ComponentPose:Component Space Pose Link Structure=@n7
    Alpha:float=1
    bCopyTranslation:bool=true
    bCopyRotation:bool=true
    bCopyScale:bool=False
   ) -> (Pose:Component Space Pose Link Structure) {
    is=title AnimGraphNode_CopyBone
    alphaBoolBlend {blendOption=Linear}
    sourceBone {boneName=VB ik_foot_root_paw_ball_l}
    targetBone {boneName=ik_foot_l}
   }
   n9 = Copy Bone - Source Bone: VB ik_foot_root_paw_ball_r - Target Bone: ik_foot_r!(
    ComponentPose:Component Space Pose Link Structure=@n8
    Alpha:float=1
    bCopyTranslation:bool=true
    bCopyRotation:bool=true
    bCopyScale:bool=False
   ) -> (Pose:Component Space Pose Link Structure) {
    is=title AnimGraphNode_CopyBone
    alphaBoolBlend {blendOption=Linear}
    sourceBone {boneName=VB ik_foot_root_paw_ball_r}
    targetBone {boneName=ik_foot_r}
   }
   n10 = Biped Slope Warping!(ComponentPose:Component Space Pose Link Structure=@n9, Alpha:float=Get SlopeWarpingAlpha() -> (SlopeWarpingAlpha:float)) -> (Pose:Component Space Pose Link Structure) {
    is=title FortAnimGraphNode_SlopeWarping
    lODThreshold=3
    maxStepHeight=150
    alphaBoolBlend {blendOption=Linear}
    feetDefinitions [
     {numBonesInLimb=2, fKFootBone {boneName=foot_l}, iKFootBone {boneName=ik_foot_l}}
     {numBonesInLimb=2, fKFootBone {boneName=foot_r}, iKFootBone {boneName=ik_foot_r}}
    ]
    iKFootRootBone {boneName=ik_foot_root}
    pelvisBone {boneName=pelvis}
   }
   n11 = Copy Bone - Source Bone: ik_foot_l - Target Bone: VB ik_foot_root_paw_ball_l!(
    ComponentPose:Component Space Pose Link Structure=@n10
    Alpha:float=1
    bCopyTranslation:bool=true
    bCopyRotation:bool=true
    bCopyScale:bool=False
   ) -> (Pose:Component Space Pose Link Structure) {
    is=title AnimGraphNode_CopyBone
    alphaBoolBlend {blendOption=Linear}
    sourceBone {boneName=ik_foot_l}
    targetBone {boneName=VB ik_foot_root_paw_ball_l}
   }
   n12 = Copy Bone - Source Bone: ik_foot_r - Target Bone: VB ik_foot_root_paw_ball_r!(
    ComponentPose:Component Space Pose Link Structure=@n11
    Alpha:float=1
    bCopyTranslation:bool=true
    bCopyRotation:bool=true
    bCopyScale:bool=False
   ) -> (Pose:Component Space Pose Link Structure) {
    is=title AnimGraphNode_CopyBone
    alphaBoolBlend {blendOption=Linear}
    sourceBone {boneName=ik_foot_r}
    targetBone {boneName=VB ik_foot_root_paw_ball_r}
   }
   n13 = Leg IK!(ComponentPose:Component Space Pose Link Structure=@n12, Alpha:float=1) -> (Pose:Component Space Pose Link Structure) {
    is=title AnimGraphNode_LegIK
    lODThreshold=3
    alphaBoolBlend {blendOption=Linear}
    legsDefinition [
     {
      bEnableKneeTwistCorrection=false
      bEnableRotationLimit=false
      footBoneForwardAxis=Y
      hingeRotationAxis=Z
      minRotationAngle=15
      numBonesInLimb=3
      twistOffsetCurveName=None
      fKFootBone {boneName=paw_ball_l}
      iKFootBone {boneName=VB ik_foot_root_paw_ball_l}
     }
     {
      bEnableKneeTwistCorrection=false
      bEnableRotationLimit=false
      footBoneForwardAxis=Y
      hingeRotationAxis=Z
      minRotationAngle=15
      numBonesInLimb=3
      twistOffsetCurveName=None
      fKFootBone {boneName=paw_ball_r}
      iKFootBone {boneName=VB ik_foot_root_paw_ball_r}
     }
    ]
   }
   n14 = Component To Local(ComponentPose=@n13)
   n15 = Control Rig(Source=Copy Pose From Mesh(SourceMeshComponent=Get EmoteMeshToCopy() -> (EmoteMeshToCopy:Skeletal Mesh Component Object Reference))) {
    alphaBoolBlend {blendOption=Linear}
    controlRigAssetReference {blueprintRigClass=&/Mech_Veh_ConrolRig.Mech_Veh_ConrolRig_C}
   }
   n16 = "Transform (Modify) Bone - Bone: hand_l"(ComponentPose=Local To Component(LocalPose=@n15), Alpha=1, Translation="0,0,0", Rotation="0,0,0", Scale="0,0,0") {scaleMode=BMM_Replace, alphaBoolBlend {blendOption=Linear}, boneToModify {boneName=hand_l}, scale=<0 0 0>}
   n17 = "Transform (Modify) Bone - Bone: hand_r"(ComponentPose=@n16, Alpha=1, Translation="0,0,0", Rotation="0,0,0", Scale="0,0,0") {scaleMode=BMM_Replace, alphaBoolBlend {blendOption=Linear}, boneToModify {boneName=hand_r}, scale=<0 0 0>}
   n18 = Layered blend per bone(
    BasePose=Component To Local(ComponentPose=@n17)
    BlendPoses_0=Sequence Player 'OstrichVehicle_Idle_NoPassenger_Pose'() {sequence=/Game/Animation/Game/Vehicles/OstrichVehicle/OstrichVehicle_Idle_NoPassenger_Pose}
    BlendWeights_0=Get JoyStickAlphaValue() -> (JoyStickAlphaValue:float)
   ) {layerSetup [{branchFilters [{boneName=dyn_gunner_joystick_l}, {boneName=dyn_gunner_joystick_r}]}]}
   n19 = Layered blend per bone(
    BasePose=@n18
    BlendPoses_0=Sequence Player 'OstrichVehicle_Idle_Pose'() {sequence=/Game/Animation/Game/Vehicles/OstrichVehicle/OstrichVehicle_Idle_Pose}
    BlendWeights_0=Get NotJustGotDriverFloat() -> (NotJustGotDriverFloat:float)
   ) {layerSetup [{branchFilters [{boneName=weapon_base_l}, {boneName=weapon_base_r}]}]}
   n20 = Get bIsEmoting() -> (bIsEmoting:bool)
   n21 = Control Rig(Source=Slot 'MechEmote'(Source=Blend Poses by bool(BlendPose_0=@n19, BlendPose_1=@n14, BlendTime_0=0.2, BlendTime_1=0.3, bActiveValue=@n20) {blendType=Linear}) {slotName=MechEmote}) {
    alphaBoolBlend {blendOption=Linear}
    controlRigAssetReference {blueprintRigClass=&Mech_Veh_ControRig_Parts_C}
   }
   n22 = Layered blend per bone(
    BasePose=@n21
    BlendPoses_0=Blend Poses by bool(
     BlendPose_0=Sequence Player 'OstrichVehicle_Driver_BootUp'() {bLoopAnimation=false, sequence=/Game/Animation/Game/Vehicles/OstrichVehicle/OstrichVehicle_Driver_BootUp}
     BlendPose_1=Sequence Player 'OstrichVehicle_Driver_ShutDown'!(StartPosition:float=@n2) -> (Pose) {bLoopAnimation=false, sequence=/Game/Animation/Game/Vehicles/OstrichVehicle/OstrichVehicle_Driver_ShutDown}
     bActiveValue=Get bHasAnyPlayer() -> (bHasAnyPlayer:bool)
    ) {blendType=Linear, childUpateMode=ResetChildOnActivate}
   ) {layerSetup [{branchFilters [{boneName=dyn_eye_l}, {boneName=dyn_eye_r}]}]}
   n23 = Get bIsCosmeticPreview() -> (bIsCosmeticPreview:bool)
   n24 = Get RigidBodyAlpha() -> (RigidBodyAlpha:float)
   n25 = "Transform (Modify) Bone - Bone: dyn_shell_01"(
    ComponentPose=Local To Component(LocalPose=Blend Poses by bool(BlendPose_0=@n1, BlendPose_1=@n22, bActiveValue=@n23) {blendType=Linear})
    Alpha=@n24
    Translation="0,0,0"
    Rotation="-38.835,0.000,0.000"
    Scale="1,1,1"
   ) {
    rotationMode=BMM_Additive
    rotationSpace=BCS_BoneSpace
    alphaBoolBlend {blendOption=Linear}
    boneToModify {boneName=dyn_shell_01}
    rotation {pitch=-38.835}
   }
   n26 = "Transform (Modify) Bone - Bone: dyn_shell_02"(ComponentPose=@n25, Alpha=@n24, Translation="0,0,0", Rotation="-53.062,0.000,0.000", Scale="1,1,1") {
    rotationMode=BMM_Additive
    rotationSpace=BCS_BoneSpace
    alphaBoolBlend {blendOption=Linear}
    boneToModify {boneName=dyn_shell_02}
    rotation {pitch=-53.062}
   }
   n27 = Trail controller - Bone: dyn_shell_010(ComponentPose=@n26, Alpha=@n24) {
    chainLength=10
    fakeVelocity {z=50}
    rotationLimits [
     {limitMax=<180 180 180>}
     {limitMax=<180 180 180>}
     {limitMax=<180 180 180>}
     {limitMax=<180 180 180>}
     {limitMax=<180 180 180>}
     {limitMax=<180 180 180>}
     {limitMax=<180 180 180>}
     {limitMax=<180 180 180>}
     {limitMax=<180 180 180>}
     {limitMax=<180 180 180>}
    ]
    rotationOffsets [<0 0 0>, <0 0 0>, <0 0 0>, <0 0 0>, <0 0 0>, <0 0 0>, <0 0 0>, <0 0 0>, <0 0 0>, <0 0 0>]
    trailBone {boneName=dyn_shell_010}
    trailRelaxationSpeed {editorCurveData {keys [<at 0 10 RCIM_Linear>,<at 1 10 RCIM_Linear>]}}
   }
   n28 = RigidBody(ComponentPose=@n27) {
    bClampLinearTranslationLimitToRefPose=true
    bOverrideWorldGravity=true
    overridePhysicsAsset=&/Mech_Veh_Dyn
    simulationSpace=BaseBoneSpace
    alphaBoolBlend {blendOption=Linear}
    baseBoneRef {boneName=spine_03}
    componentAppliedLinearAccClamp=<3000 3000 3000>
    componentLinearAccScale=<0.1 0.1 0.1>
    componentLinearVelScale=<0.1 0.1 0.1>
    overrideWorldGravity {z=-980}
   }
   n29 = RigidBody(ComponentPose=@n28) {
    bClampLinearTranslationLimitToRefPose=true
    bOverrideWorldGravity=true
    overridePhysicsAsset=&/Mech_Veh_Dyn_Belt
    simulationSpace=BaseBoneSpace
    alphaBoolBlend {blendOption=Linear}
    baseBoneRef {boneName=hand_r}
    componentAppliedLinearAccClamp=<3000 3000 3000>
    componentLinearAccScale=<0.2 0.2 0.2>
    componentLinearVelScale=<0.2 0.2 0.2>
    overrideWorldGravity {z=-1500}
   }
   n30 = Control Rig!(Source=Component To Local(ComponentPose=@n29), Alpha, ChainSlide:double=Get ShotgunAmmo_SlideValue() -> (ShotgunAmmo_SlideValue:float)) -> (Pose) {
    alphaBoolBlend {blendOption=Linear}
    controlRigAssetReference {blueprintRigClass=&/Mech_Veh_AmmoFeed_ControlRig.Mech_Veh_AmmoFeed_ControlRig_C}
   }
   Output Pose(Result=@n30) {layerGroup=None}
   n31 = "Sequence Player (additive) 'OstrichVehicle_Idle_GunnerShoulder_Pose'"() {sequence=/Game/Animation/Game/Vehicles/OstrichVehicle/OstrichVehicle_Idle_GunnerShoulder_Pose}
   AnimGraphNode_StateMachine_0 = OstrichVehicle() {is=title AnimGraphNode_StateMachine}
   n32 = Get bHasGunner() -> (bHasGunner:bool)
   n33 = Get TurnInPlaceInput() -> (TurnInPlaceInput:Fort Anim Input Turn In Place Structure)
   Save cached pose 'LocomotionPose'(Pose=Rotate Root Bone!(
     BasePose:Pose Link Structure=Apply Additive!(Base=@AnimGraphNode_StateMachine_0, Additive=@n31, bAlphaBoolEnabled:bool=@n32) -> (Pose) {alphaInputType=Bool, alphaBoolBlend {blendOption=Linear}}
     Yaw:float=Break Fort Anim Input Turn in Place!(FortAnimInput_TurnInPlace:"Fort Anim Input Turn In Place Structure (by ref)"=@n33) -> (
      RootYawOffset:float=~
      TurnAnimPlayRate:float=~
      bWantsToTurn:bool=~
      bExitTurn:bool=~
      bUseTurnA:bool=~
      TurnACurveAlpha:float=~
      TurnBCurveAlpha:float=~
      bWantsToTurnInVehicle:bool=~
      bWantsToTurnAgain:bool=~
      bTurningLeft:bool=~
      ShufflePlayRate:float=~
      CrouchShufflePlayRate:float=~
      CrouchShuffleStartPosition:float=~
      CrouchShuffleSpeedScaling:float=~
      bTransition_LowerBody_Shuffle:bool=~
      bTransition_LowerBody_ShuffleADS:bool=~
      bPlayOutro:bool=~
     ) {is=title K2Node_BreakStruct}
    ) -> (Pose:Pose Link Structure) {is=title AnimGraphNode_RotateRootBone}) {cacheName=LocomotionPose}
   n34 = Get LocalAimYaw() -> (LocalAimYaw:float)
   n35 = Blendspace Player 'OstrichVehicle_Jog_Steer_BS'!(X=@n34, PlayRate:float=1) -> (Pose) {blendSpace=/Game/Animation/Game/Vehicles/OstrichVehicle/OstrichVehicle_Jog_Steer_BS, groupName=Locomotion, method=SyncGroup}
   n36 = Blendspace Player 'OstrichVehicle_Jog_Steer_BS_WithGunner'!(X=@n34, PlayRate:float=1) -> (Pose) {blendSpace=/Game/Animation/Game/Vehicles/OstrichVehicle/OstrichVehicle_Jog_Steer_BS_WithGunner, groupName=Locomotion, method=SyncGroup}
   n37 = Get bHasGunner()
   n38 = Orientation Warping!(
    ComponentPose:Component Space Pose Link Structure=Local To Component(LocalPose=Blend Poses by bool(BlendPose_0=@n36, BlendPose_1=@n35, bActiveValue=@n37) {blendType=Linear})
    Alpha:float=1
    LocomotionAngle:float=Get LocalVelocityYawAngleInterpolated() -> (LocalVelocityYawAngleInterpolated:float)
   ) -> (Pose:Component Space Pose Link Structure) {
    is=title FortAnimGraphNode_OrientationWarping
    alphaBoolBlend {blendOption=Linear}
    settings {
     iKFootBones [{boneName=VB ik_foot_root_paw_ball_l}, {boneName=VB ik_foot_root_paw_ball_r}]
     iKFootRootBone {boneName=ik_foot_root}
     spineBones [{bone {boneName=spine_01}}, {bone {boneName=spine_02}}, {bone {boneName=spine_03}}]
    }
   }
   Save cached pose 'MovementPose'(Pose=Component To Local(ComponentPose=@n38)) {cacheName=MovementPose}
   n39 = Rotate Root Bone!(
    BasePose=Rotate Root Bone(BasePose=Use cached pose 'PostGunnerIdleAdditivePose'(), Pitch=0, Yaw=Get LocalAimYaw(), MeshToComponent="0,-90,0") {yawScaleBiasClamp {scale=-1}}
    Pitch:float=Get AimPitch()
    Yaw=0
    MeshToComponent:Rotator="0,-90,0"
   ) -> (Pose) {pitchScaleBiasClamp {scale=-1}, yawScaleBiasClamp {scale=-1}}
   n40 = Rotate Root Bone(
    BasePose=Slot 'RightArm'(Source=Slot 'LeftArm'(Source=@n39) {slotName=LeftArm}) {slotName=RightArm}
    Pitch=Get AimPitch()
    Yaw=0
    MeshToComponent="0,-90,0"
   ) {yawScaleBiasClamp {scale=-1}}
   n41 = Get LocalAimYaw()
   n42 = Look At - Bone: lowerarm_r!(
    ComponentPose=Local To Component(LocalPose=Rotate Root Bone(BasePose=@n40, Pitch=0, Yaw=@n41, MeshToComponent="0,-90,0"))
    bAlphaBoolEnabled:bool=Get bShouldRightArmLookAt() -> (bShouldRightArmLookAt:bool)
    LookAtLocation:Vector=Get WorldSpaceAimPoint() -> (WorldSpaceAimPoint:Vector)
   ) -> (Pose) {
    alphaInputType=Bool
    interpolationTime=1
    lookAtClamp=45
    alphaBoolBlend {blendInTime=0.15, blendOutTime=0.35}
    boneToModify {boneName=lowerarm_r}
    lookAt_Axis {axis=<-1 0>}
    lookAtTarget=~
   }
   AnimGraphNode_StateMachine_6 = ShotgunRecoilLowerBody() {is=title AnimGraphNode_StateMachine}
   n43 = Layered blend per bone(
    BasePose=Component To Local(ComponentPose=@n42)
    BlendPoses_0=@AnimGraphNode_StateMachine_6
    BlendWeights_0=Get IsShotgunFiredFloat() -> (IsShotgunFiredFloat:float)
   ) {bMeshSpaceRotationBlend=true, layerSetup [{branchFilters [{boneName=pelvis}, {blendDepth=-1, boneName=spine_01}]}]}
   AnimGraphNode_StateMachine_3 = RocketRecoilLowerBody() {is=title AnimGraphNode_StateMachine}
   n44 = Layered blend per bone(BasePose=@n43, BlendPoses_0=@AnimGraphNode_StateMachine_3, BlendWeights_0=Get RocketCurveValueFloat() -> (RocketCurveValueFloat:float)) {bMeshSpaceRotationBlend=true, layerSetup [{branchFilters [{boneName=pelvis}, {blendDepth=-1, boneName=spine_01}]}]}
   Save cached pose 'LeftRightArmsPose'(Pose=@n44) {cacheName=LeftRightArmsPose}
   n45 = Layered blend per bone(BasePose=Use cached pose 'LeftRightArmsPose'(), BlendPoses_0=Use cached pose 'PostGunnerIdleAdditivePose'()) {bMeshSpaceRotationBlend=true, layerSetup [{branchFilters [{boneName=pelvis}, {blendDepth=-1, boneName=spine_01}]}]}
   Save cached pose 'LeftRightArmsInMotionPose'(Pose=Blend Poses by bool(BlendPose_0=@n45, BlendPose_1=Use cached pose 'LeftRightArmsPose'(), bActiveValue=Get bIsMoving2D()) {blendType=Linear}) {cacheName=LeftRightArmsInMotionPose}
   AnimGraphNode_StateMachine_2 = Gunner() {is=title AnimGraphNode_StateMachine}
   n46 = Use cached pose 'PostJumpPose'()
   Save cached pose 'PreModifyBonePose'(Pose=AimOffset Player 'OstrichVehicle_Pitch_AO'!(
     X:float=Get AimPitch()
     BasePose:Pose Link Structure=Layered blend per bone(BasePose=@n46, BlendPoses_0=@AnimGraphNode_StateMachine_2) {layerSetup [{branchFilters [{boneName=upperarm_l}, {boneName=upperarm_r}]}]}
     Alpha:float=Get GunnerAlphaValue() -> (GunnerAlphaValue:float)
    ) -> (Pose:Pose Link Structure) {blendSpace=/Game/Animation/Game/Vehicles/OstrichVehicle/AO/OstrichVehicle_Pitch_AO, alphaBoolBlend {blendOption=Linear}}) {cacheName=PreModifyBonePose}
   n47 = Get GunnerAimTargetDeltaPitch() -> (GunnerAimTargetDeltaPitch:float)
   n48 = Get GunnerAimTargetDeltaYaw() -> (GunnerAimTargetDeltaYaw:float)
   AnimGraphNode_StateMachine_1 = JumpDashAbilities() {is=title AnimGraphNode_StateMachine}
   n49 = Blend Poses by bool(
    BlendPose_0=Sequence Player 'OstrichVehicle_ChargedJumpFall'() {sequence=/Game/Animation/Game/Vehicles/OstrichVehicle/OstrichVehicle_ChargedJumpFall}
    BlendPose_1=@AnimGraphNode_StateMachine_1
    bActiveValue=Get bShouldPlaySkyTubingLoop() -> (bShouldPlaySkyTubingLoop:bool)
   ) {blendType=Linear}
   n50 = Blendspace Player 'OstrichVehicle_Dash_Leans_BS'!(X=Get LocalAccelerationYawAngle() -> (LocalAccelerationYawAngle:float)) -> (Pose) {blendSpace=/Game/Animation/Game/Vehicles/OstrichVehicle/OstrichVehicle_Dash_Leans_BS}
   n51 = Get DashLeansAlphaValue() -> (DashLeansAlphaValue:float)
   n52 = Layered blend per bone(
    BasePose=Slot 'FullBody'(Source=Apply Additive!(Base=@n49, Additive=@n50, Alpha:float=@n51) -> (Pose) {alphaBoolBlend {blendInTime=0.15, blendOutTime=0.15}}) {slotName=FullBody}
    BlendPoses_0=Blend Poses by bool(
     BlendPose_0=Blendspace Player 'OstrichVehicle_Joystick_BS'(X=@n48, Y=@n47) {blendSpace=/Game/Animation/Game/Vehicles/OstrichVehicle/Parts/OstrichVehicle_Joystick_BS}
     BlendPose_1=Sequence Player 'OstrichVehicle_Idle_NoPassenger_Pose'() {sequence=&OstrichVehicle_Idle_NoPassenger_Pose}
     bActiveValue=Get bHasGunner()
    ) {blendType=Linear}
   ) {layerSetup [{branchFilters [{boneName=dyn_gunner_joystick_l}, {boneName=dyn_gunner_joystick_r}]}]}
   Save cached pose 'PostJumpPose'(Pose=Layered blend per bone(
     BasePose=@n52
     BlendPoses_0=Sequence Player 'OstrichVehicle_Idle_Pose'() {sequence=&OstrichVehicle_Idle_Pose}
     BlendWeights_0=Get NotJustGotDriverFloat()
    ) {layerSetup [{branchFilters [{boneName=weapon_base_l}, {boneName=weapon_base_r}]}]}) {cacheName=PostJumpPose}
   n53 = Layered blend per bone(
    BasePose=Use cached pose 'LeftRightArmsInMotionPose'()
    BlendPoses_0=Use cached pose 'PostGunnerIdleAdditivePose'()
    BlendWeights_0=Get RecoilAmountAlphaValue() -> (RecoilAmountAlphaValue:float)
   ) {
    bMeshSpaceRotationBlend=true
    layerSetup [{branchFilters [{boneName=pelvis}, {blendDepth=-1, boneName=upperarm_l}, {blendDepth=-1, boneName=upperarm_r}]}]
   }
   n54 = Use cached pose 'PostGunnerIdleAdditivePose'()
   Save cached pose 'PostUpperRightArmPose'(Pose=Layered blend per bone(
     BasePose=@n53
     BlendPoses_0=Slot 'UpperRightArm'(Source=@n54) {slotName=UpperRightArm}
     BlendWeights_0=Get RightArmNoAdditiveAlphaValue() -> (RightArmNoAdditiveAlphaValue:float)
    ) {layerSetup [{branchFilters [{boneName=upperarm_r}]}]}) {cacheName=PostUpperRightArmPose}
   AnimGraphNode_StateMachine_5 = UpperBodyBounce() {is=title AnimGraphNode_StateMachine}
   Save cached pose 'PostUpperBodyAdditivePose'(Pose=Layered blend per bone(
     BasePose=Use cached pose 'PostModifyBonePose'()
     BlendPoses_0=@AnimGraphNode_StateMachine_5
     BlendWeights_0=Get GunnerJogBounceAlphaValue() -> (GunnerJogBounceAlphaValue:float)
    ) {layerSetup [{branchFilters [{boneName=spine_05}]}]}) {cacheName=PostUpperBodyAdditivePose}
   n55 = Layered blend per bone(
    BasePose=Use cached pose 'PostUpperBodyAdditivePose'()
    BlendPoses_0=Apply Additive(Base=Use cached pose 'PostUpperBodyAdditivePose'(), Additive=Use cached pose 'IdleAdditive'(), Alpha=0.5) {alphaBoolBlend {blendOption=Linear}}
    BlendWeights_0=Get GunnerAlphaValue()
   ) {
    bMeshSpaceRotationBlend=true
    layerSetup [{branchFilters [{blendDepth=3, boneName=spine_01}, {blendDepth=-2, boneName=upperarm_l}, {blendDepth=-2, boneName=upperarm_r}]}]
   }
   Save cached pose 'PostGunnerIdleAdditivePose'(Pose=@n55) {cacheName=PostGunnerIdleAdditivePose}
   Save cached pose 'IdleAdditive'(Pose="Sequence Player (additive) 'OstrichVehicle_Idle'"() {sequence=/Game/Animation/Game/Vehicles/OstrichVehicle/OstrichVehicle_Idle}) {cacheName=IdleAdditive}
   AnimGraphNode_StateMachine_7 = ShoulderFinAdditive() {is=title AnimGraphNode_StateMachine}
   n56 = Layered blend per bone(
    BasePose=Use cached pose 'PostUpperRightArmPose'()
    BlendPoses_0=Apply Additive(Base=Use cached pose 'PostUpperRightArmPose'(), Additive=@AnimGraphNode_StateMachine_7) {alphaBoolBlend {blendOption=Linear}}
    BlendWeights_0=Get RightArmNoAdditiveAlphaValue()
   ) {bMeshSpaceRotationBlend=true, layerSetup [{branchFilters [{boneName=clavicle_r}, {blendDepth=-1, boneName=upperarm_r}]}]}
   n57 = Layered blend per bone(
    BasePose=@n56
    BlendPoses_0=Sequence Player 'OstrichVehicle_ShotgunShells_Loop'!(PlayRate:float=Get ShellsPlayRateValue() -> (ShellsPlayRateValue:float)) -> (Pose) {sequence=/Game/Animation/Game/Vehicles/OstrichVehicle/OstrichVehicle_ShotgunShells_Loop}
    BlendWeights_0=Get ShotgunShellsAlphaValue() -> (ShotgunShellsAlphaValue:float)
   ) {layerSetup [{branchFilters [{boneName=weapon_shotgunshell_01}, {boneName=weapon_shotgunshell_02}, {boneName=weapon_shotgunshell_03}, {boneName=weapon_shotgunshell_04}]}]}
   Save cached pose 'PostAdditivePose'(Pose=@n57) {cacheName=PostAdditivePose}
   n58 = "Transform (Modify) Bone - Bone: spine_05"!(
    ComponentPose=Local To Component(LocalPose=Use cached pose 'PreModifyBonePose'())
    Alpha=1
    Rotation=Get UpperBodyLagOffset() -> (UpperBodyLagOffset:Rotator)
   ) -> (Pose) {lODThreshold=3, rotationMode=BMM_Additive, alphaBoolBlend {blendOption=Linear}, boneToModify {boneName=spine_05}}
   n59 = "Transform (Modify) Bone - Bone: spine_05"(
    ComponentPose=Local To Component(LocalPose=Use cached pose 'PreModifyBonePose'())
    Alpha=Get GunnerAlphaValue()
    Rotation=Get AimYawRotationValue() -> (AimYawRotationValue:Rotator)
   ) {
    lODThreshold=3
    rotationMode=BMM_Replace
    rotationSpace=BCS_WorldSpace
    alphaBoolBlend {blendOption=Linear}
    boneToModify {boneName=spine_05}
   }
   Save cached pose 'PostModifyBonePose'(Pose=Blend Poses by bool(
     BlendPose_0=Component To Local(ComponentPose=@n59)
     BlendPose_1=Component To Local(ComponentPose=@n58)
     BlendTime_0=Get GunnerAimYawBlendIn() -> (GunnerAimYawBlendIn:float)
     BlendTime_1=Get GunnerAimYawBlendOut() -> (GunnerAimYawBlendOut:float)
     bActiveValue=Get bHasGunner()
    ) {blendType=Linear}) {cacheName=PostModifyBonePose}
  ]}
 OstrichVehicle {
  entry=Locomotion
  path=AnimGraph.AnimGraphNode_StateMachine_0.OstrichVehicle
  ids=id
  nodes [AnimStateNode_5 = Locomotion() {is=title AnimStateNode}]
 }
 Locomotion {
  path=AnimGraph.AnimGraphNode_StateMachine_0.OstrichVehicle.AnimStateNode_5.Locomotion
  nodes [
   AnimGraphNode_StateMachine_0 = LocomotionSub() {is=title AnimGraphNode_StateMachine}
   Output Animation Pose(Result=@AnimGraphNode_StateMachine_0)
  ]
 }
 LocomotionSub {
  entry=Idle
  path=AnimGraph.AnimGraphNode_StateMachine_0.OstrichVehicle.AnimStateNode_5.Locomotion.AnimGraphNode_StateMachine_0.LocomotionSub
  ids=id
  nodes [
   AnimStateNode_0 = ShutDown() {is=title AnimStateNode}
   AnimStateNode_1 = Empty() {is=title AnimStateNode, stateEntered {guid=#1, notifyName=EmptyPoseState}}
   AnimStateNode_2 = Movement() {is=title AnimStateNode}
   AnimStateNode_3 = Idle()
   AnimStateNode_4 = BootUp() {is=title AnimStateNode}
   AnimStateNode_5 = Turn() {is=title AnimStateNode}
   AnimStateNode_6 = Stop() {is=title AnimStateNode}
   AnimStateNode_7 = Pivot() {is=title AnimStateNode}
   AnimStateNode_8 = StopMovement() {is=title AnimStateNode}
   AnimStateNode_9 = Start() {is=title AnimStateNode}
   AnimStateTransitionNode_0 = AnimStateTransitionNode() {crossfadeDuration=0}
   AnimStateTransitionNode_1 = AnimStateTransitionNode() {crossfadeDuration=0.1}
   AnimStateTransitionNode_10 = AnimStateTransitionNode() {crossfadeDuration=0.1}
   AnimStateTransitionNode_11 = AnimStateTransitionNode() {bAutomaticRuleBasedOnSequencePlayerInState=true, crossfadeDuration=0.3}
   AnimStateTransitionNode_12 = AnimStateTransitionNode() {crossfadeDuration=0.1}
   AnimStateTransitionNode_13 = AnimStateTransitionNode()
   AnimStateTransitionNode_14 = AnimStateTransitionNode() {bAutomaticRuleBasedOnSequencePlayerInState=true, priorityOrder=2}
   AnimStateTransitionNode_15 = AnimStateTransitionNode() {priorityOrder=2}
   AnimStateTransitionNode_16 = AnimStateTransitionNode()
   AnimStateTransitionNode_17 = AnimStateTransitionNode() {crossfadeDuration=0.1}
   AnimStateTransitionNode_18 = AnimStateTransitionNode() {crossfadeDuration=0}
   AnimStateTransitionNode_19 = AnimStateTransitionNode() {priorityOrder=2}
   AnimStateTransitionNode_2 = AnimStateTransitionNode() {bAutomaticRuleBasedOnSequencePlayerInState=true}
   AnimStateTransitionNode_20 = AnimStateTransitionNode() {bAutomaticRuleBasedOnSequencePlayerInState=true, crossfadeDuration=0.4}
   AnimStateTransitionNode_21 = AnimStateTransitionNode()
   AnimStateTransitionNode_22 = AnimStateTransitionNode() {crossfadeDuration=0.25}
   AnimStateTransitionNode_23 = AnimStateTransitionNode() {bAutomaticRuleBasedOnSequencePlayerInState=true, crossfadeDuration=0.25}
   AnimStateTransitionNode_24 = AnimStateTransitionNode()
   AnimStateTransitionNode_25 = AnimStateTransitionNode() {crossfadeDuration=0.25}
   AnimStateTransitionNode_26 = AnimStateTransitionNode() {bAutomaticRuleBasedOnSequencePlayerInState=true, crossfadeDuration=0.4}
   AnimStateTransitionNode_27 = AnimStateTransitionNode()
   AnimStateTransitionNode_28 = AnimStateTransitionNode()
   AnimStateTransitionNode_29 = AnimStateTransitionNode()
   AnimStateTransitionNode_3 = AnimStateTransitionNode() {priorityOrder=2}
   AnimStateTransitionNode_30 = AnimStateTransitionNode()
   AnimStateTransitionNode_31 = AnimStateTransitionNode() {crossfadeDuration=0, priorityOrder=2}
   AnimStateTransitionNode_32 = AnimStateTransitionNode() {crossfadeDuration=0, priorityOrder=2}
   AnimStateTransitionNode_33 = AnimStateTransitionNode() {priorityOrder=2}
   AnimStateTransitionNode_34 = AnimStateTransitionNode()
   AnimStateTransitionNode_35 = AnimStateTransitionNode()
   AnimStateTransitionNode_4 = AnimStateTransitionNode()
   AnimStateTransitionNode_5 = AnimStateTransitionNode() {priorityOrder=2}
   AnimStateTransitionNode_6 = AnimStateTransitionNode() {crossfadeDuration=0}
   AnimStateTransitionNode_7 = AnimStateTransitionNode() {crossfadeDuration=0.1}
   AnimStateTransitionNode_8 = AnimStateTransitionNode()
   AnimStateTransitionNode_9 = AnimStateTransitionNode() {bAutomaticRuleBasedOnSequencePlayerInState=true}
  ]
  links [
   {
    dst=Idle
    rule=AnimGraph.AnimGraphNode_StateMachine_0.OstrichVehicle.AnimStateNode_5.Locomotion.AnimGraphNode_StateMachine_0.LocomotionSub.AnimStateTransitionNode_2.AnimationTransitionGraph_1
    src=BootUp
    via=AnimStateTransitionNode_2
   }
   {
    dst=Movement
    rule=AnimGraph.AnimGraphNode_StateMachine_0.OstrichVehicle.AnimStateNode_5.Locomotion.AnimGraphNode_StateMachine_0.LocomotionSub.AnimStateTransitionNode_1.AnimationTransitionGraph_2
    src=BootUp
    via=AnimStateTransitionNode_1
   }
   {
    dst=ShutDown
    rule=AnimGraph.AnimGraphNode_StateMachine_0.OstrichVehicle.AnimStateNode_5.Locomotion.AnimGraphNode_StateMachine_0.LocomotionSub.AnimStateTransitionNode_12.AnimationTransitionGraph_2
    src=BootUp
    via=AnimStateTransitionNode_12
   }
   {
    dst=Start
    rule=AnimGraph.AnimGraphNode_StateMachine_0.OstrichVehicle.AnimStateNode_5.Locomotion.AnimGraphNode_StateMachine_0.LocomotionSub.AnimStateTransitionNode_29.AnimationTransitionGraph_1
    src=BootUp
    via=AnimStateTransitionNode_29
   }
   {
    dst=BootUp
    rule=AnimGraph.AnimGraphNode_StateMachine_0.OstrichVehicle.AnimStateNode_5.Locomotion.AnimGraphNode_StateMachine_0.LocomotionSub.AnimStateTransitionNode_6.AnimationTransitionGraph_1
    src=Empty
    via=AnimStateTransitionNode_6
   }
   {
    dst=Idle
    rule=AnimGraph.AnimGraphNode_StateMachine_0.OstrichVehicle.AnimStateNode_5.Locomotion.AnimGraphNode_StateMachine_0.LocomotionSub.AnimStateTransitionNode_32.AnimationTransitionGraph_2
    src=Empty
    via=AnimStateTransitionNode_32
   }
   {
    dst=BootUp
    rule=AnimGraph.AnimGraphNode_StateMachine_0.OstrichVehicle.AnimStateNode_5.Locomotion.AnimGraphNode_StateMachine_0.LocomotionSub.AnimStateTransitionNode_0.AnimationTransitionGraph_2
    src=Idle
    via=AnimStateTransitionNode_0
   }
   {
    dst=Empty
    rule=AnimGraph.AnimGraphNode_StateMachine_0.OstrichVehicle.AnimStateNode_5.Locomotion.AnimGraphNode_StateMachine_0.LocomotionSub.AnimStateTransitionNode_18.AnimationTransitionGraph_1
    src=Idle
    via=AnimStateTransitionNode_18
   }
   {
    dst=Empty
    rule=AnimGraph.AnimGraphNode_StateMachine_0.OstrichVehicle.AnimStateNode_5.Locomotion.AnimGraphNode_StateMachine_0.LocomotionSub.AnimStateTransitionNode_31.AnimationTransitionGraph_2
    src=Idle
    via=AnimStateTransitionNode_31
   }
   {
    dst=Movement
    rule=AnimGraph.AnimGraphNode_StateMachine_0.OstrichVehicle.AnimStateNode_5.Locomotion.AnimGraphNode_StateMachine_0.LocomotionSub.AnimStateTransitionNode_3.AnimationTransitionGraph_1
    src=Idle
    via=AnimStateTransitionNode_3
   }
   {
    dst=ShutDown
    rule=AnimGraph.AnimGraphNode_StateMachine_0.OstrichVehicle.AnimStateNode_5.Locomotion.AnimGraphNode_StateMachine_0.LocomotionSub.AnimStateTransitionNode_5.AnimationTransitionGraph_1
    src=Idle
    via=AnimStateTransitionNode_5
   }
   {
    dst=Start
    rule=AnimGraph.AnimGraphNode_StateMachine_0.OstrichVehicle.AnimStateNode_5.Locomotion.AnimGraphNode_StateMachine_0.LocomotionSub.AnimStateTransitionNode_24.AnimationTransitionGraph_1
    src=Idle
    via=AnimStateTransitionNode_24
   }
   {
    dst=Turn
    rule=AnimGraph.AnimGraphNode_StateMachine_0.OstrichVehicle.AnimStateNode_5.Locomotion.AnimGraphNode_StateMachine_0.LocomotionSub.AnimStateTransitionNode_13.AnimationTransitionGraph_1
    src=Idle
    via=AnimStateTransitionNode_13
   }
   {
    dst=Idle
    rule=AnimGraph.AnimGraphNode_StateMachine_0.OstrichVehicle.AnimStateNode_5.Locomotion.AnimGraphNode_StateMachine_0.LocomotionSub.AnimStateTransitionNode_19.AnimationTransitionGraph_1
    src=Movement
    via=AnimStateTransitionNode_19
   }
   {
    dst=Pivot
    rule=AnimGraph.AnimGraphNode_StateMachine_0.OstrichVehicle.AnimStateNode_5.Locomotion.AnimGraphNode_StateMachine_0.LocomotionSub.AnimStateTransitionNode_8.AnimationTransitionGraph_1
    src=Movement
    via=AnimStateTransitionNode_8
   }
   {
    dst=ShutDown
    rule=AnimGraph.AnimGraphNode_StateMachine_0.OstrichVehicle.AnimStateNode_5.Locomotion.AnimGraphNode_StateMachine_0.LocomotionSub.AnimStateTransitionNode_10.AnimationTransitionGraph_2
    src=Movement
    via=AnimStateTransitionNode_10
   }
   {
    dst=ShutDown
    rule=AnimGraph.AnimGraphNode_StateMachine_0.OstrichVehicle.AnimStateNode_5.Locomotion.AnimGraphNode_StateMachine_0.LocomotionSub.AnimStateTransitionNode_33.Transition
    src=Movement
    via=AnimStateTransitionNode_33
   }
   {
    dst=StopMovement
    rule=AnimGraph.AnimGraphNode_StateMachine_0.OstrichVehicle.AnimStateNode_5.Locomotion.AnimGraphNode_StateMachine_0.LocomotionSub.AnimStateTransitionNode_22.AnimationTransitionGraph_1
    src=Movement
    via=AnimStateTransitionNode_22
   }
   {
    dst=Movement
    rule=AnimGraph.AnimGraphNode_StateMachine_0.OstrichVehicle.AnimStateNode_5.Locomotion.AnimGraphNode_StateMachine_0.LocomotionSub.AnimStateTransitionNode_20.AnimationTransitionGraph_1
    src=Pivot
    via=AnimStateTransitionNode_20
   }
   {
    dst=Movement
    rule=AnimGraph.AnimGraphNode_StateMachine_0.OstrichVehicle.AnimStateNode_5.Locomotion.AnimGraphNode_StateMachine_0.LocomotionSub.AnimStateTransitionNode_21.AnimationTransitionGraph_1
    src=Pivot
    via=AnimStateTransitionNode_21
   }
   {
    dst=BootUp
    rule=AnimGraph.AnimGraphNode_StateMachine_0.OstrichVehicle.AnimStateNode_5.Locomotion.AnimGraphNode_StateMachine_0.LocomotionSub.AnimStateTransitionNode_4.AnimationTransitionGraph_1
    src=ShutDown
    via=AnimStateTransitionNode_4
   }
   {
    dst=Empty
    rule=AnimGraph.AnimGraphNode_StateMachine_0.OstrichVehicle.AnimStateNode_5.Locomotion.AnimGraphNode_StateMachine_0.LocomotionSub.AnimStateTransitionNode_9.AnimationTransitionGraph_1
    src=ShutDown
    via=AnimStateTransitionNode_9
   }
   {
    dst=Movement
    rule=AnimGraph.AnimGraphNode_StateMachine_0.OstrichVehicle.AnimStateNode_5.Locomotion.AnimGraphNode_StateMachine_0.LocomotionSub.AnimStateTransitionNode_26.AnimationTransitionGraph_1
    src=Start
    via=AnimStateTransitionNode_26
   }
   {
    dst=Movement
    rule=AnimGraph.AnimGraphNode_StateMachine_0.OstrichVehicle.AnimStateNode_5.Locomotion.AnimGraphNode_StateMachine_0.LocomotionSub.AnimStateTransitionNode_27.AnimationTransitionGraph_1
    src=Start
    via=AnimStateTransitionNode_27
   }
   {
    dst=ShutDown
    rule=AnimGraph.AnimGraphNode_StateMachine_0.OstrichVehicle.AnimStateNode_5.Locomotion.AnimGraphNode_StateMachine_0.LocomotionSub.AnimStateTransitionNode_34.Transition
    src=Start
    via=AnimStateTransitionNode_34
   }
   {
    dst=StopMovement
    rule=AnimGraph.AnimGraphNode_StateMachine_0.OstrichVehicle.AnimStateNode_5.Locomotion.AnimGraphNode_StateMachine_0.LocomotionSub.AnimStateTransitionNode_30.AnimationTransitionGraph_1
    src=Start
    via=AnimStateTransitionNode_30
   }
   {
    dst=Idle
    rule=AnimGraph.AnimGraphNode_StateMachine_0.OstrichVehicle.AnimStateNode_5.Locomotion.AnimGraphNode_StateMachine_0.LocomotionSub.AnimStateTransitionNode_11.AnimationTransitionGraph_1
    src=Stop
    via=AnimStateTransitionNode_11
   }
   {
    dst=Idle
    rule=AnimGraph.AnimGraphNode_StateMachine_0.OstrichVehicle.AnimStateNode_5.Locomotion.AnimGraphNode_StateMachine_0.LocomotionSub.AnimStateTransitionNode_17.AnimationTransitionGraph_1
    src=Stop
    via=AnimStateTransitionNode_17
   }
   {
    dst=ShutDown
    rule=AnimGraph.AnimGraphNode_StateMachine_0.OstrichVehicle.AnimStateNode_5.Locomotion.AnimGraphNode_StateMachine_0.LocomotionSub.AnimStateTransitionNode_7.AnimationTransitionGraph_1
    src=Stop
    via=AnimStateTransitionNode_7
   }
   {
    dst=Idle
    rule=AnimGraph.AnimGraphNode_StateMachine_0.OstrichVehicle.AnimStateNode_5.Locomotion.AnimGraphNode_StateMachine_0.LocomotionSub.AnimStateTransitionNode_23.AnimationTransitionGraph_1
    src=StopMovement
    via=AnimStateTransitionNode_23
   }
   {
    dst=Movement
    rule=AnimGraph.AnimGraphNode_StateMachine_0.OstrichVehicle.AnimStateNode_5.Locomotion.AnimGraphNode_StateMachine_0.LocomotionSub.AnimStateTransitionNode_25.AnimationTransitionGraph_1
    src=StopMovement
    via=AnimStateTransitionNode_25
   }
   {
    dst=ShutDown
    rule=AnimGraph.AnimGraphNode_StateMachine_0.OstrichVehicle.AnimStateNode_5.Locomotion.AnimGraphNode_StateMachine_0.LocomotionSub.AnimStateTransitionNode_35.Transition
    src=StopMovement
    via=AnimStateTransitionNode_35
   }
   {
    dst=Start
    rule=AnimGraph.AnimGraphNode_StateMachine_0.OstrichVehicle.AnimStateNode_5.Locomotion.AnimGraphNode_StateMachine_0.LocomotionSub.AnimStateTransitionNode_28.AnimationTransitionGraph_1
    src=StopMovement
    via=AnimStateTransitionNode_28
   }
   {
    dst=Idle
    rule=AnimGraph.AnimGraphNode_StateMachine_0.OstrichVehicle.AnimStateNode_5.Locomotion.AnimGraphNode_StateMachine_0.LocomotionSub.AnimStateTransitionNode_14.AnimationTransitionGraph_1
    src=Turn
    via=AnimStateTransitionNode_14
   }
   {
    dst=Idle
    rule=AnimGraph.AnimGraphNode_StateMachine_0.OstrichVehicle.AnimStateNode_5.Locomotion.AnimGraphNode_StateMachine_0.LocomotionSub.AnimStateTransitionNode_15.AnimationTransitionGraph_1
    src=Turn
    via=AnimStateTransitionNode_15
   }
   {
    dst=Idle
    rule=AnimGraph.AnimGraphNode_StateMachine_0.OstrichVehicle.AnimStateNode_5.Locomotion.AnimGraphNode_StateMachine_0.LocomotionSub.AnimStateTransitionNode_16.AnimationTransitionGraph_1
    src=Turn
    via=AnimStateTransitionNode_16
   }
  ]
 }
 ShutDown {
  path=AnimGraph.AnimGraphNode_StateMachine_0.OstrichVehicle.AnimStateNode_5.Locomotion.AnimGraphNode_StateMachine_0.LocomotionSub.AnimStateNode_0.ShutDown
  nodes [Output Animation Pose(Result=Sequence Player 'OstrichVehicle_Driver_ShutDown'!(PlayRate:float=1.2) -> (Pose) {bLoopAnimation=false, sequence=&OstrichVehicle_Driver_ShutDown})]
 }
 Empty {
  path=AnimGraph.AnimGraphNode_StateMachine_0.OstrichVehicle.AnimStateNode_5.Locomotion.AnimGraphNode_StateMachine_0.LocomotionSub.AnimStateNode_1.Empty
  nodes [Output Animation Pose(Result=Sequence Player 'OstrichVehicle_Empty_Pose'() {sequence=/Game/Animation/Game/Vehicles/OstrichVehicle/OstrichVehicle_Empty_Pose})]
 }
 Movement {
  path=AnimGraph.AnimGraphNode_StateMachine_0.OstrichVehicle.AnimStateNode_5.Locomotion.AnimGraphNode_StateMachine_0.LocomotionSub.AnimStateNode_2.Movement
  nodes [
   n1 = Speed Warping!(
    ComponentPose:Component Space Pose Link Structure=Local To Component(LocalPose=Use cached pose 'MovementPose'())
    Alpha:float=1
    SpeedScaling:float=1
   ) -> (Pose:Component Space Pose Link Structure) {
    is=title FortAnimGraphNode_SpeedWarping
    speedWarpingAxisMode=IKFootRootLocalY
    alphaBoolBlend {blendOption=Linear}
    feetDefinitions [
     {numBonesInLimb=3, fKFootBone {boneName=paw_ball_l}, iKFootBone {boneName=VB ik_foot_root_paw_ball_l}}
     {numBonesInLimb=3, fKFootBone {boneName=paw_ball_r}, iKFootBone {boneName=VB ik_foot_root_paw_ball_r}}
    ]
    iKFootRootBone {boneName=ik_foot_root}
    pelvisBone {boneName=pelvis}
   }
   Output Animation Pose(Result=Component To Local(ComponentPose=@n1))
  ]
 }
 Idle {
  path=AnimGraph.AnimGraphNode_StateMachine_0.OstrichVehicle.AnimStateNode_5.Locomotion.AnimGraphNode_StateMachine_0.LocomotionSub.AnimStateNode_3.Idle
  nodes [
   n1 = Random Sequence Player!() -> (Pose:Pose Link Structure) {
    is=title AnimGraphNode_RandomPlayer
    entries [
     {
      chanceToPlay=0.85
      maxLoopCount=0
      maxPlayRate=1
      minLoopCount=0
      minPlayRate=1
      sequence=/Game/Animation/Game/Vehicles/OstrichVehicle/OstrichVehicle_Idle_Pose_Hold
      blendIn {blendOption=Linear, blendTime=0.2, customCurve=null}
     }
     {
      chanceToPlay=0.15
      maxLoopCount=0
      maxPlayRate=1
      minLoopCount=0
      minPlayRate=1
      sequence=/Game/Animation/Game/Vehicles/OstrichVehicle/OstrichVehicle_Idle_A
      blendIn {blendOption=Linear, blendTime=0.2, customCurve=null}
     }
    ]
   }
   Output Animation Pose(Result=Apply Additive(Base=@n1, Additive=Use cached pose 'IdleAdditive'(), Alpha=0.7) {alphaBoolBlend {blendOption=Linear}})
  ]
 }
 BootUp {
  path=AnimGraph.AnimGraphNode_StateMachine_0.OstrichVehicle.AnimStateNode_5.Locomotion.AnimGraphNode_StateMachine_0.LocomotionSub.AnimStateNode_4.BootUp
  nodes [Output Animation Pose(Result=Sequence Player 'OstrichVehicle_Driver_BootUp'() {bLoopAnimation=false, sequence=&OstrichVehicle_Driver_BootUp})]
 }
 Turn {
  path=AnimGraph.AnimGraphNode_StateMachine_0.OstrichVehicle.AnimStateNode_5.Locomotion.AnimGraphNode_StateMachine_0.LocomotionSub.AnimStateNode_5.Turn
  nodes [
   n1 = Get TurnInPlaceInput()
   Output Animation Pose(Result=Blend Poses by bool(
     BlendPose_0=Sequence Player 'OstrichVehicle_TurnLeft90'() {bLoopAnimation=false, sequence=/Game/Animation/Game/Vehicles/OstrichVehicle/OstrichVehicle_TurnLeft90}
     BlendPose_1=Sequence Player 'OstrichVehicle_TurnRight90'() {bLoopAnimation=false, sequence=/Game/Animation/Game/Vehicles/OstrichVehicle/OstrichVehicle_TurnRight90}
     bActiveValue=Break Fort Anim Input Turn in Place(FortAnimInput_TurnInPlace=@n1) -> (
      RootYawOffset=~
      TurnAnimPlayRate=~
      bWantsToTurn=~
      bExitTurn=~
      bUseTurnA=~
      TurnACurveAlpha=~
      TurnBCurveAlpha=~
      bWantsToTurnInVehicle=~
      bWantsToTurnAgain=~
      bTurningLeft=~
      ShufflePlayRate=~
      CrouchShufflePlayRate=~
      CrouchShuffleStartPosition=~
      CrouchShuffleSpeedScaling=~
      bTransition_LowerBody_Shuffle=~
      bTransition_LowerBody_ShuffleADS=~
      bPlayOutro=~
     ).bTurningLeft
    ) {blendType=Linear})
  ]
 }
 Stop {
  path=AnimGraph.AnimGraphNode_StateMachine_0.OstrichVehicle.AnimStateNode_5.Locomotion.AnimGraphNode_StateMachine_0.LocomotionSub.AnimStateNode_6.Stop
  nodes [
   n1 = Sequence Player 'OstrichVehicle_Jog_N_Stop'() {
    bLoopAnimation=false
    groupName=Locomotion
    method=SyncGroup
    sequence=/Game/Animation/Game/Vehicles/OstrichVehicle/OstrichVehicle_Jog_N_Stop
   }
   Output Animation Pose(Result=@n1)
  ]
 }
 Pivot {
  path=AnimGraph.AnimGraphNode_StateMachine_0.OstrichVehicle.AnimStateNode_5.Locomotion.AnimGraphNode_StateMachine_0.LocomotionSub.AnimStateNode_7.Pivot
  nodes [Output Animation Pose(Result=Rotate Root Bone(
     BasePose="Sequence Evaluator (by time)"!(Sequence:Anim Sequence Base Object Reference, ExplicitTime:float=0) -> (Pose:Pose Link Structure) {is=title AnimGraphNode_SequenceEvaluator, bTeleportToExplicitTime=false, groupName=Locomotion, groupRole=TransitionLeader, method=SyncGroup}
     Pitch=0
     Yaw=0
     MeshToComponent="0,0,0"
    ))]
 }
 StopMovement {
  path=AnimGraph.AnimGraphNode_StateMachine_0.OstrichVehicle.AnimStateNode_5.Locomotion.AnimGraphNode_StateMachine_0.LocomotionSub.AnimStateNode_8.StopMovement
  nodes [Output Animation Pose(Result=Sequence Player 'OstrichVehicle_Jog_N_Stop'() {bLoopAnimation=false, sequence=&OstrichVehicle_Jog_N_Stop})]
 }
 Start {
  path=AnimGraph.AnimGraphNode_StateMachine_0.OstrichVehicle.AnimStateNode_5.Locomotion.AnimGraphNode_StateMachine_0.LocomotionSub.AnimStateNode_9.Start
  nodes [
   n1 = "Sequence Evaluator (by time)"(ExplicitTime=0) {bTeleportToExplicitTime=false, groupName=Locomotion, groupRole=TransitionLeader, method=SyncGroup}
   n2 = Speed Warping(ComponentPose=Local To Component(LocalPose=Rotate Root Bone(BasePose=@n1, Pitch=0, Yaw=0, MeshToComponent="0,0,0")), Alpha=1, SpeedScaling=1) {
    speedWarpingAxisMode=IKFootRootLocalY
    alphaBoolBlend {blendOption=Linear}
    feetDefinitions [
     {numBonesInLimb=3, fKFootBone {boneName=paw_ball_l}, iKFootBone {boneName=VB ik_foot_root_paw_ball_l}}
     {numBonesInLimb=3, fKFootBone {boneName=paw_ball_r}, iKFootBone {boneName=VB ik_foot_root_paw_ball_r}}
    ]
    iKFootRootBone {boneName=ik_foot_root}
    pelvisBone {boneName=pelvis}
   }
   Output Animation Pose(Result=Component To Local(ComponentPose=@n2))
  ]
 }
 AnimationTransitionGraph_2 {
  path=AnimGraph.AnimGraphNode_StateMachine_0.OstrichVehicle.AnimStateNode_5.Locomotion.AnimGraphNode_StateMachine_0.LocomotionSub.AnimStateTransitionNode_0.AnimationTransitionGraph_2
  nodes [Result(bCanEnterTransition=Get bJustGotDriver() -> (bJustGotDriver:bool))]
 }
 AnimationTransitionGraph_2 {
  path=AnimGraph.AnimGraphNode_StateMachine_0.OstrichVehicle.AnimStateNode_5.Locomotion.AnimGraphNode_StateMachine_0.LocomotionSub.AnimStateTransitionNode_1.AnimationTransitionGraph_2
  nodes [
   Result(bCanEnterTransition=Get bTransition_BootUp_To_Movement() -> (bTransition_BootUp_To_Movement:bool))
   "float > float"(A=Get PawnSpeed2D(), B=50) {is=title K2Node_CallFunction}
  ]
 }
 AnimationTransitionGraph_2 {
  path=AnimGraph.AnimGraphNode_StateMachine_0.OstrichVehicle.AnimStateNode_5.Locomotion.AnimGraphNode_StateMachine_0.LocomotionSub.AnimStateTransitionNode_10.AnimationTransitionGraph_2
  nodes [Result(bCanEnterTransition=Get bJustLostDriver() -> (bJustLostDriver:bool))]
 }
 AnimationTransitionGraph_1 {
  path=AnimGraph.AnimGraphNode_StateMachine_0.OstrichVehicle.AnimStateNode_5.Locomotion.AnimGraphNode_StateMachine_0.LocomotionSub.AnimStateTransitionNode_11.AnimationTransitionGraph_1
  nodes [Result()]
 }
 AnimationTransitionGraph_2 {
  path=AnimGraph.AnimGraphNode_StateMachine_0.OstrichVehicle.AnimStateNode_5.Locomotion.AnimGraphNode_StateMachine_0.LocomotionSub.AnimStateTransitionNode_12.AnimationTransitionGraph_2
  nodes [Result(bCanEnterTransition=Get bJustLostDriver())]
 }
 AnimationTransitionGraph_1 {
  path=AnimGraph.AnimGraphNode_StateMachine_0.OstrichVehicle.AnimStateNode_5.Locomotion.AnimGraphNode_StateMachine_0.LocomotionSub.AnimStateTransitionNode_13.AnimationTransitionGraph_1
  nodes [Result(bCanEnterTransition=Break Fort Anim Input Turn in Place(FortAnimInput_TurnInPlace=Get TurnInPlaceInput()) -> (
     RootYawOffset=~
     TurnAnimPlayRate=~
     bWantsToTurn=~
     bExitTurn=~
     bUseTurnA=~
     TurnACurveAlpha=~
     TurnBCurveAlpha=~
     bWantsToTurnInVehicle=~
     bWantsToTurnAgain=~
     bTurningLeft=~
     ShufflePlayRate=~
     CrouchShufflePlayRate=~
     CrouchShuffleStartPosition=~
     CrouchShuffleSpeedScaling=~
     bTransition_LowerBody_Shuffle=~
     bTransition_LowerBody_ShuffleADS=~
     bPlayOutro=~
    ).bWantsToTurn)]
 }
 AnimationTransitionGraph_1 {
  path=AnimGraph.AnimGraphNode_StateMachine_0.OstrichVehicle.AnimStateNode_5.Locomotion.AnimGraphNode_StateMachine_0.LocomotionSub.AnimStateTransitionNode_14.AnimationTransitionGraph_1
  nodes [Result()]
 }
 AnimationTransitionGraph_1 {
  path=AnimGraph.AnimGraphNode_StateMachine_0.OstrichVehicle.AnimStateNode_5.Locomotion.AnimGraphNode_StateMachine_0.LocomotionSub.AnimStateTransitionNode_15.AnimationTransitionGraph_1
  nodes [Result(bCanEnterTransition=Break Fort Anim Input Turn in Place(FortAnimInput_TurnInPlace=Get TurnInPlaceInput()) -> (
     RootYawOffset=~
     TurnAnimPlayRate=~
     bWantsToTurn=~
     bExitTurn=~
     bUseTurnA=~
     TurnACurveAlpha=~
     TurnBCurveAlpha=~
     bWantsToTurnInVehicle=~
     bWantsToTurnAgain=~
     bTurningLeft=~
     ShufflePlayRate=~
     CrouchShufflePlayRate=~
     CrouchShuffleStartPosition=~
     CrouchShuffleSpeedScaling=~
     bTransition_LowerBody_Shuffle=~
     bTransition_LowerBody_ShuffleADS=~
     bPlayOutro=~
    ).bWantsToTurnAgain)]
 }
 AnimationTransitionGraph_1 {
  path=AnimGraph.AnimGraphNode_StateMachine_0.OstrichVehicle.AnimStateNode_5.Locomotion.AnimGraphNode_StateMachine_0.LocomotionSub.AnimStateTransitionNode_16.AnimationTransitionGraph_1
  nodes [
   Result(bCanEnterTransition=Get bTransition_Turn_To_Idle() -> (bTransition_Turn_To_Idle:bool))
   "float > float"(A=Get PawnSpeed2D(), B=50)
  ]
 }
 AnimationTransitionGraph_1 {
  path=AnimGraph.AnimGraphNode_StateMachine_0.OstrichVehicle.AnimStateNode_5.Locomotion.AnimGraphNode_StateMachine_0.LocomotionSub.AnimStateTransitionNode_17.AnimationTransitionGraph_1
  nodes [
   Result(bCanEnterTransition=Get bTransition_Stop_To_Idle() -> (bTransition_Stop_To_Idle:bool))
   "float > float"(A=Get PawnSpeed2D(), B=50)
  ]
 }
 AnimationTransitionGraph_1 {
  path=AnimGraph.AnimGraphNode_StateMachine_0.OstrichVehicle.AnimStateNode_5.Locomotion.AnimGraphNode_StateMachine_0.LocomotionSub.AnimStateTransitionNode_18.AnimationTransitionGraph_1
  nodes [Result(bCanEnterTransition=Get bIsEmptyOnFirstLoad() -> (bIsEmptyOnFirstLoad:bool))]
 }
 AnimationTransitionGraph_1 {
  path=AnimGraph.AnimGraphNode_StateMachine_0.OstrichVehicle.AnimStateNode_5.Locomotion.AnimGraphNode_StateMachine_0.LocomotionSub.AnimStateTransitionNode_19.AnimationTransitionGraph_1
  nodes [Result(bCanEnterTransition=Get bShouldPlayJogStop() -> (bShouldPlayJogStop:bool))]
 }
 AnimationTransitionGraph_1 {
  path=AnimGraph.AnimGraphNode_StateMachine_0.OstrichVehicle.AnimStateNode_5.Locomotion.AnimGraphNode_StateMachine_0.LocomotionSub.AnimStateTransitionNode_2.AnimationTransitionGraph_1
  nodes [Result()]
 }
 AnimationTransitionGraph_1 {
  path=AnimGraph.AnimGraphNode_StateMachine_0.OstrichVehicle.AnimStateNode_5.Locomotion.AnimGraphNode_StateMachine_0.LocomotionSub.AnimStateTransitionNode_20.AnimationTransitionGraph_1
  nodes [Result()]
 }
 AnimationTransitionGraph_1 {
  path=AnimGraph.AnimGraphNode_StateMachine_0.OstrichVehicle.AnimStateNode_5.Locomotion.AnimGraphNode_StateMachine_0.LocomotionSub.AnimStateTransitionNode_21.AnimationTransitionGraph_1
  nodes [Result(bCanEnterTransition=Get bPivotEarlyOut() -> (bPivotEarlyOut:bool))]
 }
 AnimationTransitionGraph_1 {
  path=AnimGraph.AnimGraphNode_StateMachine_0.OstrichVehicle.AnimStateNode_5.Locomotion.AnimGraphNode_StateMachine_0.LocomotionSub.AnimStateTransitionNode_22.AnimationTransitionGraph_1
  nodes [Result(bCanEnterTransition=Get bShouldPlayJogStop())]
 }
 AnimationTransitionGraph_1 {
  path=AnimGraph.AnimGraphNode_StateMachine_0.OstrichVehicle.AnimStateNode_5.Locomotion.AnimGraphNode_StateMachine_0.LocomotionSub.AnimStateTransitionNode_23.AnimationTransitionGraph_1
  nodes [Result()]
 }
 AnimationTransitionGraph_1 {
  path=AnimGraph.AnimGraphNode_StateMachine_0.OstrichVehicle.AnimStateNode_5.Locomotion.AnimGraphNode_StateMachine_0.LocomotionSub.AnimStateTransitionNode_24.AnimationTransitionGraph_1
  nodes [Result(bCanEnterTransition=Get bStartTransitionTriggeredAndEnableStarts() -> (bStartTransitionTriggeredAndEnableStarts:bool))]
 }
 AnimationTransitionGraph_1 {
  path=AnimGraph.AnimGraphNode_StateMachine_0.OstrichVehicle.AnimStateNode_5.Locomotion.AnimGraphNode_StateMachine_0.LocomotionSub.AnimStateTransitionNode_25.AnimationTransitionGraph_1
  nodes [Result(bCanEnterTransition=Get bTransition_Idle_To_Movement() -> (bTransition_Idle_To_Movement:bool))]
 }
 AnimationTransitionGraph_1 {
  path=AnimGraph.AnimGraphNode_StateMachine_0.OstrichVehicle.AnimStateNode_5.Locomotion.AnimGraphNode_StateMachine_0.LocomotionSub.AnimStateTransitionNode_26.AnimationTransitionGraph_1
  nodes [Result()]
 }
 AnimationTransitionGraph_1 {
  path=AnimGraph.AnimGraphNode_StateMachine_0.OstrichVehicle.AnimStateNode_5.Locomotion.AnimGraphNode_StateMachine_0.LocomotionSub.AnimStateTransitionNode_27.AnimationTransitionGraph_1
  nodes [Result(bCanEnterTransition=Get bStartEarlyOut() -> (bStartEarlyOut:bool))]
 }
 AnimationTransitionGraph_1 {
  path=AnimGraph.AnimGraphNode_StateMachine_0.OstrichVehicle.AnimStateNode_5.Locomotion.AnimGraphNode_StateMachine_0.LocomotionSub.AnimStateTransitionNode_28.AnimationTransitionGraph_1
  nodes [Result(bCanEnterTransition=Get bStartTransitionTriggeredAndEnableStarts())]
 }
 AnimationTransitionGraph_1 {
  path=AnimGraph.AnimGraphNode_StateMachine_0.OstrichVehicle.AnimStateNode_5.Locomotion.AnimGraphNode_StateMachine_0.LocomotionSub.AnimStateTransitionNode_29.AnimationTransitionGraph_1
  nodes [Result(bCanEnterTransition=Get bStartTransitionTriggeredAndEnableStarts())]
 }
 AnimationTransitionGraph_1 {
  path=AnimGraph.AnimGraphNode_StateMachine_0.OstrichVehicle.AnimStateNode_5.Locomotion.AnimGraphNode_StateMachine_0.LocomotionSub.AnimStateTransitionNode_3.AnimationTransitionGraph_1
  nodes [
   Result(bCanEnterTransition=Get bTransition_Idle_To_Movement())
   "float > float"(A=Get PawnSpeed2D(), B=50)
  ]
 }
 AnimationTransitionGraph_1 {
  path=AnimGraph.AnimGraphNode_StateMachine_0.OstrichVehicle.AnimStateNode_5.Locomotion.AnimGraphNode_StateMachine_0.LocomotionSub.AnimStateTransitionNode_30.AnimationTransitionGraph_1
  nodes [Result(bCanEnterTransition=NOT Boolean(A=Get bIsMoving2D()))]
 }
 AnimationTransitionGraph_2 {
  path=AnimGraph.AnimGraphNode_StateMachine_0.OstrichVehicle.AnimStateNode_5.Locomotion.AnimGraphNode_StateMachine_0.LocomotionSub.AnimStateTransitionNode_31.AnimationTransitionGraph_2
  nodes [Result(bCanEnterTransition=NOT Boolean(A=Get bHasDriver() -> (bHasDriver:bool)))]
 }
 AnimationTransitionGraph_2 {
  path=AnimGraph.AnimGraphNode_StateMachine_0.OstrichVehicle.AnimStateNode_5.Locomotion.AnimGraphNode_StateMachine_0.LocomotionSub.AnimStateTransitionNode_32.AnimationTransitionGraph_2
  nodes [Result(bCanEnterTransition=Get bHasDriver())]
 }
 Transition {
  path=AnimGraph.AnimGraphNode_StateMachine_0.OstrichVehicle.AnimStateNode_5.Locomotion.AnimGraphNode_StateMachine_0.LocomotionSub.AnimStateTransitionNode_33.Transition
  nodes [Result(bCanEnterTransition=NOT Boolean(A=Get bHasDriver()))]
 }
 Transition {
  path=AnimGraph.AnimGraphNode_StateMachine_0.OstrichVehicle.AnimStateNode_5.Locomotion.AnimGraphNode_StateMachine_0.LocomotionSub.AnimStateTransitionNode_34.Transition
  nodes [Result(bCanEnterTransition=Get bJustLostDriver())]
 }
 Transition {
  path=AnimGraph.AnimGraphNode_StateMachine_0.OstrichVehicle.AnimStateNode_5.Locomotion.AnimGraphNode_StateMachine_0.LocomotionSub.AnimStateTransitionNode_35.Transition
  nodes [Result(bCanEnterTransition=Get bJustLostDriver())]
 }
 AnimationTransitionGraph_1 {
  path=AnimGraph.AnimGraphNode_StateMachine_0.OstrichVehicle.AnimStateNode_5.Locomotion.AnimGraphNode_StateMachine_0.LocomotionSub.AnimStateTransitionNode_4.AnimationTransitionGraph_1
  nodes [Result(bCanEnterTransition=Get bJustGotDriver())]
 }
 AnimationTransitionGraph_1 {
  path=AnimGraph.AnimGraphNode_StateMachine_0.OstrichVehicle.AnimStateNode_5.Locomotion.AnimGraphNode_StateMachine_0.LocomotionSub.AnimStateTransitionNode_5.AnimationTransitionGraph_1
  nodes [Result(bCanEnterTransition=Get bJustLostDriver())]
 }
 AnimationTransitionGraph_1 {
  path=AnimGraph.AnimGraphNode_StateMachine_0.OstrichVehicle.AnimStateNode_5.Locomotion.AnimGraphNode_StateMachine_0.LocomotionSub.AnimStateTransitionNode_6.AnimationTransitionGraph_1
  nodes [Result(bCanEnterTransition=Get bJustGotDriver())]
 }
 AnimationTransitionGraph_1 {
  path=AnimGraph.AnimGraphNode_StateMachine_0.OstrichVehicle.AnimStateNode_5.Locomotion.AnimGraphNode_StateMachine_0.LocomotionSub.AnimStateTransitionNode_7.AnimationTransitionGraph_1
  nodes [Result(bCanEnterTransition=NOT Boolean(A=Get bHasDriver()))]
 }
 AnimationTransitionGraph_1 {
  path=AnimGraph.AnimGraphNode_StateMachine_0.OstrichVehicle.AnimStateNode_5.Locomotion.AnimGraphNode_StateMachine_0.LocomotionSub.AnimStateTransitionNode_8.AnimationTransitionGraph_1
  nodes [Result(bCanEnterTransition=Get bPivotTransitionTriggered() -> (bPivotTransitionTriggered:bool))]
 }
 AnimationTransitionGraph_1 {
  path=AnimGraph.AnimGraphNode_StateMachine_0.OstrichVehicle.AnimStateNode_5.Locomotion.AnimGraphNode_StateMachine_0.LocomotionSub.AnimStateTransitionNode_9.AnimationTransitionGraph_1
  nodes [Result()]
 }
 JumpDashAbilities {
  entry=Default
  path=AnimGraph.AnimGraphNode_StateMachine_1.JumpDashAbilities
  ids=id
  nodes [
   AnimStateNode_0 = Default() {is=title AnimStateNode}
   AnimStateNode_1 = Stomp_InAir() {is=title AnimStateNode}
   AnimStateNode_10 = JumpChargingIntro() {is=title AnimStateNode}
   AnimStateNode_3 = JumpLand() {is=title AnimStateNode, bAlwaysResetOnEntry=true}
   AnimStateNode_5 = Jump() {is=title AnimStateNode, stateEntered {guid=#2, notifyName=JumpEnter}}
   AnimStateNode_6 = Dash_Into() {is=title AnimStateNode}
   AnimStateNode_7 = Dash_Loop() {is=title AnimStateNode}
   AnimStateNode_8 = Dash_Outro() {is=title AnimStateNode}
   AnimStateTransitionNode_0 = AnimStateTransitionNode() {crossfadeDuration=0.1}
   AnimStateTransitionNode_1 = AnimStateTransitionNode()
   AnimStateTransitionNode_10 = AnimStateTransitionNode() {crossfadeDuration=0.1}
   AnimStateTransitionNode_11 = AnimStateTransitionNode() {crossfadeDuration=0.08}
   AnimStateTransitionNode_12 = AnimStateTransitionNode() {crossfadeDuration=0.067}
   AnimStateTransitionNode_13 = AnimStateTransitionNode() {crossfadeDuration=0.08}
   AnimStateTransitionNode_14 = AnimStateTransitionNode() {bAutomaticRuleBasedOnSequencePlayerInState=true, crossfadeDuration=0.5}
   AnimStateTransitionNode_15 = AnimStateTransitionNode() {crossfadeDuration=0.1}
   AnimStateTransitionNode_16 = AnimStateTransitionNode() {crossfadeDuration=0.08}
   AnimStateTransitionNode_17 = AnimStateTransitionNode()
   AnimStateTransitionNode_18 = AnimStateTransitionNode()
   AnimStateTransitionNode_19 = AnimStateTransitionNode()
   AnimStateTransitionNode_2 = AnimStateTransitionNode() {crossfadeDuration=0}
   AnimStateTransitionNode_20 = AnimStateTransitionNode()
   AnimStateTransitionNode_21 = AnimStateTransitionNode()
   AnimStateTransitionNode_22 = AnimStateTransitionNode()
   AnimStateTransitionNode_23 = AnimStateTransitionNode() {priorityOrder=2}
   AnimStateTransitionNode_24 = AnimStateTransitionNode() {crossfadeDuration=0}
   AnimStateTransitionNode_25 = AnimStateTransitionNode()
   AnimStateTransitionNode_26 = AnimStateTransitionNode()
   AnimStateTransitionNode_27 = AnimStateTransitionNode()
   AnimStateTransitionNode_28 = AnimStateTransitionNode()
   AnimStateTransitionNode_29 = AnimStateTransitionNode()
   AnimStateTransitionNode_3 = AnimStateTransitionNode() {bAutomaticRuleBasedOnSequencePlayerInState=true, crossfadeDuration=0.5}
   AnimStateTransitionNode_30 = AnimStateTransitionNode() {crossfadeDuration=0.1}
   AnimStateTransitionNode_31 = AnimStateTransitionNode() {crossfadeDuration=0.3}
   AnimStateTransitionNode_32 = AnimStateTransitionNode() {crossfadeDuration=0.3}
   AnimStateTransitionNode_33 = AnimStateTransitionNode() {crossfadeDuration=0.45}
   AnimStateTransitionNode_34 = AnimStateTransitionNode() {priorityOrder=2}
   AnimStateTransitionNode_4 = AnimStateTransitionNode() {bAutomaticRuleBasedOnSequencePlayerInState=true, crossfadeDuration=0}
   AnimStateTransitionNode_5 = AnimStateTransitionNode()
   AnimStateTransitionNode_6 = AnimStateTransitionNode()
   AnimStateTransitionNode_7 = AnimStateTransitionNode() {crossfadeDuration=0.08, priorityOrder=2}
   AnimStateTransitionNode_8 = AnimStateTransitionNode() {crossfadeDuration=0.1}
   AnimStateTransitionNode_9 = AnimStateTransitionNode() {crossfadeDuration=0.08}
  ]
  links [
   {
    dst=Dash_Loop
    rule=AnimGraph.AnimGraphNode_StateMachine_1.JumpDashAbilities.AnimStateTransitionNode_29.Transition
    src=Dash_Into
    via=AnimStateTransitionNode_29
   }
   {
    dst=Dash_Loop
    rule=AnimGraph.AnimGraphNode_StateMachine_1.JumpDashAbilities.AnimStateTransitionNode_31.Transition
    src=Dash_Into
    via=AnimStateTransitionNode_31
   }
   {
    dst=Dash_Loop
    rule=AnimGraph.AnimGraphNode_StateMachine_1.JumpDashAbilities.AnimStateTransitionNode_4.AnimationTransitionGraph_0
    src=Dash_Into
    via=AnimStateTransitionNode_4
   }
   {
    dst=JumpChargingIntro
    rule=AnimGraph.AnimGraphNode_StateMachine_1.JumpDashAbilities.AnimStateTransitionNode_28.Transition
    src=Dash_Into
    via=AnimStateTransitionNode_28
   }
   {
    dst=Dash_Outro
    rule=AnimGraph.AnimGraphNode_StateMachine_1.JumpDashAbilities.AnimStateTransitionNode_32.Transition
    src=Dash_Loop
    via=AnimStateTransitionNode_32
   }
   {
    dst=Dash_Outro
    rule=AnimGraph.AnimGraphNode_StateMachine_1.JumpDashAbilities.AnimStateTransitionNode_7.AnimationTransitionGraph_1
    src=Dash_Loop
    via=AnimStateTransitionNode_7
   }
   {
    dst=Default
    rule=AnimGraph.AnimGraphNode_StateMachine_1.JumpDashAbilities.AnimStateTransitionNode_15.Transition
    src=Dash_Loop
    via=AnimStateTransitionNode_15
   }
   {
    dst=Default
    rule=AnimGraph.AnimGraphNode_StateMachine_1.JumpDashAbilities.AnimStateTransitionNode_6.Transition
    src=Dash_Loop
    via=AnimStateTransitionNode_6
   }
   {
    dst=Jump
    rule=AnimGraph.AnimGraphNode_StateMachine_1.JumpDashAbilities.AnimStateTransitionNode_17.Transition
    src=Dash_Loop
    via=AnimStateTransitionNode_17
   }
   {
    dst=Jump
    rule=AnimGraph.AnimGraphNode_StateMachine_1.JumpDashAbilities.AnimStateTransitionNode_25.Transition
    src=Dash_Loop
    via=AnimStateTransitionNode_25
   }
   {
    dst=JumpChargingIntro
    rule=AnimGraph.AnimGraphNode_StateMachine_1.JumpDashAbilities.AnimStateTransitionNode_18.Transition
    src=Dash_Loop
    via=AnimStateTransitionNode_18
   }
   {
    dst=Default
    rule=AnimGraph.AnimGraphNode_StateMachine_1.JumpDashAbilities.AnimStateTransitionNode_14.Transition
    src=Dash_Outro
    via=AnimStateTransitionNode_14
   }
   {
    dst=Default
    rule=AnimGraph.AnimGraphNode_StateMachine_1.JumpDashAbilities.AnimStateTransitionNode_21.Transition
    src=Dash_Outro
    via=AnimStateTransitionNode_21
   }
   {
    dst=Default
    rule=AnimGraph.AnimGraphNode_StateMachine_1.JumpDashAbilities.AnimStateTransitionNode_30.Transition
    src=Dash_Outro
    via=AnimStateTransitionNode_30
   }
   {
    dst=Default
    rule=AnimGraph.AnimGraphNode_StateMachine_1.JumpDashAbilities.AnimStateTransitionNode_8.Transition
    src=Dash_Outro
    via=AnimStateTransitionNode_8
   }
   {
    dst=Dash_Into
    rule=AnimGraph.AnimGraphNode_StateMachine_1.JumpDashAbilities.AnimStateTransitionNode_13.AnimationTransitionGraph_0
    src=Default
    via=AnimStateTransitionNode_13
   }
   {
    dst=Jump
    rule=AnimGraph.AnimGraphNode_StateMachine_1.JumpDashAbilities.AnimStateTransitionNode_1.Transition
    src=Default
    via=AnimStateTransitionNode_1
   }
   {
    dst=Jump
    rule=AnimGraph.AnimGraphNode_StateMachine_1.JumpDashAbilities.AnimStateTransitionNode_23.Transition
    src=Default
    via=AnimStateTransitionNode_23
   }
   {
    dst=Jump
    rule=AnimGraph.AnimGraphNode_StateMachine_1.JumpDashAbilities.AnimStateTransitionNode_5.Transition
    src=Default
    via=AnimStateTransitionNode_5
   }
   {
    dst=JumpChargingIntro
    rule=AnimGraph.AnimGraphNode_StateMachine_1.JumpDashAbilities.AnimStateTransitionNode_26.Transition
    src=Default
    via=AnimStateTransitionNode_26
   }
   {
    dst=Dash_Into
    rule=AnimGraph.AnimGraphNode_StateMachine_1.JumpDashAbilities.AnimStateTransitionNode_9.Transition
    src=Jump
    via=AnimStateTransitionNode_9
   }
   {
    dst=Default
    rule=AnimGraph.AnimGraphNode_StateMachine_1.JumpDashAbilities.AnimStateTransitionNode_10.Transition
    src=Jump
    via=AnimStateTransitionNode_10
   }
   {
    dst=Default
    rule=AnimGraph.AnimGraphNode_StateMachine_1.JumpDashAbilities.AnimStateTransitionNode_20.Transition
    src=Jump
    via=AnimStateTransitionNode_20
   }
   {
    dst=Default
    rule=AnimGraph.AnimGraphNode_StateMachine_1.JumpDashAbilities.AnimStateTransitionNode_22.Transition
    src=Jump
    via=AnimStateTransitionNode_22
   }
   {
    dst=Default
    rule=AnimGraph.AnimGraphNode_StateMachine_1.JumpDashAbilities.AnimStateTransitionNode_27.Transition
    src=Jump
    via=AnimStateTransitionNode_27
   }
   {
    dst=Stomp_InAir
    rule=AnimGraph.AnimGraphNode_StateMachine_1.JumpDashAbilities.AnimStateTransitionNode_0.Transition
    src=Jump
    via=AnimStateTransitionNode_0
   }
   {
    dst=Default
    rule=AnimGraph.AnimGraphNode_StateMachine_1.JumpDashAbilities.AnimStateTransitionNode_19.Transition
    src=JumpChargingIntro
    via=AnimStateTransitionNode_19
   }
   {
    dst=Default
    rule=AnimGraph.AnimGraphNode_StateMachine_1.JumpDashAbilities.AnimStateTransitionNode_34.Transition
    src=JumpChargingIntro
    via=AnimStateTransitionNode_34
   }
   {
    dst=Jump
    rule=AnimGraph.AnimGraphNode_StateMachine_1.JumpDashAbilities.AnimStateTransitionNode_12.Transition
    src=JumpChargingIntro
    via=AnimStateTransitionNode_12
   }
   {
    dst=Jump
    rule=AnimGraph.AnimGraphNode_StateMachine_1.JumpDashAbilities.AnimStateTransitionNode_11.Transition
    src=JumpLand
    via=AnimStateTransitionNode_11
   }
   {
    dst=Jump
    rule=AnimGraph.AnimGraphNode_StateMachine_1.JumpDashAbilities.AnimStateTransitionNode_24.Transition
    src=JumpLand
    via=AnimStateTransitionNode_24
   }
   {
    dst=Jump
    rule=AnimGraph.AnimGraphNode_StateMachine_1.JumpDashAbilities.AnimStateTransitionNode_3.Transition
    src=JumpLand
    via=AnimStateTransitionNode_3
   }
   {
    dst=Jump
    rule=AnimGraph.AnimGraphNode_StateMachine_1.JumpDashAbilities.AnimStateTransitionNode_33.Transition
    src=JumpLand
    via=AnimStateTransitionNode_33
   }
   {
    dst=Jump
    rule=AnimGraph.AnimGraphNode_StateMachine_1.JumpDashAbilities.AnimStateTransitionNode_16.Transition
    src=Stomp_InAir
    via=AnimStateTransitionNode_16
   }
   {
    dst=JumpLand
    rule=AnimGraph.AnimGraphNode_StateMachine_1.JumpDashAbilities.AnimStateTransitionNode_2.Transition
    src=Stomp_InAir
    via=AnimStateTransitionNode_2
   }
  ]
 }
 Default {
  path=AnimGraph.AnimGraphNode_StateMachine_1.JumpDashAbilities.AnimStateNode_0.Default
  nodes [Output Animation Pose(Result=Use cached pose 'LocomotionPose'())]
 }
 Stomp_InAir {
  path=AnimGraph.AnimGraphNode_StateMachine_1.JumpDashAbilities.AnimStateNode_1.Stomp_InAir
  nodes [Output Animation Pose(Result=Sequence Player 'OstrichVehicle_StompAttack_InAir'() {bLoopAnimation=false, sequence=/Game/Animation/Game/Vehicles/OstrichVehicle/OstrichVehicle_StompAttack_InAir})]
 }
 JumpChargingIntro {
  path=AnimGraph.AnimGraphNode_StateMachine_1.JumpDashAbilities.AnimStateNode_10.JumpChargingIntro
  nodes [
   AnimGraphNode_StateMachine_0 = JumpCharge() {is=title AnimGraphNode_StateMachine}
   Output Animation Pose(Result=@AnimGraphNode_StateMachine_0)
  ]
 }
 JumpCharge {
  entry=JumpChargeStart
  path=AnimGraph.AnimGraphNode_StateMachine_1.JumpDashAbilities.AnimStateNode_10.JumpChargingIntro.AnimGraphNode_StateMachine_0.JumpCharge
  ids=id
  nodes [
   AnimStateNode_0 = JumpChargeStart() {is=title AnimStateNode}
   AnimStateNode_1 = JumpChargeLoop() {is=title AnimStateNode}
   AnimStateTransitionNode_0 = AnimStateTransitionNode() {bAutomaticRuleBasedOnSequencePlayerInState=true, crossfadeDuration=0}
  ]
  links [{
    dst=JumpChargeLoop
    rule=AnimGraph.AnimGraphNode_StateMachine_1.JumpDashAbilities.AnimStateNode_10.JumpChargingIntro.AnimGraphNode_StateMachine_0.JumpCharge.AnimStateTransitionNode_0.Transition
    src=JumpChargeStart
    via=AnimStateTransitionNode_0
   }]
 }
 JumpChargeStart {
  path=AnimGraph.AnimGraphNode_StateMachine_1.JumpDashAbilities.AnimStateNode_10.JumpChargingIntro.AnimGraphNode_StateMachine_0.JumpCharge.AnimStateNode_0.JumpChargeStart
  nodes [
   n1 = "Sequence Player (additive) 'OstrichVehicle_JumpCharge_Into_WithGunner'"() {
    bLoopAnimation=false
    groupName=JumpCharge
    method=SyncGroup
    sequence=/Game/Animation/Game/Vehicles/OstrichVehicle/OstrichVehicle_JumpCharge_Into_WithGunner
   }
   Apply Additive(Additive=@n1, Alpha=Get GunnerAlphaValue()) {alphaBoolBlend {blendOption=Linear}}
   n2 = Sequence Player 'OstrichVehicle_JumpCharge_Into'() {
    bLoopAnimation=false
    groupName=JumpCharge
    method=SyncGroup
    sequence=/Game/Animation/Game/Vehicles/OstrichVehicle/OstrichVehicle_JumpCharge_Into
   }
   Output Animation Pose(Result=@n2)
  ]
 }
 JumpChargeLoop {
  path=AnimGraph.AnimGraphNode_StateMachine_1.JumpDashAbilities.AnimStateNode_10.JumpChargingIntro.AnimGraphNode_StateMachine_0.JumpCharge.AnimStateNode_1.JumpChargeLoop
  nodes [
   n1 = "Sequence Player (additive) 'OstrichVehicle_JumpCharge_Into_Loop_WithGunner'"() {
    groupName=JumpChargeLoop
    method=SyncGroup
    sequence=/Game/Animation/Game/Vehicles/OstrichVehicle/OstrichVehicle_JumpCharge_Into_Loop_WithGunner
   }
   Apply Additive(Additive=@n1, Alpha=Get GunnerAlphaValue()) {alphaBoolBlend {blendOption=Linear}}
   n2 = Sequence Player 'OstrichVehicle_JumpCharge_Into_Loop'() {groupName=JumpChargeLoop, method=SyncGroup, sequence=/Game/Animation/Game/Vehicles/OstrichVehicle/OstrichVehicle_JumpCharge_Into_Loop}
   Output Animation Pose(Result=@n2)
  ]
 }
 Transition {
  path=AnimGraph.AnimGraphNode_StateMachine_1.JumpDashAbilities.AnimStateNode_10.JumpChargingIntro.AnimGraphNode_StateMachine_0.JumpCharge.AnimStateTransitionNode_0.Transition
  nodes [Result()]
 }
 JumpLand {
  path=AnimGraph.AnimGraphNode_StateMachine_1.JumpDashAbilities.AnimStateNode_3.JumpLand
  nodes [
   n1 = Sequence Player 'OstrichVehicle_StompAttack_InAir_Land'() {bLoopAnimation=false, sequence=/Game/Animation/Game/Vehicles/OstrichVehicle/OstrichVehicle_StompAttack_InAir_Land, startPosition=0.067}
   Output Animation Pose(Result=@n1)
  ]
 }
 Jump {
  path=AnimGraph.AnimGraphNode_StateMachine_1.JumpDashAbilities.AnimStateNode_5.Jump
  nodes [
   AnimGraphNode_StateMachine_0 = JumpState() {is=title AnimGraphNode_StateMachine}
   Output Animation Pose(Result=@AnimGraphNode_StateMachine_0)
  ]
 }
 JumpState {
  entry=Default
  path=AnimGraph.AnimGraphNode_StateMachine_1.JumpDashAbilities.AnimStateNode_5.Jump.AnimGraphNode_StateMachine_0.JumpState
  ids=id
  nodes [
   AnimStateNode_0 = Default()
   AnimStateNode_1 = JumpApex() {is=title AnimStateNode}
   AnimStateNode_2 = JumpFall() {is=title AnimStateNode}
   AnimStateNode_3 = JumpLand()
   AnimStateNode_4 = JumpLandPredicted() {is=title AnimStateNode}
   AnimStateTransitionNode_0 = AnimStateTransitionNode() {crossfadeDuration=0}
   AnimStateTransitionNode_1 = AnimStateTransitionNode() {bAutomaticRuleBasedOnSequencePlayerInState=true, crossfadeDuration=0.05}
   AnimStateTransitionNode_10 = AnimStateTransitionNode() {priorityOrder=2}
   AnimStateTransitionNode_11 = AnimStateTransitionNode() {bAutomaticRuleBasedOnSequencePlayerInState=true, crossfadeDuration=0.5}
   AnimStateTransitionNode_17 = AnimStateTransitionNode()
   AnimStateTransitionNode_2 = AnimStateTransitionNode() {crossfadeDuration=0.3}
   AnimStateTransitionNode_3 = AnimStateTransitionNode() {crossfadeDuration=0}
   AnimStateTransitionNode_4 = AnimStateTransitionNode()
   AnimStateTransitionNode_5 = AnimStateTransitionNode() {crossfadeDuration=0}
   AnimStateTransitionNode_6 = AnimStateTransitionNode() {crossfadeDuration=0.45}
   AnimStateTransitionNode_9 = AnimStateTransitionNode() {crossfadeDuration=0.25}
  ]
  links [
   {
    dst=JumpApex
    rule=AnimGraph.AnimGraphNode_StateMachine_1.JumpDashAbilities.AnimStateNode_5.Jump.AnimGraphNode_StateMachine_0.JumpState.AnimStateTransitionNode_0.AnimationTransitionGraph_0
    src=Default
    via=AnimStateTransitionNode_0
   }
   {
    dst=JumpFall
    rule=AnimGraph.AnimGraphNode_StateMachine_1.JumpDashAbilities.AnimStateNode_5.Jump.AnimGraphNode_StateMachine_0.JumpState.AnimStateTransitionNode_17.AnimationTransitionGraph_0
    src=Default
    via=AnimStateTransitionNode_17
   }
   {
    dst=JumpFall
    rule=AnimGraph.AnimGraphNode_StateMachine_1.JumpDashAbilities.AnimStateNode_5.Jump.AnimGraphNode_StateMachine_0.JumpState.AnimStateTransitionNode_1.AnimationTransitionGraph_0
    src=JumpApex
    via=AnimStateTransitionNode_1
   }
   {
    dst=JumpLand
    rule=AnimGraph.AnimGraphNode_StateMachine_1.JumpDashAbilities.AnimStateNode_5.Jump.AnimGraphNode_StateMachine_0.JumpState.AnimStateTransitionNode_10.AnimationTransitionGraph_0
    src=JumpApex
    via=AnimStateTransitionNode_10
   }
   {
    dst=JumpLand
    rule=AnimGraph.AnimGraphNode_StateMachine_1.JumpDashAbilities.AnimStateNode_5.Jump.AnimGraphNode_StateMachine_0.JumpState.AnimStateTransitionNode_3.AnimationTransitionGraph_0
    src=JumpFall
    via=AnimStateTransitionNode_3
   }
   {
    dst=JumpLandPredicted
    rule=AnimGraph.AnimGraphNode_StateMachine_1.JumpDashAbilities.AnimStateNode_5.Jump.AnimGraphNode_StateMachine_0.JumpState.AnimStateTransitionNode_2.AnimationTransitionGraph_0
    src=JumpFall
    via=AnimStateTransitionNode_2
   }
   {
    dst=Default
    rule=AnimGraph.AnimGraphNode_StateMachine_1.JumpDashAbilities.AnimStateNode_5.Jump.AnimGraphNode_StateMachine_0.JumpState.AnimStateTransitionNode_11.AnimationTransitionGraph_0
    src=JumpLand
    via=AnimStateTransitionNode_11
   }
   {
    dst=Default
    rule=AnimGraph.AnimGraphNode_StateMachine_1.JumpDashAbilities.AnimStateNode_5.Jump.AnimGraphNode_StateMachine_0.JumpState.AnimStateTransitionNode_6.Transition
    src=JumpLand
    via=AnimStateTransitionNode_6
   }
   {
    dst=Default
    rule=AnimGraph.AnimGraphNode_StateMachine_1.JumpDashAbilities.AnimStateNode_5.Jump.AnimGraphNode_StateMachine_0.JumpState.AnimStateTransitionNode_9.AnimationTransitionGraph_0
    src=JumpLand
    via=AnimStateTransitionNode_9
   }
   {
    dst=Default
    rule=AnimGraph.AnimGraphNode_StateMachine_1.JumpDashAbilities.AnimStateNode_5.Jump.AnimGraphNode_StateMachine_0.JumpState.AnimStateTransitionNode_4.Transition
    src=JumpLandPredicted
    via=AnimStateTransitionNode_4
   }
   {
    dst=JumpLand
    rule=AnimGraph.AnimGraphNode_StateMachine_1.JumpDashAbilities.AnimStateNode_5.Jump.AnimGraphNode_StateMachine_0.JumpState.AnimStateTransitionNode_5.AnimationTransitionGraph_0
    src=JumpLandPredicted
    via=AnimStateTransitionNode_5
   }
  ]
 }
 Default {
  path=AnimGraph.AnimGraphNode_StateMachine_1.JumpDashAbilities.AnimStateNode_5.Jump.AnimGraphNode_StateMachine_0.JumpState.AnimStateNode_0.Default
  nodes [Output Animation Pose(Result=Use cached pose 'LocomotionPose'())]
 }
 JumpApex {
  path=AnimGraph.AnimGraphNode_StateMachine_1.JumpDashAbilities.AnimStateNode_5.Jump.AnimGraphNode_StateMachine_0.JumpState.AnimStateNode_1.JumpApex
  nodes [Output Animation Pose(Result=Sequence Player!(Sequence:Anim Sequence Base Object Reference=Get JumpBaseApexSeq() -> (JumpBaseApexSeq:Anim Sequence Object Reference)) -> (Pose) {bLoopAnimation=false})]
 }
 JumpFall {
  path=AnimGraph.AnimGraphNode_StateMachine_1.JumpDashAbilities.AnimStateNode_5.Jump.AnimGraphNode_StateMachine_0.JumpState.AnimStateNode_2.JumpFall
  nodes [
   n1 = Get JumpBaseFallSeq() -> (JumpBaseFallSeq:Anim Sequence Object Reference)
   Output Animation Pose(Result=Apply Additive(
     Base=Sequence Player(Sequence=@n1) {bLoopAnimation=false}
     Additive="Sequence Player (additive) 'OstrichVehicle_JumpApexLoop'"() {sequence=/Game/Animation/Game/Vehicles/OstrichVehicle/OstrichVehicle_JumpApexLoop}
     Alpha=Get JumpApexLoopAlphaValue() -> (JumpApexLoopAlphaValue:float)
    ) {alphaBoolBlend {blendOption=Linear}})
  ]
 }
 JumpLand {
  path=AnimGraph.AnimGraphNode_StateMachine_1.JumpDashAbilities.AnimStateNode_5.Jump.AnimGraphNode_StateMachine_0.JumpState.AnimStateNode_3.JumpLand
  nodes [Output Animation Pose(Result=Blendspace Player!(
     X=Get Speed2D() -> (Speed2D:float)
     Y=0
     BlendSpace:Blend Space Object Reference=Get JumpBaseLandSeq() -> (JumpBaseLandSeq:Blend Space 1D Object Reference)
    ) -> (Pose) {bLoop=false, groupName=Locomotion, method=SyncGroup})]
 }
 JumpLandPredicted {
  path=AnimGraph.AnimGraphNode_StateMachine_1.JumpDashAbilities.AnimStateNode_5.Jump.AnimGraphNode_StateMachine_0.JumpState.AnimStateNode_4.JumpLandPredicted
  nodes [Output Animation Pose(Result=Sequence Player(Sequence=Get JumpBaseLandPredictedSeq() -> (JumpBaseLandPredictedSeq:Anim Sequence Object Reference)) {bLoopAnimation=false, playRate=0.4})]
 }
 AnimationTransitionGraph_0 {
  path=AnimGraph.AnimGraphNode_StateMachine_1.JumpDashAbilities.AnimStateNode_5.Jump.AnimGraphNode_StateMachine_0.JumpState.AnimStateTransitionNode_0.AnimationTransitionGraph_0
  nodes [Result(bCanEnterTransition=Get bIsJumping())]
 }
 AnimationTransitionGraph_0 {
  path=AnimGraph.AnimGraphNode_StateMachine_1.JumpDashAbilities.AnimStateNode_5.Jump.AnimGraphNode_StateMachine_0.JumpState.AnimStateTransitionNode_1.AnimationTransitionGraph_0
  nodes [Result()]
 }
 AnimationTransitionGraph_0 {
  path=AnimGraph.AnimGraphNode_StateMachine_1.JumpDashAbilities.AnimStateNode_5.Jump.AnimGraphNode_StateMachine_0.JumpState.AnimStateTransitionNode_10.AnimationTransitionGraph_0
  nodes [Result(bCanEnterTransition=Get bIsOnGround() -> (bIsOnGround:bool))]
 }
 AnimationTransitionGraph_0 {
  path=AnimGraph.AnimGraphNode_StateMachine_1.JumpDashAbilities.AnimStateNode_5.Jump.AnimGraphNode_StateMachine_0.JumpState.AnimStateTransitionNode_11.AnimationTransitionGraph_0
  nodes [Result()]
 }
 AnimationTransitionGraph_0 {
  path=AnimGraph.AnimGraphNode_StateMachine_1.JumpDashAbilities.AnimStateNode_5.Jump.AnimGraphNode_StateMachine_0.JumpState.AnimStateTransitionNode_17.AnimationTransitionGraph_0
  nodes [Result(bCanEnterTransition=Get bIsFalling())]
 }
 AnimationTransitionGraph_0 {
  path=AnimGraph.AnimGraphNode_StateMachine_1.JumpDashAbilities.AnimStateNode_5.Jump.AnimGraphNode_StateMachine_0.JumpState.AnimStateTransitionNode_2.AnimationTransitionGraph_0
  nodes [Result(bCanEnterTransition=Get bLandingPredicted() -> (bLandingPredicted:bool))]
 }
 AnimationTransitionGraph_0 {
  path=AnimGraph.AnimGraphNode_StateMachine_1.JumpDashAbilities.AnimStateNode_5.Jump.AnimGraphNode_StateMachine_0.JumpState.AnimStateTransitionNode_3.AnimationTransitionGraph_0
  nodes [Result(bCanEnterTransition=Get bIsOnGround())]
 }
 Transition {
  path=AnimGraph.AnimGraphNode_StateMachine_1.JumpDashAbilities.AnimStateNode_5.Jump.AnimGraphNode_StateMachine_0.JumpState.AnimStateTransitionNode_4.Transition
  nodes [Result(bCanEnterTransition=Get bIsSkyTubing() -> (bIsSkyTubing:bool))]
 }
 AnimationTransitionGraph_0 {
  path=AnimGraph.AnimGraphNode_StateMachine_1.JumpDashAbilities.AnimStateNode_5.Jump.AnimGraphNode_StateMachine_0.JumpState.AnimStateTransitionNode_5.AnimationTransitionGraph_0
  nodes [Result(bCanEnterTransition=Get bIsOnGround())]
 }
 Transition {
  path=AnimGraph.AnimGraphNode_StateMachine_1.JumpDashAbilities.AnimStateNode_5.Jump.AnimGraphNode_StateMachine_0.JumpState.AnimStateTransitionNode_6.Transition
  nodes [Result(bCanEnterTransition=Get bShouldEarlyOutStompLanding() -> (bShouldEarlyOutStompLanding:bool))]
 }
 AnimationTransitionGraph_0 {
  path=AnimGraph.AnimGraphNode_StateMachine_1.JumpDashAbilities.AnimStateNode_5.Jump.AnimGraphNode_StateMachine_0.JumpState.AnimStateTransitionNode_9.AnimationTransitionGraph_0
  nodes [Result(bCanEnterTransition=Get bIsChargingJump() -> (bIsChargingJump:bool))]
 }
 Dash_Into {
  path=AnimGraph.AnimGraphNode_StateMachine_1.JumpDashAbilities.AnimStateNode_6.Dash_Into
  nodes [Output Animation Pose(Result=Blend Poses by bool(
     BlendPose_0=Sequence Player 'OstrichVehicle_Dash_N_Start_WithGunner'() {bLoopAnimation=false, sequence=/Game/Animation/Game/Vehicles/OstrichVehicle/OstrichVehicle_Dash_N_Start_WithGunner}
     BlendPose_1=Sequence Player 'OstrichVehicle_Dash_N_Start'() {bLoopAnimation=false, sequence=/Game/Animation/Game/Vehicles/OstrichVehicle/OstrichVehicle_Dash_N_Start}
     bActiveValue=Get bHasGunner()
    ) {blendType=Linear})]
 }
 Dash_Loop {
  path=AnimGraph.AnimGraphNode_StateMachine_1.JumpDashAbilities.AnimStateNode_7.Dash_Loop
  nodes [Output Animation Pose(Result=Blend Poses by bool(
     BlendPose_0=Sequence Player 'OstrichVehicle_Dash_Loop_WithGunner'() {sequence=/Game/Animation/Game/Vehicles/OstrichVehicle/OstrichVehicle_Dash_Loop_WithGunner}
     BlendPose_1=Sequence Player 'OstrichVehicle_Dash_Loop'() {sequence=/Game/Animation/Game/Vehicles/OstrichVehicle/OstrichVehicle_Dash_Loop}
     bActiveValue=Get bHasGunner()
    ) {blendType=Linear})]
 }
 Dash_Outro {
  path=AnimGraph.AnimGraphNode_StateMachine_1.JumpDashAbilities.AnimStateNode_8.Dash_Outro
  nodes [
   n1 = Get Speed2D()
   n2 = Blendspace Player 'OstrichVehicle_DashStop_BS'!(X=@n1) -> (Pose) {
    bLoop=false
    blendSpace=/Game/Animation/Game/Vehicles/OstrichVehicle/OstrichVehicle_DashStop_BS
    groupName=Locomotion
    groupRole=AlwaysLeader
    method=SyncGroup
   }
   n3 = Blendspace Player 'OstrichVehicle_DashStop_WithGunner_BS'!(X=@n1) -> (Pose) {
    blendSpace=/Game/Animation/Game/Vehicles/OstrichVehicle/OstrichVehicle_DashStop_WithGunner_BS
    groupName=Locomotion
    groupRole=AlwaysLeader
    method=SyncGroup
   }
   Output Animation Pose(Result=Blend Poses by bool(BlendPose_0=@n3, BlendPose_1=@n2, bActiveValue=Get bHasGunner()) {blendType=Linear})
  ]
 }
 Transition {
  path=AnimGraph.AnimGraphNode_StateMachine_1.JumpDashAbilities.AnimStateTransitionNode_0.Transition
  nodes [Result(bCanEnterTransition=Get bJustDidStompInAir() -> (bJustDidStompInAir:bool))]
 }
 Transition {
  path=AnimGraph.AnimGraphNode_StateMachine_1.JumpDashAbilities.AnimStateTransitionNode_1.Transition
  nodes [Result(bCanEnterTransition=Get bIsFallingWithDriver() -> (bIsFallingWithDriver:bool))]
 }
 Transition {
  path=AnimGraph.AnimGraphNode_StateMachine_1.JumpDashAbilities.AnimStateTransitionNode_10.Transition
  nodes [Result(bCanEnterTransition=Get bIsChargingJump())]
 }
 Transition {
  path=AnimGraph.AnimGraphNode_StateMachine_1.JumpDashAbilities.AnimStateTransitionNode_11.Transition
  nodes [Result(bCanEnterTransition=Get bIsDashing() -> (bIsDashing:bool))]
 }
 Transition {
  path=AnimGraph.AnimGraphNode_StateMachine_1.JumpDashAbilities.AnimStateTransitionNode_12.Transition
  nodes [Result(bCanEnterTransition=Get bIsJumping())]
 }
 AnimationTransitionGraph_0 {
  path=AnimGraph.AnimGraphNode_StateMachine_1.JumpDashAbilities.AnimStateTransitionNode_13.AnimationTransitionGraph_0
  nodes [Result(bCanEnterTransition=Get bIsDashing())]
 }
 Transition {path=AnimGraph.AnimGraphNode_StateMachine_1.JumpDashAbilities.AnimStateTransitionNode_14.Transition, nodes [Result()]}
 Transition {
  path=AnimGraph.AnimGraphNode_StateMachine_1.JumpDashAbilities.AnimStateTransitionNode_15.Transition
  nodes [Result(bCanEnterTransition=Get bIsChargingJump())]
 }
 Transition {
  path=AnimGraph.AnimGraphNode_StateMachine_1.JumpDashAbilities.AnimStateTransitionNode_16.Transition
  nodes [Result(bCanEnterTransition=Get bStartedDash() -> (bStartedDash:bool))]
 }
 Transition {
  path=AnimGraph.AnimGraphNode_StateMachine_1.JumpDashAbilities.AnimStateTransitionNode_17.Transition
  nodes [Result(bCanEnterTransition=Get bIsSkyTubing())]
 }
 Transition {
  path=AnimGraph.AnimGraphNode_StateMachine_1.JumpDashAbilities.AnimStateTransitionNode_18.Transition
  nodes [Result(bCanEnterTransition=Get bIsChargingJump())]
 }
 Transition {
  path=AnimGraph.AnimGraphNode_StateMachine_1.JumpDashAbilities.AnimStateTransitionNode_19.Transition
  nodes [Result(bCanEnterTransition=NOT Boolean(A=Get bHasDriver()))]
 }
 Transition {
  path=AnimGraph.AnimGraphNode_StateMachine_1.JumpDashAbilities.AnimStateTransitionNode_2.Transition
  nodes [Result(bCanEnterTransition=Get bIsOnGround())]
 }
 Transition {
  path=AnimGraph.AnimGraphNode_StateMachine_1.JumpDashAbilities.AnimStateTransitionNode_20.Transition
  nodes [Result(bCanEnterTransition=Get bStartedDash())]
 }
 Transition {
  path=AnimGraph.AnimGraphNode_StateMachine_1.JumpDashAbilities.AnimStateTransitionNode_21.Transition
  nodes [Result(bCanEnterTransition=Get bIsSkyTubing())]
 }
 Transition {
  path=AnimGraph.AnimGraphNode_StateMachine_1.JumpDashAbilities.AnimStateTransitionNode_22.Transition
  nodes [Result(bCanEnterTransition=Get bIsSkyTubing())]
 }
 Transition {
  path=AnimGraph.AnimGraphNode_StateMachine_1.JumpDashAbilities.AnimStateTransitionNode_23.Transition
  nodes [Result(bCanEnterTransition=Get bIsInAirWithDriver() -> (bIsInAirWithDriver:bool))]
 }
 Transition {
  path=AnimGraph.AnimGraphNode_StateMachine_1.JumpDashAbilities.AnimStateTransitionNode_24.Transition
  nodes [Result(bCanEnterTransition=Get bIsJumping())]
 }
 Transition {
  path=AnimGraph.AnimGraphNode_StateMachine_1.JumpDashAbilities.AnimStateTransitionNode_25.Transition
  nodes [Result(bCanEnterTransition=Get bLandingPredicted())]
 }
 Transition {
  path=AnimGraph.AnimGraphNode_StateMachine_1.JumpDashAbilities.AnimStateTransitionNode_26.Transition
  nodes [Result(bCanEnterTransition=Get bIsChargingJump())]
 }
 Transition {
  path=AnimGraph.AnimGraphNode_StateMachine_1.JumpDashAbilities.AnimStateTransitionNode_27.Transition
  nodes [Result(bCanEnterTransition=NOT Boolean(A=Get bHasDriver()))]
 }
 Transition {
  path=AnimGraph.AnimGraphNode_StateMachine_1.JumpDashAbilities.AnimStateTransitionNode_28.Transition
  nodes [Result(bCanEnterTransition=Get bIsChargingJump())]
 }
 Transition {
  path=AnimGraph.AnimGraphNode_StateMachine_1.JumpDashAbilities.AnimStateTransitionNode_29.Transition
  nodes [Result(bCanEnterTransition=Get bTransition_DashLoop_To_DashOutro() -> (bTransition_DashLoop_To_DashOutro:bool))]
 }
 Transition {path=AnimGraph.AnimGraphNode_StateMachine_1.JumpDashAbilities.AnimStateTransitionNode_3.Transition, nodes [Result()]}
 Transition {
  path=AnimGraph.AnimGraphNode_StateMachine_1.JumpDashAbilities.AnimStateTransitionNode_30.Transition
  nodes [Result(bCanEnterTransition=Get bIsChargingJump())]
 }
 Transition {
  path=AnimGraph.AnimGraphNode_StateMachine_1.JumpDashAbilities.AnimStateTransitionNode_31.Transition
  nodes [Result(bCanEnterTransition=Get bIsDashAnimDurationElapsed() -> (bIsDashAnimDurationElapsed:bool))]
 }
 Transition {
  path=AnimGraph.AnimGraphNode_StateMachine_1.JumpDashAbilities.AnimStateTransitionNode_32.Transition
  nodes [Result(bCanEnterTransition=Get bIsDashAnimDurationElapsed())]
 }
 Transition {
  path=AnimGraph.AnimGraphNode_StateMachine_1.JumpDashAbilities.AnimStateTransitionNode_33.Transition
  nodes [Result(bCanEnterTransition=Get bShouldEarlyOutStompLanding())]
 }
 Transition {
  path=AnimGraph.AnimGraphNode_StateMachine_1.JumpDashAbilities.AnimStateTransitionNode_34.Transition
  nodes [Result(bCanEnterTransition=NOT Boolean(A=Get bIsChargingJump()))]
 }
 AnimationTransitionGraph_0 {path=AnimGraph.AnimGraphNode_StateMachine_1.JumpDashAbilities.AnimStateTransitionNode_4.AnimationTransitionGraph_0, nodes [Result()]}
 Transition {
  path=AnimGraph.AnimGraphNode_StateMachine_1.JumpDashAbilities.AnimStateTransitionNode_5.Transition
  nodes [Result(bCanEnterTransition=Get bIsSkyTubingWithDriver() -> (bIsSkyTubingWithDriver:bool))]
 }
 Transition {
  path=AnimGraph.AnimGraphNode_StateMachine_1.JumpDashAbilities.AnimStateTransitionNode_6.Transition
  nodes [Result(bCanEnterTransition=Get bTransition_DashLoop_To_Default() -> (bTransition_DashLoop_To_Default:bool))]
 }
 AnimationTransitionGraph_1 {
  path=AnimGraph.AnimGraphNode_StateMachine_1.JumpDashAbilities.AnimStateTransitionNode_7.AnimationTransitionGraph_1
  nodes [Result(bCanEnterTransition=Get bTransition_DashLoop_To_DashOutro())]
 }
 Transition {
  path=AnimGraph.AnimGraphNode_StateMachine_1.JumpDashAbilities.AnimStateTransitionNode_8.Transition
  nodes [Result(bCanEnterTransition=NOT Boolean(A=Get bHasDriver()))]
 }
 Transition {
  path=AnimGraph.AnimGraphNode_StateMachine_1.JumpDashAbilities.AnimStateTransitionNode_9.Transition
  nodes [Result(bCanEnterTransition=Get bStartedDash())]
 }
 Gunner {
  entry=Default
  path=AnimGraph.AnimGraphNode_StateMachine_2.Gunner
  ids=id
  nodes [
   AnimStateNode_0 = Default()
   AnimStateNode_1 = GunnerBootUp() {is=title AnimStateNode}
   AnimStateNode_2 = IdleGunner() {is=title AnimStateNode}
   AnimStateNode_3 = GunnerShutDown() {is=title AnimStateNode}
   AnimStateTransitionNode_0 = AnimStateTransitionNode() {crossfadeDuration=0}
   AnimStateTransitionNode_1 = AnimStateTransitionNode() {crossfadeDuration=0}
   AnimStateTransitionNode_2 = AnimStateTransitionNode() {crossfadeDuration=0, priorityOrder=2}
   AnimStateTransitionNode_4 = AnimStateTransitionNode() {crossfadeDuration=0.067}
   AnimStateTransitionNode_5 = AnimStateTransitionNode() {bAutomaticRuleBasedOnSequencePlayerInState=true, crossfadeDuration=0}
   AnimStateTransitionNode_6 = AnimStateTransitionNode() {crossfadeDuration=0.067}
   AnimStateTransitionNode_7 = AnimStateTransitionNode() {bAutomaticRuleBasedOnSequencePlayerInState=true}
   AnimStateTransitionNode_8 = AnimStateTransitionNode() {crossfadeDuration=0, priorityOrder=2}
  ]
  links [
   {
    dst=GunnerBootUp
    rule=AnimGraph.AnimGraphNode_StateMachine_2.Gunner.AnimStateTransitionNode_4.AnimationTransitionGraph_0
    src=Default
    via=AnimStateTransitionNode_4
   }
   {
    dst=IdleGunner
    rule=AnimGraph.AnimGraphNode_StateMachine_2.Gunner.AnimStateTransitionNode_2.AnimationTransitionGraph_2
    src=Default
    via=AnimStateTransitionNode_2
   }
   {
    dst=IdleGunner
    rule=AnimGraph.AnimGraphNode_StateMachine_2.Gunner.AnimStateTransitionNode_1.AnimationTransitionGraph_2
    src=GunnerBootUp
    via=AnimStateTransitionNode_1
   }
   {
    dst=IdleGunner
    rule=AnimGraph.AnimGraphNode_StateMachine_2.Gunner.AnimStateTransitionNode_5.AnimationTransitionGraph_0
    src=GunnerBootUp
    via=AnimStateTransitionNode_5
   }
   {
    dst=Default
    rule=AnimGraph.AnimGraphNode_StateMachine_2.Gunner.AnimStateTransitionNode_0.AnimationTransitionGraph_2
    src=GunnerShutDown
    via=AnimStateTransitionNode_0
   }
   {
    dst=Default
    rule=AnimGraph.AnimGraphNode_StateMachine_2.Gunner.AnimStateTransitionNode_7.AnimationTransitionGraph_0
    src=GunnerShutDown
    via=AnimStateTransitionNode_7
   }
   {
    dst=Default
    rule=AnimGraph.AnimGraphNode_StateMachine_2.Gunner.AnimStateTransitionNode_8.AnimationTransitionGraph_0
    src=IdleGunner
    via=AnimStateTransitionNode_8
   }
   {
    dst=GunnerShutDown
    rule=AnimGraph.AnimGraphNode_StateMachine_2.Gunner.AnimStateTransitionNode_6.AnimationTransitionGraph_0
    src=IdleGunner
    via=AnimStateTransitionNode_6
   }
  ]
 }
 Default {
  path=AnimGraph.AnimGraphNode_StateMachine_2.Gunner.AnimStateNode_0.Default
  nodes [Output Animation Pose(Result=Use cached pose 'PostJumpPose'())]
 }
 GunnerBootUp {
  path=AnimGraph.AnimGraphNode_StateMachine_2.Gunner.AnimStateNode_1.GunnerBootUp
  nodes [Output Animation Pose(Result=Sequence Player 'OstrichVehicle_Gunner_BootUp'() {bLoopAnimation=false, sequence=/Game/Animation/Game/Vehicles/OstrichVehicle/OstrichVehicle_Gunner_BootUp})]
 }
 IdleGunner {
  path=AnimGraph.AnimGraphNode_StateMachine_2.Gunner.AnimStateNode_2.IdleGunner
  nodes [Output Animation Pose(Result=Sequence Player 'OstrichVehicle_Gunner_Idle_Pose'() {sequence=/Game/Animation/Game/Vehicles/OstrichVehicle/OstrichVehicle_Gunner_Idle_Pose})]
 }
 GunnerShutDown {
  path=AnimGraph.AnimGraphNode_StateMachine_2.Gunner.AnimStateNode_3.GunnerShutDown
  nodes [Output Animation Pose(Result=Sequence Player 'OstrichVehicle_Gunner_ShutDown'() {bLoopAnimation=false, sequence=/Game/Animation/Game/Vehicles/OstrichVehicle/OstrichVehicle_Gunner_ShutDown})]
 }
 AnimationTransitionGraph_2 {
  path=AnimGraph.AnimGraphNode_StateMachine_2.Gunner.AnimStateTransitionNode_0.AnimationTransitionGraph_2
  nodes [Result(bCanEnterTransition=Get bHasGunner())]
 }
 AnimationTransitionGraph_2 {
  path=AnimGraph.AnimGraphNode_StateMachine_2.Gunner.AnimStateTransitionNode_1.AnimationTransitionGraph_2
  nodes [Result(bCanEnterTransition=NOT Boolean(A=Get bHasGunner()))]
 }
 AnimationTransitionGraph_2 {
  path=AnimGraph.AnimGraphNode_StateMachine_2.Gunner.AnimStateTransitionNode_2.AnimationTransitionGraph_2
  nodes [Result(bCanEnterTransition=Get bHasGunner())]
 }
 AnimationTransitionGraph_0 {
  path=AnimGraph.AnimGraphNode_StateMachine_2.Gunner.AnimStateTransitionNode_4.AnimationTransitionGraph_0
  nodes [Result(bCanEnterTransition=Get bJustGotGunner() -> (bJustGotGunner:bool))]
 }
 AnimationTransitionGraph_0 {path=AnimGraph.AnimGraphNode_StateMachine_2.Gunner.AnimStateTransitionNode_5.AnimationTransitionGraph_0, nodes [Result()]}
 AnimationTransitionGraph_0 {
  path=AnimGraph.AnimGraphNode_StateMachine_2.Gunner.AnimStateTransitionNode_6.AnimationTransitionGraph_0
  nodes [Result(bCanEnterTransition=Get bJustGotGunner())]
 }
 AnimationTransitionGraph_0 {path=AnimGraph.AnimGraphNode_StateMachine_2.Gunner.AnimStateTransitionNode_7.AnimationTransitionGraph_0, nodes [Result()]}
 AnimationTransitionGraph_0 {
  path=AnimGraph.AnimGraphNode_StateMachine_2.Gunner.AnimStateTransitionNode_8.AnimationTransitionGraph_0
  nodes [Result(bCanEnterTransition=NOT Boolean(A=Get bHasGunner()))]
 }
 RocketRecoilLowerBody {
  entry=RocketFireLowerBody
  path=AnimGraph.AnimGraphNode_StateMachine_3.RocketRecoilLowerBody
  ids=id
  nodes [AnimStateNode_1 = RocketFireLowerBody() {is=title AnimStateNode, bAlwaysResetOnEntry=true}]
 }
 RocketFireLowerBody {
  path=AnimGraph.AnimGraphNode_StateMachine_3.RocketRecoilLowerBody.AnimStateNode_1.RocketFireLowerBody
  nodes [
   n1 = Get LocalAimYaw()
   Output Animation Pose(Result=Apply Additive(
     Base=Use cached pose 'PostGunnerIdleAdditivePose'()
     Additive=Blendspace Player 'OstrichVehicle_Rocket_Fire_LowerBody_BS'!(X=@n1) -> (Pose) {blendSpace=/Game/Animation/Game/Vehicles/OstrichVehicle/OstrichVehicle_Rocket_Fire_LowerBody_BS}
    ) {alphaBoolBlend {blendOption=Linear}})
  ]
 }
 RocketOnCoolDown {
  entry=Default
  path=AnimGraph.AnimGraphNode_StateMachine_4.RocketOnCoolDown
  ids=id
  nodes [
   AnimStateNode_0 = OnCoolDown() {is=title AnimStateNode, bAlwaysResetOnEntry=true}
   AnimStateNode_1 = Default()
   AnimStateNode_2 = CoolDownEnd() {is=title AnimStateNode}
   AnimStateTransitionNode_0 = AnimStateTransitionNode() {crossfadeDuration=0}
   AnimStateTransitionNode_2 = AnimStateTransitionNode() {crossfadeDuration=0}
   AnimStateTransitionNode_3 = AnimStateTransitionNode() {bAutomaticRuleBasedOnSequencePlayerInState=true, crossfadeDuration=0}
  ]
  links [
   {
    dst=Default
    rule=AnimGraph.AnimGraphNode_StateMachine_4.RocketOnCoolDown.AnimStateTransitionNode_3.Transition
    src=CoolDownEnd
    via=AnimStateTransitionNode_3
   }
   {
    dst=OnCoolDown
    rule=AnimGraph.AnimGraphNode_StateMachine_4.RocketOnCoolDown.AnimStateTransitionNode_0.Transition
    src=Default
    via=AnimStateTransitionNode_0
   }
   {
    dst=CoolDownEnd
    rule=AnimGraph.AnimGraphNode_StateMachine_4.RocketOnCoolDown.AnimStateTransitionNode_2.Transition
    src=OnCoolDown
    via=AnimStateTransitionNode_2
   }
  ]
 }
 OnCoolDown {
  path=AnimGraph.AnimGraphNode_StateMachine_4.RocketOnCoolDown.AnimStateNode_0.OnCoolDown
  nodes [Output Animation Pose(Result=Sequence Player 'OstrichVehicle_Rocket_CoolDown'() {bLoopAnimation=false, sequence=/Game/Animation/Game/Vehicles/OstrichVehicle/OstrichVehicle_Rocket_CoolDown})]
 }
 Default {
  path=AnimGraph.AnimGraphNode_StateMachine_4.RocketOnCoolDown.AnimStateNode_1.Default
  nodes [Output Animation Pose(Result=Use cached pose 'PostAdditivePose'())]
 }
 CoolDownEnd {
  path=AnimGraph.AnimGraphNode_StateMachine_4.RocketOnCoolDown.AnimStateNode_2.CoolDownEnd
  nodes [Output Animation Pose(Result=Sequence Player 'OstrichVehicle_Rocket_CoolDown_End'() {bLoopAnimation=false, sequence=/Game/Animation/Game/Vehicles/OstrichVehicle/OstrichVehicle_Rocket_CoolDown_End})]
 }
 Transition {
  path=AnimGraph.AnimGraphNode_StateMachine_4.RocketOnCoolDown.AnimStateTransitionNode_0.Transition
  nodes [Result(bCanEnterTransition=Get bIsOnCoolDown() -> (bIsOnCoolDown:bool))]
 }
 Transition {
  path=AnimGraph.AnimGraphNode_StateMachine_4.RocketOnCoolDown.AnimStateTransitionNode_2.Transition
  nodes [Result(bCanEnterTransition=NOT Boolean(A=Get bIsRocketOnCooldown() -> (bIsRocketOnCooldown:bool)))]
 }
 Transition {path=AnimGraph.AnimGraphNode_StateMachine_4.RocketOnCoolDown.AnimStateTransitionNode_3.Transition, nodes [Result()]}
 UpperBodyBounce {
  entry=BounceAdditive
  path=AnimGraph.AnimGraphNode_StateMachine_5.UpperBodyBounce
  ids=id
  nodes [AnimStateNode_0 = BounceAdditive() {is=title AnimStateNode, bAlwaysResetOnEntry=true}]
 }
 BounceAdditive {
  path=AnimGraph.AnimGraphNode_StateMachine_5.UpperBodyBounce.AnimStateNode_0.BounceAdditive
  nodes [
   n1 = "Sequence Player (additive) 'OstrichVehicle_Jog_N_WithGunner_Additive'"() {
    groupName=Locomotion
    groupRole=AlwaysFollower
    method=SyncGroup
    sequence=/Game/Animation/Game/Vehicles/OstrichVehicle/OstrichVehicle_Jog_N_WithGunner_Additive
   }
   Output Animation Pose(Result=Apply Additive(Base=Use cached pose 'PostModifyBonePose'(), Additive=@n1, Alpha=0.6) {alphaBoolBlend {blendOption=Linear}})
  ]
 }
 ShotgunRecoilLowerBody {
  entry=ShotgunFireLowerBody
  path=AnimGraph.AnimGraphNode_StateMachine_6.ShotgunRecoilLowerBody
  ids=id
  nodes [
   AnimStateNode_0 = ShotgunFireLowerBody_Loop() {is=title AnimStateNode, bAlwaysResetOnEntry=true}
   AnimStateNode_1 = ShotgunFireLowerBody() {is=title AnimStateNode, bAlwaysResetOnEntry=true}
   AnimStateTransitionNode_0 = AnimStateTransitionNode() {crossfadeDuration=0}
   AnimStateTransitionNode_1 = AnimStateTransitionNode() {crossfadeDuration=0}
  ]
  links [
   {
    dst=ShotgunFireLowerBody_Loop
    rule=AnimGraph.AnimGraphNode_StateMachine_6.ShotgunRecoilLowerBody.AnimStateTransitionNode_0.Transition
    src=ShotgunFireLowerBody
    via=AnimStateTransitionNode_0
   }
   {
    dst=ShotgunFireLowerBody
    rule=AnimGraph.AnimGraphNode_StateMachine_6.ShotgunRecoilLowerBody.AnimStateTransitionNode_1.Transition
    src=ShotgunFireLowerBody_Loop
    via=AnimStateTransitionNode_1
   }
  ]
 }
 ShotgunFireLowerBody_Loop {
  path=AnimGraph.AnimGraphNode_StateMachine_6.ShotgunRecoilLowerBody.AnimStateNode_0.ShotgunFireLowerBody_Loop
  nodes [
   n1 = Get LocalAimYaw()
   Output Animation Pose(Result=Apply Additive(
     Base=Use cached pose 'PostGunnerIdleAdditivePose'()
     Additive=Blendspace Player 'OstrichVehicle_Shotgun_Fire_LowerBody_BS'!(X=@n1) -> (Pose) {blendSpace=/Game/Animation/Game/Vehicles/OstrichVehicle/OstrichVehicle_Shotgun_Fire_LowerBody_BS}
     Alpha=Get LowerBodyRecoilAdditiveAlphaValue() -> (LowerBodyRecoilAdditiveAlphaValue:float)
    ) {alphaBoolBlend {blendOption=Linear}})
  ]
 }
 ShotgunFireLowerBody {
  path=AnimGraph.AnimGraphNode_StateMachine_6.ShotgunRecoilLowerBody.AnimStateNode_1.ShotgunFireLowerBody
  nodes [
   n1 = Get LocalAimYaw()
   Output Animation Pose(Result=Apply Additive(
     Base=Use cached pose 'PostGunnerIdleAdditivePose'()
     Additive=Blendspace Player 'OstrichVehicle_Shotgun_Fire_LowerBody_BS'(X=@n1) {blendSpace=&OstrichVehicle_Shotgun_Fire_LowerBody_BS}
     Alpha=Get LowerBodyRecoilAdditiveAlphaValue()
    ) {alphaBoolBlend {blendOption=Linear}})
  ]
 }
 Transition {
  path=AnimGraph.AnimGraphNode_StateMachine_6.ShotgunRecoilLowerBody.AnimStateTransitionNode_0.Transition
  nodes [Result(bCanEnterTransition=Get bIsMechBeginFired() -> (bIsMechBeginFired:bool))]
 }
 Transition {
  path=AnimGraph.AnimGraphNode_StateMachine_6.ShotgunRecoilLowerBody.AnimStateTransitionNode_1.Transition
  nodes [Result(bCanEnterTransition=NOT Boolean(A=Get bIsMechBeginFired()))]
 }
 ShoulderFinAdditive {
  entry=ShoulderFinReloadAdditive
  path=AnimGraph.AnimGraphNode_StateMachine_7.ShoulderFinAdditive
  ids=id
  nodes [AnimStateNode_0 = ShoulderFinReloadAdditive() {is=title AnimStateNode, bAlwaysResetOnEntry=true}]
 }
 ShoulderFinReloadAdditive {
  path=AnimGraph.AnimGraphNode_StateMachine_7.ShoulderFinAdditive.AnimStateNode_0.ShoulderFinReloadAdditive
  nodes [Output Animation Pose(Result="Sequence Player (additive) 'OstrichVehicle_Shotgun_Reload_Fin_Additive'"() {bLoopAnimation=false, sequence=/Game/Animation/Game/Vehicles/OstrichVehicle/OstrichVehicle_Shotgun_Reload_Fin_Additive})]
 }
 EventGraph {nodes [
   Event StopSeatTransition_DriverEnter() {is=title K2Node_Event}
   > Set DriverBootUp(DriverBootUp:bool=false) -> (Output_Get:bool)
   Event StartSeatTransition_PassengerEnter() {is=title K2Node_Event}
   > Set bPassengerBootUp(bPassengerBootUp:bool=true) -> (Output_Get:bool)
   Event StopSeatTransition_PassengerEnter() {is=title K2Node_Event}
   > Set bPassengerBootUp(bPassengerBootUp=false)
   Event StartSeatTransition_DriverEnter() {is=title K2Node_Event}
   > Set DriverBootUp(DriverBootUp=true)
  ]}
]
variables [DriverBootUp {type=boolean}]
