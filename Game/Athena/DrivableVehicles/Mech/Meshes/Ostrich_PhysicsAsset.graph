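# Physics asset for the Ostrich drivable mech, giving its skeleton simplified collision shapes.
# Capsules cover the arms, upper-arm twists, legs and spine, while boxes cover the head, toes, pelvis and clavicle fin bones.
# The pelvis and both fins use several rotated boxes to build compound shapes.
# No constraints join the bodies, so the asset defines collision volumes without a jointed ragdoll setup.
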PhysicsAsset /Game/Athena/DrivableVehicles/Mech/Meshes/Ostrich_PhysicsAsset
bodies {
 dyn_clav_fin_l [
  {shapeType=Box, extentX=35.5392, extentY=119.2955, extentZ=98.4109, center=<-8.4505 -21.0092 -1.108>, rotation=<rot -24.2639 -0.0255 1.6221>}
  {shapeType=Box, extentX=17.2542, extentY=94.9571, extentZ=45.144, center=<4.0925 -14.8042 44.5932>, rotation=<rot -12.6471 -3.2554 24.307>}
 ]
 dyn_clav_fin_r [
  {shapeType=Box, extentX=39.3839, extentY=119.7147, extentZ=103.2015, center=<10.4445 24.0846 2.4517>, rotation=<rot -21.2518 -2.6696 -0.5287>}
  {shapeType=Box, extentX=14.5059, extentY=90.6024, extentZ=37.8489, center=<-5.5709 13.454 -48.0985>, rotation=<rot -10.763 1.7205 21.3445>}
 ]
 head [{shapeType=Box, extentX=62.0625, extentY=145.1268, extentZ=135.6224, center=<-79.6331 5.7172 -0.0002>, rotation=<rot 0 -19.9991 0>}]
 lowerarm_l [{shapeType=Capsule, length=124.8525, radius=37.3549, center=<62.5358 -16.4422 6.9887>, rotation=<rot 0.4883 88 -84.4091>}]
 lowerarm_r [{shapeType=Capsule, length=128.9126, radius=37.7604, center=<-85.7442 7.5895 -14.3215>, rotation=<rot -0.0388 88.9999 -84.557>}]
 paw_calf_l [{shapeType=Capsule, length=147.4764, radius=31.1065, center=<84.7753 -0.1346 0.887>, rotation=<rot 0.5665 -86.9532 90.0299>}]
 paw_calf_r [{shapeType=Capsule, length=146.3734, radius=29.5065, center=<-88.4155 0.8929 -2.193>, rotation=<rot 89.1993 179.9975 179.9975>}]
 paw_foot_l [{shapeType=Capsule, length=75.441, radius=27.3496, center=<69.0915 -0.7763 2.1806>, rotation=<rot -86.1827 -180 180>}]
 paw_foot_r [{shapeType=Capsule, length=72.8102, radius=28.0037, center=<-66.234 -1.2056 -3.6105>, rotation=<rot 87.5688 -0.0001 -3.417>}]
 paw_thigh_l [{shapeType=Capsule, length=137.9364, radius=32.5516, center=<48.9156 7.0562 18.4945>, rotation=<rot 77.4496 1.1568 -2.0608>}]
 paw_thigh_r [{shapeType=Capsule, length=136.8926, radius=34.1346, center=<-51.4085 -1.4155 -18.8815>, rotation=<rot 77.797 0.0001 -2.5517>}]
 paw_toe_l [{shapeType=Box, extentX=132.6459, extentY=41.4299, extentZ=72.4673, center=<0.2033 -18.2006 0>, rotation=<rot 0 28.8613 0>}]
 paw_toe_r [{shapeType=Box, extentX=135.4207, extentY=39.2191, extentZ=72.1866, center=<-2.2094 13.237 5.4661>, rotation=<rot 0 27.9391 0>}]
 pelvis [
  {shapeType=Box, extentX=12.1875, extentY=120.25, extentZ=85.8934, center=<-22.9578 -8.1224 0>, rotation=<rot 0 24.998 0>}
  {shapeType=Box, extentX=44.7845, extentY=82.9859, extentZ=14, center=<-5.5516 36.1383 -32.9342>, rotation=<rot -1.4999 51.0467 -22.4592>}
  {shapeType=Box, extentX=47.1834, extentY=83.7245, extentZ=14, center=<-6.924 35.4632 32.1731>, rotation=<rot -4.3812 47.515 21.8188>}
  {shapeType=Box, extentX=41.125, extentY=7.5, extentZ=38.875, center=<-33.318 61.0798 0.0001>, rotation=<rot 0 33.3704 0>}
 ]
 spine_03 [{shapeType=Capsule, length=60.893, radius=35.3985, center=<-7.781 -41.1727 0.9047>, rotation=<rot -0.3793 -79.0689 -88.0359>}]
 spine_05 [{shapeType=Capsule, length=74.758, radius=48.2972, center=<-24.4818 68.3261 -0.0004>, rotation=<rot 0 61.2179 -89.9996>}]
 upperarm_l [{shapeType=Capsule, length=47.2307, radius=28.859, center=<45.8015 -1.9409 4.8607>, rotation=<rot 0.0331 84.5015 -92.9811>}]
 upperarm_r [{shapeType=Capsule, length=48.0044, radius=28.0853, center=<-51.2526 2.0124 -4.6593>, rotation=<rot 0.0331 84.5018 -91.4297>}]
 upperarm_twist_01_l [{shapeType=Capsule, length=34, radius=26, center=<5.525 -0.731 34.1597>, rotation=<rot 0.6994 -12.7203 110.0846>}]
 upperarm_twist_01_r [{shapeType=Capsule, length=44.7631, radius=26, center=<-9.901 -5.1262 -25.6552>, rotation=<rot 2.0721 -7.7658 122.6218>}]
}
constraints []
