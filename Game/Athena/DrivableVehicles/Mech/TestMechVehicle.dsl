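; TestMechVehicle is the Blueprint for the drivable two-seat Ostrich mech, with driver kick, dash, charged jump, slam and horn.
; The server spawns an attached shield bubble, which supports overcharge and gunner-shield damage.
; Landings, stomps and dashes use traces, overlaps or hits to damage buildings and damage or launch pawns and vehicles.
; Gameplay cues, camera shakes, rumble and audio provide feedback, including for health-based damage states and timed detonation.
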(event EventK2_ApplyCosmeticWrap
  (params (LoadedWrap :type Athena%20Item%20Wrap%20Definition%20Object%20Reference))
  ;[1 1 1] Apply Cosmetic Wraps
  (for (_array_element _array_index) (Variables|/Script/Engine.SkeletalMesh|GetMaterials (Variables|/Script/Engine.SkinnedMeshComponent|GetSkeletalMesh (Variables|Character|GetMesh)))
    (if (not (== _array_index 5))
      (Class|FortVehicleInterface|ApplyMaterialandUseWrapifPossible (Utilities|Struct|BreakSkeletalMaterial _array_element) _array_index (delegate Custom|OnMIDCreated)))))

(event Custom|OnMIDCreated
  (params (MID :type Material%20Instance%20Dynamic%20Object%20Reference))
  ;[1 1 1] Apply Cosmetic Wraps
  (Variables|MechVehicle|SetBodyMID MID))

(event EventGunnerShieldTakeDamage
  (params (Damage :type Float%20%28single-precision%29))
  (bind _shieldbubbleactor (Variables|Components|GetShieldBubbleActor))
  ;[1 1 1] EVENT -> Gunner Shield Take Damage
  (Utilities|IsValid _shieldbubbleactor
    (:Is%20Valid
      (Environment|ApplyGlobalEnvironmentDamagetoActor _shieldbubbleactor Damage))))

(event Custom|EventOverchargeShieldActivated
  (bind _shieldbubbleactor (Variables|Components|GetShieldBubbleActor))
  ;[1 1 1] EVENT -> On Press Shield Boost
  (if (> (ScalableFloat|GetValueatLevel (Variables|HotfixableValues|GetShouldAllowOvercharge)) 0.0)
    (Utilities|IsValid _shieldbubbleactor
      (:Is%20Valid
        (Variables|/Game/Athena/DrivableVehicles/Mech/BGA_Ostrich_ShieldBubble.BGA_Ostrich_ShieldBubble_C|SetIsOvercharged true _shieldbubbleactor) ; Trigger Overcharge Event
      ))))

(event EventOnPressShieldBoost
  ;[1 1 1] EVENT -> On Press Shield Boost
  (CallFunction|EventOverchargeShieldActivated))

(event EventBeginPlay
  ;[1 1 1] EVENT -> BeginPlay
  (Utilities|FlowControl|Sequence
    (:then_0
      ;;[0 2 0.0137] SPAWN SHIELD BUBBLE
      (if (> (ScalableFloat|GetValueatLevel (Variables|HotfixableValues|GetShouldAllowShield)) 0.0)
        (if (Networking|IsServer)
          (bind _returnvalue (Building|SpawnBuildingGameplayActor "/Game/Athena/DrivableVehicles/Mech/BGA_Ostrich_ShieldBubble.BGA_Ostrich_ShieldBubble_C" (Transformation|GetActorTransform) self))
          (bind _asbga_ostrich_shield_bubble (Utilities|Casting|CastToBGA_Ostrich_ShieldBubble _returnvalue))
          (bind _output_get (Variables|Components|SetShieldBubbleActor _asbga_ostrich_shield_bubble))
          (Transformation|AttachActorToComponent _output_get (Variables|Character|GetMesh) "spine_05" "SnapToTarget" "SnapToTarget" "SnapToTarget" false)
          (Variables|/Game/Athena/DrivableVehicles/Mech/BGA_Ostrich_ShieldBubble.BGA_Ostrich_ShieldBubble_C|SetMech self _output_get)
          (Networking|ForceNetUpdate)
          (Gameplay|SetShieldActor _output_get))))
    (:then_1
      ;;[1 1 1] Bind to footsteps for servo "release" sounds
      (EventDispatchers|BindEventtoOnFootstepEvent (EventDispatchers|CreateEvent :!event Custom|On%20Mech%20Footstep))
      ;;[1 1 1] Bind on reached Jump Apex
      (Character|AssignOnReachedJumpApex (delegate Custom|EventOnReachedJumpApex))
      ;;[1 1 1] Bind On health changed
      (Health|BindEventtoOnHealthChanged (EventDispatchers|CreateEvent :!event Custom|EventOnHealthChanged))
      ;;[1 1 1] Make sure shield bubble VFX actor is removed when Destroyed
      (Game|BindEventtoOnDestroyed (delegate Custom|Event%20On%20Vehicle%20Destroyed)))))

(event Custom|EventOnReachedJumpApex
  ;[1 1 1] EVENT -> BeginPlay
  ;;[1 1 1] Bind on reached Jump Apex
  (Variables|MechVehicle|SetSetJumpApex true))

(event Custom|Event%20On%20Vehicle%20Destroyed
  (params (DestroyedActor :type Actor%20Object%20Reference))
  (bind _shieldbubbleactor (Variables|Components|GetShieldBubbleActor))
  ;[1 1 1] EVENT -> BeginPlay
  ;;[1 1 1] Make sure shield bubble VFX actor is removed when Destroyed
  (Utilities|IsValid _shieldbubbleactor
    (:Is%20Valid
      (Actor|DestroyActor _shieldbubbleactor)
      (Rendering|SetActorHiddenInGame _shieldbubbleactor true)
      (Networking|ForceNetUpdate _shieldbubbleactor))))

(event Custom|On%20Mech%20Footstep
  (params (FootIndex :type Integer))
  ;[0 0 0] Audio Land
  ;;[1 1 1] Servo release sounds after every step
  (Fortnite|PlayLocalForceFeedbackatLocation (Variables|ForceFeedback|GetFootstepForceFeedback) (Transformation|GetActorLocation self) 1e+02) ; Gamepad Vibrate
  (Audio|Components|Audio|Play (Variables|Default|GetFootstepServo) 0.0)
  (Utilities|FlowControl|RetriggerableDelay 0.25)
  (if (<= (Math|Vector|VectorLength (Variables|/Script/Engine.MovementComponent|GetVelocity (Variables|Character|GetCharacterMovement))) 2.5e+02)
    (if (Audio|Components|Audio|IsPlaying (Variables|Default|GetFootstepServo))
      (Audio|Components|Audio|FadeOut (Variables|Default|GetFootstepServo) 0.25 0.0 "Linear" :!class /Script/Engine.AudioComponent))))

(event EventOnPawnEnterVehicle
  (params (PlayerPawn :type Fort%20Player%20Pawn%20Object%20Reference) (SeatIdx :type Integer))
  (bind _returnvalue (Ability|GetFortAbilitySystemComponent PlayerPawn))
  ;[1 1 1] EVENT -> Pawn Enter Vehicle
  (Utilities|SetPawnRefs
    (:Out
      (Utilities|GCPlayerEnter SeatIdx)
      (CallFunction|HideQuickbar true PlayerPawn)
      (Utilities|IsValid _returnvalue
        (:Is%20Valid
          (Tags|BlockAbilitiesWithTags _returnvalue (Variables|Default|GetPlayerAbilitiesToBlock)))))))

(event Custom|HideQuickbar
  (params (Hide :type Boolean) (Actor :type Fort%20Player%20Pawn%20Object%20Reference))
  ;[1 1 1] Hide Quickbar
  (Utilities|IsValid Actor
    (:Is%20Valid
      (bind _asfort_player_controller_gameplay (Utilities|Casting|CastToFortPlayerControllerGameplay (Fortnite|GetFortPlayerControllerfromActor Actor)))
      (HUD|SetHUDElementVisibilitywithContext _asfort_player_controller_gameplay (GameplayTags|MakeGameplayTagContainerfromTag "(TagName=\"HUD.Athena.HideforVehicle\")") Hide "OstrichEnter" "")
      (if Hide
        (EventDispatchers|AssignOnDied Actor (delegate Custom|EventOnDied))
        (EventDispatchers|AssignOnPlayerStartDBNO Actor (delegate Custom|EventOnPlayerStartDBNO))
        (else
          (EventDispatchers|UnbindEventfromOnDied Actor (delegate Custom|EventOnDied))
          (EventDispatchers|UnbindEventfromOnPlayerStartDBNO Actor (delegate Custom|EventOnPlayerStartDBNO)))))))

(event Custom|EventOnPlayerStartDBNO
  ;[1 1 1] Hide Quickbar
  (CallFunction|HideQuickbar false))

(event Custom|EventOnDied
  (params (DamagedActor :type Actor%20Object%20Reference) (Damage :type Float%20%28single-precision%29) (InstigatedBy :type Controller%20Object%20Reference) (DamageCauser :type Actor%20Object%20Reference) (HitLocation :type Vector) (FHitComponent :type Primitive%20Component%20Object%20Reference) (BoneName :type Name) (Momentum :type Vector))
  ;[1 1 1] Hide Quickbar
  (CallFunction|HideQuickbar false))

(event Custom|On%20Audio%20Tick)

(event EventOnPawnExitVehicle
  (params (PlayerPawn :type Fort%20Player%20Pawn%20Object%20Reference) (ExitSocketName :type Name :default "None"))
  (bind _returnvalue (Ability|GetFortAbilitySystemComponent PlayerPawn))
  ;[1 1 1] EVENT -> Pawn Exit Vehicle
  (Utilities|SetPawnRefs
    (:Out
      (CallFunction|HideQuickbar false PlayerPawn)
      (CallFunction|SetRecentlyExited PlayerPawn ExitSocketName)
      (Utilities|IsValid _returnvalue
        (:Is%20Valid
          (Tags|UnBlockAbilitiesWithTags _returnvalue (Variables|Default|GetPlayerAbilitiesToBlock) false))))))

(event EventOnHoldExitStarted
  (params (ExitingPawn :type Fort%20Pawn%20Object%20Reference) (ExitDuration :type Float%20%28single-precision%29))
  ;[1 1 1] EVENT -> OnHoldExitStarted/Stopped
  (Ability|SendGameplayEventtoActor ExitingPawn "(TagName=\"Athena.Vehicle.Biplane.OnHoldExitStarted\")" (Utilities|Struct|MakeGameplayEventData :ContextHandle "()" :InstigatorTags "(GameplayTags=)" :TargetTags "(GameplayTags=)" :EventMagnitude ExitDuration :TargetData "()")))

(event Custom|SetRecentlyExited
  (params (Player :type Fort%20Player%20Pawn%20Object%20Reference) (ExitSocket :type Name))
  ;[1 1 1] Set Recently Exited Driver and/or Gunner
  (if (Utilities|Name|Equal(Name) ExitSocket "None")
    (Variables|Components|SetRecentlyExitedGunner Player)
    (Utilities|FlowControl|Delay 0.75)
    (Variables|Components|SetRecentlyExitedGunner 0)
    (else
      (Variables|Components|SetRecentlyExitedDriver Player)
      (Utilities|FlowControl|Delay 0.75)
      (Variables|Components|SetRecentlyExitedDriver 0))))

(event EventOnHoldExitStopped
  (params (ExitingPawn :type Fort%20Pawn%20Object%20Reference))
  ;[1 1 1] EVENT -> OnHoldExitStarted/Stopped
  (Ability|SendGameplayEventtoActor ExitingPawn "(TagName=\"Athena.Vehicle.Biplane.OnHoldExitStopped\")"))

(event EventOnLanded
  (params (Hit_bBlockingHit :type Boolean) (Hit_bInitialOverlap :type Boolean) (Hit_Time :type Float%20%28single-precision%29) (Hit_Distance :type Float%20%28single-precision%29) (Hit_Location :type Vector :default "0, 0, 0") (Hit_ImpactPoint :type Vector :default "0, 0, 0") (Hit_Normal :type Vector :default "0, 0, 0") (Hit_ImpactNormal :type Vector :default "0, 0, 0") (Hit_PhysMat :type Physical%20Material%20Object%20Reference) (Hit_HitActor :type Actor%20Object%20Reference) (Hit_HitComponent :type Primitive%20Component%20Object%20Reference) (Hit_HitBoneName :type Name :default "None") (Hit_BoneName :type Name :default "None") (Hit_HitItem :type Integer) (Hit_ElementIndex :type Integer) (Hit_FaceIndex :type Integer) (Hit_TraceStart :type Vector :default "0, 0, 0") (Hit_TraceEnd :type Vector :default "0, 0, 0"))
  (bind (_returnvalue_x _returnvalue_y _returnvalue_z) (Math|Vector|BreakVector (Transformation|GetActorLocation)))
  (bind (_jumpapexlocation_x _jumpapexlocation_y _jumpapexlocation_z) (Math|Vector|BreakVector (Variables|GetJumpApexLocation_X)))
  ;[1 1 1] EVENT -> On Landed
  (Projectile|RemoveGameplayCue%28Looping%29Local self (Variables|GameplayCue|GetGC_JumpInAirLooping) (CallFunction|GCParams)) ; Remove Jump Trails GC
  (Utilities|FlowControl|Sequence
    (:then_0
      (if (and (not (Vortex|IsinVortex self)) (>= (- _jumpapexlocation_z _returnvalue_z) (Variables|MechVehicle|GetMinHeightToJumpSlam)))
        (GameplayCue|ExecuteGameplayCue%28Burst%29Local self (Variables|GameplayCue|GetGC_JumpLanded) (CallFunction|GCParams))
        (Effects|SpawnEmitterAttached :EmitterTemplate (Variables|FX|GetP_JumpCompression) :AttachToComponent (Variables|Character|GetMesh) :AttachPointName "calf_l" :Scale "1.000000,1.000000,1.000000" :LocationType "SnapToTarget" :bAutoDestroy true :PoolingMethod "None" :bAutoActivate true)
        (GameplayCue|ExecuteGameplayCue%28Burst%29Local self (Variables|FX|GetGroundShakeBigGC) (CallFunction|GCParams)) ; Camera shake + controller vibration
        (CallFunction|DamageonLanded)
        (CallFunction|ResetLandedVariables)
        (else
          (CallFunction|ResetLandedVariables)
          (GameplayCue|ExecuteGameplayCue%28Burst%29Local self (Variables|FX|GetGroundShakeGC) (CallFunction|GCParams)) ; Camera shake + controller vibration
          (GameplayCue|ExecuteGameplayCue%28Burst%29Local self (Variables|GameplayCue|GetGC_JumpLanded_Small) (CallFunction|GCParams)) ; Camera shake + controller vibration
        )))
    (:then_1
      (bind _asfort_player_controller (Utilities|Casting|CastToFortPlayerController (Pawn|GetController)))
      (CameraAnimation|PlayCameraAnimation (CameraAnimation|ConvCameraAnimationCameraModifier (Variables|/Script/Engine.PlayerController|GetPlayerCameraManager _asfort_player_controller)) (Variables|CameraShakes|GetLandCameraShake) (Utilities|Struct|MakeCameraAnimationParams 1.0 (Variables|CameraShakes|GetCameraShakeLandScale)) :!class /Script/EngineCameras.CameraAnimationCameraModifier) ; Local mech pilot only
    )))

(event WaterBody|EventOnEnteredWaterBody
  (params (WaterBody :type Fort%20Water%20Body%20Actor%20Object%20Reference) (bIsFirstBody :type Boolean))
  ;[1 1 1] Water FX Handling
  (Default|OnEnterWater WaterBody)
  (if (not (Networking|IsServer))
    (Utilities|Time|SetTimerbyEvent (delegate Custom|CheckWater) 0.2 true)))

(event Custom|CheckWater
  ;[1 1 1] Water FX Handling
  (CallFunction|OnUpdateWater))

(event Custom|ResetLandedVariables
  ;[1 1 1] EVENT -> On Landed
  (Utilities|FlowControl|Delay 0.3) ; So location isn't reset before damage is applied
  (Variables|MechVehicle|SetSetJumpApex false))

(event WaterBody|EventOnExitedWaterBody
  (params (WaterBody :type Fort%20Water%20Body%20Actor%20Object%20Reference) (bIsLastBody :type Boolean))
  ;[1 1 1] Water FX Handling
  (CallFunction|OnExitWater)
  (Utilities|Time|ClearandInvalidateTimerbyHandle (Variables|Default|GetWaterCheckTimer)))

(event Collision|EventHit
  (params (MyComp :type Primitive%20Component%20Object%20Reference) (Other :type Actor%20Object%20Reference) (OtherComp :type Primitive%20Component%20Object%20Reference) (bSelfMoved :type Boolean) (HitLocation :type Vector :default "0, 0, 0") (HitNormal :type Vector :default "0, 0, 0") (NormalImpulse :type Vector :default "0, 0, 0") (Hit_bBlockingHit :type Boolean) (Hit_bInitialOverlap :type Boolean) (Hit_Time :type Float%20%28single-precision%29) (Hit_Distance :type Float%20%28single-precision%29) (Hit_Location :type Vector :default "0, 0, 0") (Hit_ImpactPoint :type Vector :default "0, 0, 0") (Hit_Normal :type Vector :default "0, 0, 0") (Hit_ImpactNormal :type Vector :default "0, 0, 0") (Hit_PhysMat :type Physical%20Material%20Object%20Reference) (Hit_HitActor :type Actor%20Object%20Reference) (Hit_HitComponent :type Primitive%20Component%20Object%20Reference) (Hit_HitBoneName :type Name :default "None") (Hit_BoneName :type Name :default "None") (Hit_HitItem :type Integer) (Hit_ElementIndex :type Integer) (Hit_FaceIndex :type Integer) (Hit_TraceStart :type Vector :default "0, 0, 0") (Hit_TraceEnd :type Vector :default "0, 0, 0"))
  ;[1 1 1] EVENT -> OnHit - Destroy Building Prop
  (CallFunction|DashDamageBuilding Hit_HitActor)
  (bind _asbuilding_prop (Utilities|Casting|CastToBuildingProp Hit_HitActor)
    (:then
      (if (CallFunction|CanDamageBuilding _asbuilding_prop 1.5e+02)
        (GameplayEffects|ApplyGameplayEffectToTarget :self (Ability|GetAbilitySystemComponent (Variables|Components|GetDriverPawn)) :GameplayEffectClass (Variables|DamageGameplayEffects|GetGE_DestroyBuildingDamage) :Target (Ability|GetAbilitySystemComponent (.Building_Actor (CallFunction|CanDamageBuilding _asbuilding_prop 1.5e+02))) :Level 1.0)))
    (:CastFailed
      (bind _asfort_player_pawn (Utilities|Casting|CastToFortPlayerPawn Hit_HitActor))
      (if (Math|Boolean|ANDBoolean (Gameplay|GetIsDashing) (not (Utilities|Array|ContainsItem (Variables|MechVehicle|GetHitPlayers) _asfort_player_pawn)) (not (.bSuccess (Utilities|Casting|CastToTestMechVehicle _asfort_player_pawn))))
        (Utilities|Array|Add (Variables|MechVehicle|GetHitPlayers) _asfort_player_pawn)
        (Utilities|LaunchPlayerPawn (Variables|Components|GetDriverPawn) _asfort_player_pawn (CallFunction|KnockAwayDirection _asfort_player_pawn (Variables|MechVehicle|GetDashKnockStrength) (Variables|MechVehicle|GetDashZLaunchAdditive)) true true false true false true true)
        (Utilities|FlowControl|Delay 1.0) ; Delay so we don't knock same person a bunch of times
        (Utilities|Array|Clear (Variables|MechVehicle|GetHitPlayers))))))

(event Custom|DamageOnLanded
  ;[1 1 1] On Landed, damage all nearby actors
  (bind _outactors (Collision|SphereOverlapActors (Transformation|GetActorLocation) (Variables|MechVehicle|GetLandedDamageRadius) (Utilities|Array|MakeArray "ObjectTypeQuery1" "ObjectTypeQuery3" "ObjectTypeQuery5" "ObjectTypeQuery2") "/Script/Engine.Actor" (Utilities|Array|MakeArray self (Variables|Components|GetDriverPawn) (Variables|Components|GetPassengerPawn))))
  (for _array_element _outactors
    (bind _asbuilding_actor (Utilities|Casting|CastToBuildingActor _array_element)
      (:then
        (if (CallFunction|CanDamageBuilding _asbuilding_actor 2e+03)
          (GameplayEffects|ApplyGameplayEffectToTarget :self (Ability|GetAbilitySystemComponent (Variables|Components|GetDriverPawn)) :GameplayEffectClass (Variables|DamageGameplayEffects|GetGE_BuildingDamageOnLanded) :Target (Ability|GetAbilitySystemComponent (.Building_Actor (CallFunction|CanDamageBuilding _asbuilding_actor 2e+03))) :Level 1.0)))
      (:CastFailed
        (CallFunction|DamageandLaunchActors _array_element))))
  (Fortnite|PlayLocalForceFeedbackatLocationMulti "/Game/Weapons/ForceFeedback/FF_Generic_Impact_Near.FF_Generic_Impact_Near" "/Game/Weapons/ForceFeedback/FF_Generic_Impact_Far.FF_Generic_Impact_Far" (Transformation|GetActorLocation self) 1e+02 8e+02 "LandedHighDistance") ; Gamepad Vibrate
)

(event EventOnDriverPrimaryFire
  ;[1 1 1] EVENT -> Driver Primary Fire - KICK!
  (if (Gameplay|GetIsChargingJump)
    _
    (elif (Vehicle|VehicleIsinAir)
      (CallFunction|JumpSlam)
      (else
        (Animation|Montage|PlayMontage (Mesh|GetMechMesh) (Variables|Animation|GetKickAttackMontage)
          (:then
            (Gameplay|OnKickStomp)
            (Variables|MechVehicle|SetKicking true) ; So we can block dashing if we are kicking
            (Utilities|Time|SetTimerbyEvent (delegate Custom|Stomp) 0.74) ; Need this so damage also runs on server. Anim Notify is only client
          )
          (:OnNotifyBegin
            (Camera|PlayWorldCameraShake (Variables|CameraShakes|GetCameraShakeSlam) (Transformation|GetActorLocation) 5.0 7e+02)))))))

(event Custom|DashDamageBuilding
  (params (Actor :type Actor%20Object%20Reference))
  (bind _hitbuilding (Variables|MechVehicle|GetHitBuilding))
  ;[1 1 1] Damage Building if dashing
  (bind _asbuilding_actor (Utilities|Casting|CastToBuildingActor Actor))
  (if (and (Variables|/Script/Engine.Actor|GetbCanBeDamaged _asbuilding_actor) (Gameplay|GetIsDashing))
    (if (not (Utilities|Array|ContainsItem _hitbuilding _asbuilding_actor))
      (Utilities|Array|Add _hitbuilding _asbuilding_actor)
      (GameplayEffects|ApplyGameplayEffectToTarget :self (Ability|GetAbilitySystemComponent (Variables|Components|GetDriverPawn)) :GameplayEffectClass (Variables|DamageGameplayEffects|GetGE_BuildingDamageOnDash) :Target (Ability|GetAbilitySystemComponent _asbuilding_actor) :Level 1.0))
    (else
      (Utilities|Array|Clear _hitbuilding))))

(event Custom|Stomp
  ;[1 1 1] EVENT -> Driver Primary Fire - KICK!
  (GameplayCue|ExecuteGameplayCue%28Burst%29Local self (Variables|GameplayCue|GetGC_Stomp) (Ability|GameplayCue|MakeGameplayCueParameters :NormalizedMagnitude 1.0 :RawMagnitude 1.0 :Location "0,0,0" :Normal "0,0,0" :Instigator self :EffectCauser self :GameplayEffectLevel 1 :AbilityLevel 1 :bReplicateLocationWhenUsingMinimalRepProxy false))
  (Gameplay|OnKickFinished)
  (Variables|MechVehicle|SetKicking false)
  (CallFunction|KickDestroyandKnockActors)
  (Fortnite|PlayLocalForceFeedbackatLocationMulti (Variables|ForceFeedback|GetDetonateForceFeedbackNear) (Variables|ForceFeedback|GetDetonateForceFeedbackFar) (Transformation|GetActorLocation self) 1e+02 5e+02 "OstrichStomp") ; Gamepad Vibrate
)

(event EventOnDriverSecondaryFire
  ;[1 1 1] EVENT -> Driver Primary Fire - KICK!
  ;;[1 1 1] EVENT -> Driver Secondary Fire
  (Audio|SpawnSoundAttached "/Game/Athena/Sounds/Vehicles/Ostrich/Ostrich_Horn_Cue.Ostrich_Horn_Cue" (Mesh|GetMechMesh)))

(event EventOnPressHonk
  ;[1 1 1] EVENT -> Driver Primary Fire - KICK!
  ;;[1 1 1] EVENT -> Driver Secondary Fire
  (Audio|SpawnSoundAttached "/Game/Athena/Sounds/Vehicles/Ostrich/Ostrich_Horn_Cue.Ostrich_Horn_Cue" (Mesh|GetMechMesh)))

(event Custom|JumpSlam
  (bind (_returnvalue_x _returnvalue_y) (Math|Vector|BreakVector (* (Transformation|GetActorForwardVector) (Variables|MechVehicle|GetJumpSlamForwardForce))))
  ;[1 1 1] Ground Slam if in air
  (if (not (Vortex|IsinVortex (Variables|Components|GetDriverPawn)))
    (Utilities|CanJumpSlam)
    (Character|LaunchCharacter 0 (Math|Vector|MakeVector _returnvalue_x _returnvalue_y (Variables|MechVehicle|GetJumpSlamDownForce)) true true)
    (Variables|MechVehicle|SetbGroundSlam true) ; Need For Animation
    (Character|AssignMovementModeChangedDelegate (delegate Custom|MovementModeChangedDelegate_Event_0))
    (GameplayCue|ExecuteGameplayCue%28Burst%29Local self (Variables|GameplayCue|GetGC_JumpSlamActivated) (CallFunction|GCParams))))

(event Custom|MovementModeChangedDelegate_Event_0
  (params (Character :type Character%20Object%20Reference) (PrevMovementMode :type EMovementMode%20Enum :default "MOVE_None") (PreviousCustomMode :type Byte))
  ;[1 1 1] Ground Slam if in air
  (if (Utilities|Enum|Equal(Enum) (Variables|/Script/Engine.CharacterMovementComponent|GetMovementMode (Variables|/Script/Engine.Character|GetCharacterMovement Character)) "MOVE_Walking")
    (if (not (Variables|MechVehicle|GetDashed))
      (CallFunction|JumpSlamDamageandLaunch))
    (GameplayCue|ExecuteGameplayCue%28Burst%29Local self (Variables|GameplayCue|GetGC_JumpSlamLanded) (CallFunction|GCParams))
    (Character|UnbindEventfromMovementModeChangedDelegate (delegate Custom|MovementModeChangedDelegate_Event_0))
    (Variables|MechVehicle|SetbGroundSlam false) ; Need For Animation
    (Camera|PlayWorldCameraShake (Variables|CameraShakes|GetCameraShakeSlam) (Transformation|GetActorLocation) 5.0 7e+02)))

(event EventOnDetonateTimerStart
  (params (InDetonationState :type EOstrichDetonationState%20Enum))
  (bind _shieldbubbleactor (Variables|Components|GetShieldBubbleActor))
  ;[1 1 1] EVENT -> Detonate Timer Start
  (Utilities|FlowControl|DoOnce false
    (:Completed
      (if (Utilities|Enum|Equal(Enum) InDetonationState "Instant")
        (CallFunction|Detonate)
        (elif (Utilities|Enum|Equal(Enum) InDetonationState "Detonate")
          (Utilities|Time|SetTimerbyEvent (EventDispatchers|CreateEvent :!event Custom|Detonate) (Variables|Vehicle|GetDetonateDelay))
          (else
            (Utilities|Time|SetTimerbyEvent (EventDispatchers|CreateEvent :!event Custom|Detonate) (Variables|Vehicle|GetDetonateDelaySelfDestruct)) ; If we were self destructed wait a different time
            (bind _returnvalue (Effects|SpawnEmitterAttached :EmitterTemplate (Variables|FX|GetP_DetonationBuildupSelfDestruct) :AttachToComponent (Mesh|GetMechMesh) :AttachPointName "None" :Scale "1.000000,1.000000,1.000000" :LocationType "SnapToTarget" :bAutoDestroy true :PoolingMethod "None" :bAutoActivate true))
            (Variables|Default|SetKillBuildupSelfDestruct _returnvalue)
            (GameplayCue|ExecuteGameplayCue%28Burst%29Local self (Variables|GameplayCue|GetGC_Death) (CallFunction|GCParams)) ; Audio FX for Detonation Begin
            (Actor|DestroyActor _shieldbubbleactor)
            (Rendering|SetActorHiddenInGame _shieldbubbleactor true)
            (Networking|ForceNetUpdate _shieldbubbleactor)
            (Networking|ForceNetUpdate)
            (bind _returnvalue_2 (Utilities|Time|SetTimerbyEvent (delegate Custom|PulseSoundIndicator) 0.2 true))
            (Variables|MechVehicle|SetVisualizePulseIndicatorTimer _returnvalue_2))))
      (bind _returnvalue_3 (Effects|SpawnEmitterAttached :EmitterTemplate (Variables|FX|GetP_DetonationBuildup) :AttachToComponent (Mesh|GetMechMesh) :AttachPointName "None" :Scale "1.000000,1.000000,1.000000" :LocationType "SnapToTarget" :bAutoDestroy true :PoolingMethod "None" :bAutoActivate true))
      (Variables|Default|SetKillBuildup _returnvalue_3))))

(event Custom|PulseSoundIndicator
  ;[1 1 1] EVENT -> Detonate Timer Start
  (Fortnite|ShowSoundIndicatorLocalforTeam :Location (Transformation|GetActorLocation self) :MaxAudibleDistance 2e+03 :IndicatorType "Gunshot" :Team (Team|GetActorTeam self) :Affiliations (Utilities|Array|MakeArray "Hostile" "Neutral" "Friendly") :Instigator self :Tint "(R=0.000000,G=0.000000,B=0.000000,A=0.000000)" :OverrideIcon "/Game/UI/Foundation/Textures/Icons/Navigation/Mech_Visualizer_icon_32.Mech_Visualizer_icon_32" :AttenuationMaxRangeMultiplier 1.0 :AttenuationIntensityMultiplier 1.0))

(event Custom|Detonate
  ;[1 1 1] EVENT -> Detonate Timer Start
  (Components|DestroyComponent (Variables|Default|GetKillBuildup))
  (Fortnite|PlayLocalForceFeedbackatLocationMulti (Variables|ForceFeedback|GetDetonateForceFeedbackNear) (Variables|ForceFeedback|GetDetonateForceFeedbackFar) (Transformation|GetActorLocation self) 4e+02 1.4e+03 "OstrichDetonate") ; Gamepad Vibrate
  (GameplayCue|ExecuteGameplayCue%28Burst%29Local self (Variables|GameplayCue|GetGC_DeathExplode) (CallFunction|GCParams)) ; Handle Death FX and Audio
  (Fortnite|ShowSoundIndicatorLocalforTeam :Location (Transformation|GetActorLocation self) :MaxAudibleDistance 2e+03 :IndicatorType "Gunshot" :Team (Team|GetActorTeam self) :Affiliations (Utilities|Array|MakeArray "Hostile" "Neutral" "Friendly") :Instigator self :Tint "(R=0.000000,G=0.000000,B=0.000000,A=0.000000)" :OverrideIcon "/Game/UI/Foundation/Textures/Icons/Navigation/Mech_Visualizer_icon_32.Mech_Visualizer_icon_32" :AttenuationMaxRangeMultiplier 1.0 :AttenuationIntensityMultiplier 1.0)
  (Utilities|Time|ClearandInvalidateTimerbyHandle (Variables|MechVehicle|GetVisualizePulseIndicatorTimer))
  (Utilities|FlowControl|Delay 0.5) ; This so mech doesn't go invisible before fx
  (Rendering|SetActorHiddenInGame true)
  (Components|Activation|Deactivate (Variables|ForceFeedback|GetFullJumpChargeForceFeedback)) ; failsafe to ensure rumble ends if mech is killed
)

(event EventOnStartChargeJump
  ;[1 1 1] EVENT -> Jump Start Charge
  (Projectile|AddGameplayCue%28Looping%29Local self (Variables|GameplayCue|GetGC_JumpCharging) (CallFunction|GCParams))
  (bind _returnvalue (Utilities|Time|SetTimerbyEvent (EventDispatchers|CreateEvent :!event Custom|ChargeJumpCameraShake) (Variables|MechVehicle|GetFullChargeJumpTime)))
  (Variables|MechVehicle|SetFullChargeJumpTimer _returnvalue)
  (Components|Activation|Activate (Variables|ForceFeedback|GetFullJumpChargeForceFeedback) false)
  (bind (_direction _rumbleintensity) (Timeline|Timeline_0 PlayFromStart)
    (:Update
      (ForceFeedback|SetIntensityMultiplier (Variables|ForceFeedback|GetFullJumpChargeForceFeedback) _rumbleintensity))))

(event Custom|ChargeJumpCameraShake
  ;[1 1 1] CameraShake and Full Jump Charge
  (bind _asfort_player_controller (Utilities|Casting|CastToFortPlayerController (Pawn|GetController)))
  (bind _returnvalue (CameraShakes|StartCameraShake :self (Variables|/Script/Engine.PlayerController|GetPlayerCameraManager _asfort_player_controller) :ShakeClass (Variables|CameraShakes|GetCameraShakeFullChargeJump) :Scale 1.0 :PlaySpace "CameraLocal"))
  (Variables|CameraShakes|SetFullChargeJumpCameraShake (CameraShakes|ConvLegacyCameraShake _returnvalue))
  (Fortnite|PlayLocalForceFeedbackatLocation (Variables|ForceFeedback|GetDetonateForceFeedbackNear) (Transformation|GetActorLocation self) 11.0 "JumpChargeFull"))

(event EventOnReleaseChargeJump
  (params (ChargeAlpha :type Float%20%28single-precision%29))
  (bind _fullchargejumpcamerashake (Variables|CameraShakes|GetFullChargeJumpCameraShake))
  ;[1 1 1] EVENT -> Jump Charge Release
  (Variables|CameraShakes|SetCameraShakeLandScale (* ChargeAlpha (Variables|CameraShakes|GetCameraLandShakeMultiplier)))
  (Projectile|RemoveGameplayCue%28Looping%29Local self (Variables|GameplayCue|GetGC_JumpCharging) (Ability|GameplayCue|MakeGameplayCueParameters :NormalizedMagnitude 1.0 :RawMagnitude ChargeAlpha :Location "0,0,0" :Normal "0,0,0" :Instigator self :EffectCauser self :GameplayEffectLevel 1 :AbilityLevel 1 :bReplicateLocationWhenUsingMinimalRepProxy false))
  (Utilities|Time|ClearandInvalidateTimerbyHandle (Variables|MechVehicle|GetFullChargeJumpTimer))
  (Components|Activation|Deactivate (Variables|ForceFeedback|GetFullJumpChargeForceFeedback))
  (GameplayCue|ExecuteGameplayCue%28Burst%29Local self (Variables|FX|GetGroundShakeGC) (Ability|GameplayCue|MakeGameplayCueParameters 0.0 0.5)) ; Camera shake + controller vibration
  (bind _asfort_player_controller (Utilities|Casting|CastToFortPlayerController (Pawn|GetController)))
  (CameraShakes|StartCameraShake :self (Variables|/Script/Engine.PlayerController|GetPlayerCameraManager _asfort_player_controller) :ShakeClass (Variables|CameraShakes|GetCameraShakeJumpRelease) :Scale 1.0 :PlaySpace "CameraLocal")
  (Utilities|IsValid _fullchargejumpcamerashake
    (:Is%20Valid
      (CameraShakes|StopCameraShake (Variables|/Script/Engine.PlayerController|GetPlayerCameraManager _asfort_player_controller) _fullchargejumpcamerashake true))))

(event Custom|EventOnHealthChanged
  ;[1 1 1] On Health Changed -> Update Damage State
  (CallFunction|UpdateDamageState (Math|Float|Lerp 1.0 0.0 (/ (Class|FortVehicleInterface|GetHealth) (Class|FortVehicleInterface|GetMaxHealth)))))

(event Custom|UpdateDamageState
  (params (Damage :type Float%20%28double-precision%29))
  (bind _returnvalue (> Damage 0.8))
  (bind _damagedeffect (Variables|Default|GetDamagedEffect))
  ;[1 1 1] Update Damage State
  (Networking|SwitchHasAuthority
    (:Remote
      (CallFunction|SendDamageStatetoMaterial Damage)
      (if (> Damage 0.64)
        (Variables|Default|GetDamagedEffect
          (:then
            _)
          (:else
            (bind _returnvalue_2 (Effects|SpawnEmitterAttached :EmitterTemplate "/Game/Effects/Fort_Effects/Gameplay/Ostrich/P_Ostrich_Damaged.P_Ostrich_Damaged" :AttachToComponent (Variables|Character|GetMesh) :AttachPointName "center" :Scale "1.000000,1.000000,1.000000" :LocationType "SnapToTarget" :bAutoDestroy true :PoolingMethod "None" :bAutoActivate true))
            (bind _output_get (Variables|Default|SetDamagedEffect _returnvalue_2))
            (EventDispatchers|BindEventtoOnParticleSpawn _output_get (delegate Custom|LightningSpark))))
        (Effects|Components|ParticleSystem|SetFloatParameter _damagedeffect "BadlyDamaged" (Math|Conversions|ToFloat(Boolean) _returnvalue))
        (Effects|Components|ParticleSystem|SetFloatParameter _damagedeffect "Ribbon" (Math|Conversions|ToFloat(Boolean) (and _returnvalue (not (Gameplay|GetIsDashing)))))))))

(event Custom|LightningSpark
  (params (EventName :type Name) (EmitterTime :type Float%20%28single-precision%29) (Location :type Vector) (Velocity :type Vector))
  ;[1 1 1] Update Damage State
  (if (Utilities|Name|Equal(Name) EventName "spark")
    (Audio|Components|Audio|Play (Variables|Default|GetAudioSpark) 0.0)))

(event EventOnFailedShieldBoost
  ;[1 1 1] EVENT -> Failed Shield Boost
  (Audio|PlaySound2D "/Game/Sounds/Creative/Toys/ShootingTargets/Target_Error_Cue.Target_Error_Cue"))

(event EventOnLoadingRockets
  (params (bStarting :type Boolean))
  ;[1 1 1] EVENT -> On Loading Rockets
  (if bStarting
    (Variables|Default|GetTargetingLaserAudio
      (:then)
      (:else
        (bind _returnvalue (Audio|SpawnSoundAttached :Sound "/Game/Athena/Sounds/Vehicles/Ostrich/Veh_Ostrich_TargetLaser_Start_Cue.Veh_Ostrich_TargetLaser_Start_Cue" :AttachToComponent (Variables|Character|GetMesh) :AttachPointName "Laser" :Rotation "0,90.000000, 0" :LocationType "KeepRelativeOffset" :bStopWhenAttachedToDestroyed true :VolumeMultiplier 1.0 :PitchMultiplier 1.0 :StartTime 0.0 :bAutoDestroy true))
        (bind _output_get (Variables|Default|SetTargetingLaserAudio _returnvalue))
        (Gameplay|SetTargetingLaserAudioComp _output_get)))
    (else
      (bind _targetinglaseraudio (Variables|Default|GetTargetingLaserAudio)
        (:then
          (Audio|Components|Audio|FadeOut _targetinglaseraudio 0.25 0.0 "Linear" :!class /Script/Engine.AudioComponent)
          (Audio|SpawnSoundAttached "/Game/Athena/Sounds/Vehicles/Ostrich/Veh_Ostrich_TargetLaser_Stop_Cue.Veh_Ostrich_TargetLaser_Stop_Cue" (Variables|Character|GetMesh) "Laser")
          (Gameplay|SetTargetingLaserAudioComp 0))
        (:else)))))

(fn CanDamageBuilding
  (params (BuildingActor :type Building%20Actor%20Object%20Reference) (MaxHealthThreshold :type Float%20%28double-precision%29))
  (return (Math|Boolean|ANDBoolean (> (Health|GetMaxHealth BuildingActor :!class /Script/FortniteGame.BuildingActor) 0.0) (<= (Health|GetMaxHealth BuildingActor :!class /Script/FortniteGame.BuildingActor) MaxHealthThreshold) (not (Actor|ActorHasTag BuildingActor "NoVehicleDamage")) (Variables|/Script/Engine.Actor|GetbCanBeDamaged BuildingActor) (not (GameplayTags|HasMatchingGameplayTag BuildingActor "(TagName=\"Abilities.Generic.Damage.Immune\")"))) BuildingActor))

(fn DamageAndLaunchActors
  (params (Actor :type Actor%20Object%20Reference))
  (bind _returnvalue (Fortnite|CheckLineOfSightToActor (+ (Transformation|GetSocketLocation (Mesh|GetMechMesh) "paw_toetip_r") "0, 0,60.000000") Actor)) ; LoS Check
  (if _returnvalue
    (if (not (Utilities|Array|ContainsItem (LocalVariables|GetLandedHitActors) Actor))
      (Utilities|Array|Add (LocalVariables|GetLandedHitActors) Actor)
      ;[1 1 1] Damage Player if they aren't on your team
      (bind _asfort_player_pawn (Utilities|Casting|CastToFortPlayerPawn Actor)
        (:then
          (if (Math|Boolean|ANDBoolean (Utilities|NotEqual(Object) _asfort_player_pawn self) (Utilities|NotEqual(Object) _asfort_player_pawn (Variables|Components|GetDriverPawn)) (Utilities|NotEqual(Object) (Variables|Components|GetPassengerPawn) _asfort_player_pawn))
            (if (Vehicle|IsinVehicle _asfort_player_pawn)
              (Utilities|Casting|CastToTestMechVehicle (Vehicle|GetVehicleActor _asfort_player_pawn)
                (:CastFailed
                  _)))
            (if (Team|OnSameTeam _asfort_player_pawn (Variables|Components|GetDriverPawn))
              _)
            ;[1 1 1] Launch Player
            (Utilities|LaunchPlayerPawn (Variables|Components|GetDriverPawn) _asfort_player_pawn (CallFunction|KnockAwayDirection _asfort_player_pawn (Variables|MechVehicle|GetKickKnockMagnitude) (Variables|MechVehicle|GetKickPlayerZLaunch)) true true false true false true true)))
        (:CastFailed
          ;[1 1 1] Damage and Knock Vehicle
          (bind _asfort_vehicle_interface (Utilities|Casting|CastToFortVehicleInterface Actor))
          (bind _returnvalue_2 (Vehicle|GetVehicleActor _asfort_vehicle_interface :!class /Script/FortniteGame.FortVehicleInterface))
          (if (Utilities|NotEqual(Object) _returnvalue_2 self)
            (GameplayEffects|ApplyGameplayEffectToTarget :self (Ability|GetAbilitySystemComponent (Variables|Components|GetDriverPawn)) :GameplayEffectClass (Variables|DamageGameplayEffects|GetGE_VehicleDamageLanded) :Target (Ability|GetAbilitySystemComponent _returnvalue_2) :Level 1.0)
            (Utilities|Casting|CastToTestMechVehicle _asfort_vehicle_interface
              (:CastFailed
                (Audio|BroadcastAbilityImpactatLocation 0 _returnvalue_2 (CallFunction|KnockAwayDirection _returnvalue_2 (Variables|MechVehicle|GetKickKnockVehicleMagnitude) (Variables|MechVehicle|GetVehicleKnockUpMagnitude)) 777.0)))))))))

(fn DoesDamageCauserPreventDamageFromBeingNulled
  (params (DamageCauser :type Actor%20Object%20Reference))
  (Utilities|FlowControl|Sequence
    (:then_0
      (bind _damagecauser (Variables|GetDamageCauser)
        (:then
          (Utilities|Casting|CastToDevice_Floor_EliminationZone _damagecauser)
          (return true))
        (:else)))
    (:then_1
      (return false))))

(fn FormatActionLabelText
  (params (InAction :type Action%20Def%20To%20Display%20Structure%20%28by%20ref%29))
  (bind (_legacyinputaction _enhancedinputaction _inputactions _actionlabel) (Utilities|Struct|BreakActionDeftoDisplay InAction))
  (if (or (Utilities|Name|Equal(Name) _legacyinputaction "OstrichShieldBoost") (Utilities|Name|Equal(Name) _legacyinputaction "GamepadOstrichShieldBoost"))
    (LocalVariables|SetReturnText (Utilities|Text|FormatText _actionlabel (ScalableFloat|GetValueatLevel (Variables|HotfixableValues|GetOverchargeMaterialCost))))
    (else
      (LocalVariables|SetReturnText _actionlabel)))
  (return (LocalVariables|GetReturnText)))

(fn GC_Params
  (return (Ability|GameplayCue|MakeGameplayCueParameters :NormalizedMagnitude 1.0 :RawMagnitude 1.0 :Location "0,0,0" :Normal "0,0,0" :Instigator self :EffectCauser self :GameplayEffectLevel 1 :AbilityLevel 1 :bReplicateLocationWhenUsingMinimalRepProxy false)))

(fn GetChargeJumpingFloat
  (return (Variables|MechVehicle|GetChargeJumping)))

(fn GetGroundSlamBool
  (return (Variables|MechVehicle|GetbGroundSlam)))

(fn IsGunnerShieldActive
  (bind _shieldbubbleactor (Variables|Components|GetShieldBubbleActor))
  (Utilities|IsValid _shieldbubbleactor
    (:Is%20Valid
      (bind _return (Class|BGAOstrichShieldBubble|IsShieldActive _shieldbubbleactor))
      (return _return))))

(fn JumpSlamDamageAndLaunch
  (bind (_returnvalue_roll _returnvalue_pitch _returnvalue_yaw) (Math|Rotator|BreakRotator (Pawn|GetControlRotation)))
  (bind _driverpawn (Variables|Components|GetDriverPawn))
  (bind _hitactors (Variables|Default|GetHitActors))
  (Utilities|Set|Clear (Variables|Default|GetHitActors))
  (bind _outhits (Collision|MultiBoxTraceForObjects :Start (Transformation|GetActorLocation) :End (+ (Transformation|GetActorLocation) "0.000000,0.000000,-350.000000") :HalfSize (Variables|MechVehicle|GetJumpSlamBoxSize) :Orientation (Math|Rotator|MakeRotator 0.0 0.0 _returnvalue_yaw) :ObjectTypes (Utilities|Array|MakeArray "ObjectTypeQuery1" "ObjectTypeQuery2" "ObjectTypeQuery3" "ObjectTypeQuery5" "ObjectTypeQuery7") :bTraceComplex false :DrawDebugType "None" :bIgnoreSelf true :TraceColor "(R=1.000000,G=0.000000,B=0.000000,A=1.000000)" :TraceHitColor "(R=0.000000,G=1.000000,B=0.000000,A=1.000000)" :DrawTime 5.0))
  (for _array_element _outhits
    (if (not (Utilities|Set|ContainsItem _hitactors (.HitActor (Collision|BreakHitResult _array_element))))
      (Utilities|Set|Add _hitactors (.HitActor (Collision|BreakHitResult _array_element)))))
  (bind _result (Utilities|Set|ToArray (Variables|Default|GetHitActors)))
  (for _array_element_2 _result
    (bind _returnvalue (Fortnite|CheckLineOfSightToActor (+ (Transformation|GetSocketLocation (Mesh|GetMechMesh) "paw_toetip_r") "0, 0,60.000000") _array_element_2)) ; LoS Check
    (if _returnvalue
      ;[1 1 1] Damage Player if they aren't on your team
      (bind _asfort_player_pawn (Utilities|Casting|CastToFortPlayerPawn _array_element_2)
        (:then
          (if (Math|Boolean|ANDBoolean (Utilities|NotEqual(Object) _asfort_player_pawn self) (Utilities|NotEqual(Object) _asfort_player_pawn (Variables|Components|GetDriverPawn)) (Utilities|NotEqual(Object) (Variables|Components|GetPassengerPawn) _asfort_player_pawn))
            (if (Vehicle|IsinVehicle _asfort_player_pawn)
              (Utilities|Casting|CastToTestMechVehicle (Vehicle|GetVehicleActor _asfort_player_pawn)
                (:CastFailed
                  _)))
            (if (not (Team|OnSameTeam _asfort_player_pawn (Variables|Components|GetDriverPawn)))
              (GameplayEffects|ApplyGameplayEffectToTarget :self (Ability|GetAbilitySystemComponent (Variables|Components|GetDriverPawn)) :GameplayEffectClass (Variables|DamageGameplayEffects|GetGE_KickPlayerDamage) :Target (Ability|GetAbilitySystemComponent _asfort_player_pawn) :Level 1.0))
            (Utilities|LaunchPlayerPawn (Variables|Components|GetDriverPawn) _asfort_player_pawn (CallFunction|KnockAwayDirection _asfort_player_pawn (Variables|MechVehicle|GetKickKnockMagnitude) (Variables|MechVehicle|GetKickPlayerZLaunch)) true true false true false true true)))
        (:CastFailed
          ;[1 1 1] Damage and Knock Vehicle
          (bind _asfort_vehicle_interface (Utilities|Casting|CastToFortVehicleInterface _array_element_2)
            (:then
              (bind _returnvalue_2 (Vehicle|GetVehicleActor _asfort_vehicle_interface :!class /Script/FortniteGame.FortVehicleInterface))
              (if (Utilities|NotEqual(Object) _returnvalue_2 self)
                (Environment|ApplyGlobalEnvironmentGameplayEffecttoActor _returnvalue_2 (Variables|DamageGameplayEffects|GetGE_VehicleDamageLanded) 1)
                (Utilities|Casting|CastToTestMechVehicle _asfort_vehicle_interface
                  (:CastFailed
                    (Audio|BroadcastAbilityImpactatLocation 0 _returnvalue_2 (CallFunction|KnockAwayDirection _returnvalue_2 (Variables|MechVehicle|GetKickKnockVehicleMagnitude) (Variables|MechVehicle|GetVehicleKnockUpMagnitude)) 777.0)))))
            (:CastFailed
              ;[1 1 1] Damage AI Pawns
              (bind _asfort_aipawn (Utilities|Casting|CastToFortAIPawn _array_element_2)
                (:then
                  (GameplayEffects|ApplyGameplayEffectToTarget :self (Ability|GetAbilitySystemComponent (Variables|Components|GetDriverPawn)) :GameplayEffectClass (Variables|DamageGameplayEffects|GetGE_KickPlayerDamage) :Target (Ability|GetAbilitySystemComponent _asfort_aipawn) :Level 1.0))
                (:CastFailed
                  (bind _asbuilding_actor (Utilities|Casting|CastToBuildingActor _array_element_2))
                  (if (CallFunction|CanDamageBuilding _asbuilding_actor 5e+03)
                    (Utilities|Array|Add (Variables|Default|GetEnvironmentalActorsHit) (.Building_Actor (CallFunction|CanDamageBuilding _asbuilding_actor 5e+03))))))))))))
  ;[1 1 1] Destroy Buildings. Have to loop for buildings AFTER players so player's don't get knocked if they weren't in Line of sight of stomp
  (for _array_element_3 (Variables|Default|GetEnvironmentalActorsHit)
    (Utilities|IsValid _driverpawn
      (:Is%20Valid
        (GameplayEffects|ApplyGameplayEffectToTarget (Ability|GetAbilitySystemComponent _driverpawn) (Variables|DamageGameplayEffects|GetGE_JumpSlamDamageBuilding) (Ability|GetAbilitySystemComponent _array_element_3) 1.0 (Utilities|Struct|MakeGameplayEffectContextHandle)))))
  (Utilities|Set|Clear (Variables|Default|GetHitActors)))

(fn KickDestroyAndKnockActors
  (bind (_returnvalue_roll _returnvalue_pitch _returnvalue_yaw) (Math|Rotator|BreakRotator (Pawn|GetControlRotation)))
  (Networking|SwitchHasAuthority
    (:Authority
      (bind _outhits (Collision|MultiBoxTraceForObjects :Start (+ (+ (Transformation|GetActorLocation) (* (Transformation|GetActorForwardVector) (Variables|MechVehicle|GetKickTraceForwardOffset))) "0.000000,0.000000,-80.000000") :End (+ (+ (+ (Transformation|GetActorLocation) (* (Transformation|GetActorForwardVector) (Variables|MechVehicle|GetKickTraceForwardOffset))) "0.000000,0.000000,-80.000000") (* (Transformation|GetActorForwardVector) 1e+02)) :HalfSize (Variables|MechVehicle|GetKickBoxSize) :Orientation (Math|Rotator|MakeRotator 0.0 0.0 _returnvalue_yaw) :ObjectTypes (Utilities|Array|MakeArray "ObjectTypeQuery1" "ObjectTypeQuery2" "ObjectTypeQuery5" "ObjectTypeQuery3" "ObjectTypeQuery7") :bTraceComplex false :DrawDebugType "None" :bIgnoreSelf true :TraceColor "(R=1.000000,G=0.000000,B=0.000000,A=1.000000)" :TraceHitColor "(R=0.000000,G=1.000000,B=0.000000,A=1.000000)" :DrawTime 5.0))
      (for _array_element _outhits
        (if (not (Utilities|Array|ContainsItem (LocalVariables|GetKickHitActors) (.HitActor (Collision|BreakHitResult _array_element))))
          (Utilities|Array|Add (LocalVariables|GetKickHitActors) (.HitActor (Collision|BreakHitResult _array_element)))
          (bind _returnvalue (Fortnite|CheckLineOfSightToActor (+ (Transformation|GetSocketLocation (Mesh|GetMechMesh) "paw_toetip_r") "0, 0,60.000000") (.HitActor (Collision|BreakHitResult _array_element)))) ; LoS Check
          (if _returnvalue
            ;[1 1 1] Damage Player if they aren't on your team
            (bind _asfort_player_pawn (Utilities|Casting|CastToFortPlayerPawn (.HitActor (Collision|BreakHitResult _array_element)))
              (:then
                (if (Math|Boolean|ANDBoolean (Utilities|NotEqual(Object) _asfort_player_pawn self) (Utilities|NotEqual(Object) _asfort_player_pawn (Variables|Components|GetDriverPawn)) (Utilities|NotEqual(Object) (Variables|Components|GetPassengerPawn) _asfort_player_pawn))
                  (if (Vehicle|IsinVehicle _asfort_player_pawn)
                    (Utilities|Casting|CastToTestMechVehicle (Vehicle|GetVehicleActor _asfort_player_pawn)
                      (:CastFailed
                        _)))
                  (if (not (Team|OnSameTeam _asfort_player_pawn (Variables|Components|GetDriverPawn)))
                    (GameplayEffects|ApplyGameplayEffectToTarget (Ability|GetAbilitySystemComponent (Variables|Components|GetDriverPawn)) (Variables|DamageGameplayEffects|GetGE_KickPlayerDamage) (Ability|GetAbilitySystemComponent _asfort_player_pawn) 1.0 (Utilities|Struct|MakeGameplayEffectContextHandle)))
                  (Utilities|LaunchPlayerPawn (Variables|Components|GetDriverPawn) _asfort_player_pawn (CallFunction|KnockAwayDirection _asfort_player_pawn (Variables|MechVehicle|GetKickKnockMagnitude) (Variables|MechVehicle|GetKickPlayerZLaunch)) true true false true false true true)))
              (:CastFailed
                ;[1 1 1] Damage and Knock Vehicle
                (bind _asfort_vehicle_interface (Utilities|Casting|CastToFortVehicleInterface (.HitActor (Collision|BreakHitResult _array_element)))
                  (:then
                    (bind _returnvalue_2 (Vehicle|GetVehicleActor _asfort_vehicle_interface :!class /Script/FortniteGame.FortVehicleInterface))
                    (if (Utilities|NotEqual(Object) _returnvalue_2 self)
                      (Environment|ApplyGlobalEnvironmentGameplayEffecttoActor _returnvalue_2 (Variables|DamageGameplayEffects|GetGE_VehicleDamageLanded) 1)
                      (Utilities|Casting|CastToTestMechVehicle _asfort_vehicle_interface
                        (:CastFailed
                          (Audio|BroadcastAbilityImpactatLocation 0 _returnvalue_2 (CallFunction|KnockAwayDirection _returnvalue_2 (Variables|MechVehicle|GetKickKnockVehicleMagnitude) (Variables|MechVehicle|GetVehicleKnockUpMagnitude)) 777.0)))))
                  (:CastFailed
                    ;[1 1 1] Damage AI Pawns
                    (bind _asfort_aipawn (Utilities|Casting|CastToFortAIPawn (.HitActor (Collision|BreakHitResult _array_element))))
                    (GameplayEffects|ApplyGameplayEffectToTarget :self (Ability|GetAbilitySystemComponent (Variables|Components|GetDriverPawn)) :GameplayEffectClass (Variables|DamageGameplayEffects|GetGE_KickPlayerDamage) :Target (Ability|GetAbilitySystemComponent _asfort_aipawn) :Level 1.0))))))))
      ;[1 1 1] Destroy Buildings. Have to loop for buildings AFTER players so player's don't get knocked if they weren't in Line of sight of stomp
      (for _array_element_2 _outhits
        (bind _asbuilding_actor (Utilities|Casting|CastToBuildingActor (.HitActor (Collision|BreakHitResult _array_element_2))))
        (if (CallFunction|CanDamageBuilding _asbuilding_actor 5e+03)
          (GameplayEffects|ApplyGameplayEffectToTarget (Ability|GetAbilitySystemComponent (Variables|Components|GetDriverPawn)) (Variables|DamageGameplayEffects|GetGE_DestroyBuildingDamage) (Ability|GetAbilitySystemComponent (.Building_Actor (CallFunction|CanDamageBuilding _asbuilding_actor 5e+03))) 1.0 (Utilities|Struct|MakeGameplayEffectContextHandle))))
      (Utilities|Array|Clear (LocalVariables|GetKickHitActors)) ; Clear array of Hit actors
    )))

(fn KnockAwayDirection
  (params (Target :type Actor%20Object%20Reference) (LaunchMagnitude :type Float%20%28double-precision%29) (ZAdditive :type Float%20%28double-precision%29))
  (return (Math|Vector|MakeVector (.x (* (Math|Vector|Normalize (- (Transformation|GetActorLocation Target) (Transformation|GetActorLocation))) LaunchMagnitude)) (.y (* (Math|Vector|Normalize (- (Transformation|GetActorLocation Target) (Transformation|GetActorLocation))) LaunchMagnitude)) (Math|Float|Max(Float) (+ (.z (* (Math|Vector|Normalize (- (Transformation|GetActorLocation Target) (Transformation|GetActorLocation))) LaunchMagnitude)) ZAdditive) 5e+02))))

(fn OnEnterWater
  (params (WaterBody :type Fort%20Water%20Body%20Actor%20Object%20Reference))
  (bind _output_get (Variables|Default|SetWater%20Body WaterBody))
  (Variables|FX|SetIsInWater true)
  (CallFunction|OnUpdateWater)
  (bind _outwatersurfaceinfo (WaterBody|GetWaterSurfaceInfoStruct _output_get (Transformation|GetActorLocation self) true false false))
  (GameplayCue|ExecuteGameplayCue%28Burst%29Local self (Variables|GameplayCue|GetGC_EnterWater) (Ability|GameplayCue|MakeGameplayCueParameters :NormalizedMagnitude (Math|Conversions|ToFloat(Integer) (.WaterBodyIdx (Utilities|Struct|BreakWaterSurfaceInfo _outwatersurfaceinfo))) :RawMagnitude (.WaterDepth (Utilities|Struct|BreakWaterSurfaceInfo _outwatersurfaceinfo)) :Location (Utilities|Struct|BreakWaterSurfaceInfo _outwatersurfaceinfo) :Normal (.WaterPlaneNormal (Utilities|Struct|BreakWaterSurfaceInfo _outwatersurfaceinfo)) :GameplayEffectLevel 1 :AbilityLevel 1 :bReplicateLocationWhenUsingMinimalRepProxy false)))

(fn OnExitWater
  (Variables|FX|SetIsInWater false)
  (CallFunction|OnUpdateWater)
  (GameplayCue|ExecuteGameplayCue%28Burst%29Local self (Variables|GameplayCue|GetGC_ExitWater) (CallFunction|GCParams)))

(fn OnRep_ChargeJumping
  (bind _returnvalue (CallFunction|GCParams))
  (bind _self self)
  (Projectile|AddGameplayCue%28Looping%29Local _self (Variables|GameplayCue|GetGC_JumpInAirLooping) _returnvalue) ; Do looping in air GC also
  (GameplayCue|ExecuteGameplayCue%28Burst%29Local _self (Variables|GameplayCue|GetGC_JumpActivate) _returnvalue) ; Do Burst Jets after jumping
)

(fn OnRep_Dashing
  (GameplayCue|ExecuteGameplayCue%28Burst%29Local self (Variables|GameplayCue|GetGC_Dash) (CallFunction|GCParams)))

(fn OnRep_LandFX)

(fn OnRep_SetJumpApex
  (if (Variables|MechVehicle|GetSetJumpApex)
    (Variables|MechVehicle|SetJumpApexLocation (Transformation|GetActorLocation))
    (else
      (Variables|MechVehicle|SetJumpApexLocation "0, 0, 0")
      (Variables|MechVehicle|SetDashed false)
      (Variables|CameraShakes|SetCameraShakeLandScale 0.5) ; Reset Camera Shake Scalar
    )))

(fn OnRep_ShieldActive)

(fn OnRep_ShieldHealth)

(fn OnUpdateWater
  (bind _returnvalue (WaterBody|GetWaterImmersionDepth))
  (bind _returnvalue_2 (<= _returnvalue 4.5e+02))
  (bind _isinwater (Variables|FX|GetIsInWater))
  (bind _returnvalue_3 (and _returnvalue_2 _isinwater))
  (bind _inwaterfxreference (Variables|Default|GetInWaterFxReference))
  (bind _inwaterfx (Variables|Default|GetInWaterFX))
  (if (Networking|IsServer)
    _
    (elif (Math|Boolean|Equal(Boolean) (Utilities|IsValid (Variables|Default|GetInWaterFX) :!class /Script/Engine.KismetSystemLibrary) _returnvalue_3)
      ;[1 1 1] Update Water FX Parameters
      (bind _outwatersurfaceinfo (WaterBody|GetWaterSurfaceInfoStruct (Variables|Default|GetWater%20Body) (Transformation|GetActorLocation self) true false false))
      (Effects|Components|ParticleSystem|SetVectorParameter _inwaterfxreference "User.WaterPlaneLocation" (Utilities|Struct|BreakWaterSurfaceInfo _outwatersurfaceinfo))
      (Effects|Components|ParticleSystem|SetVectorParameter _inwaterfxreference "User.WaterPlaneNormal" (.WaterPlaneNormal (Utilities|Struct|BreakWaterSurfaceInfo _outwatersurfaceinfo)))
      (Effects|Components|ParticleSystem|SetFloatParameter _inwaterfxreference "User.WaterDepth" (.WaterDepth (Utilities|Struct|BreakWaterSurfaceInfo _outwatersurfaceinfo)))
      (Effects|Components|ParticleSystem|SetFloatParameter _inwaterfxreference "User.WaterBodyIndex" (Math|Conversions|ToFloat(Integer) (.WaterBodyIdx (Utilities|Struct|BreakWaterSurfaceInfo _outwatersurfaceinfo))))
      (elif _returnvalue_3
        ;[1 1 1] Activate Water
        (bind _returnvalue_4 (Effects|SpawnFXSystemAttached (Variables|FX|GetInWaterLoopingFX) (Variables|Character|GetMesh) "spine_03"))
        (bind _output_get (Variables|Default|SetInWaterFX _returnvalue_4))
        (bind _asniagara_particle_system_component (Utilities|Casting|CastToNiagaraComponent _output_get))
        (Variables|Default|SetInWaterFxReference _asniagara_particle_system_component)
        (elif (Utilities|IsValid _inwaterfx :!class /Script/Engine.KismetSystemLibrary)
          ;[1 1 1] We no longer have water active
          (Transformation|DetachFromComponent _inwaterfx "KeepWorld" "KeepWorld" "KeepWorld" true)
          (Components|Activation|Deactivate _inwaterfx)
          (Variables|Default|SetInWaterFX 0)
          (Variables|Default|SetInWaterFxReference 0))))))

(fn SendDamageStateToMaterial
  (params (ParameterValue :type Float%20%28double-precision%29))
  (bind _output_get (Variables|MechVehicle|SetSavedDamageForMID ParameterValue))
  (Rendering|Material|SetScalarParameterValueonMaterials (Variables|Character|GetMesh) "Damage" _output_get))

(fn ConstructionScript
  (Components|Tick|SetTickGroup (Mesh|GetEmoteMesh) "TG_DuringPhysics")
  (Components|Tick|SetTickGroup (Mesh|GetMechMesh) "TG_DuringPhysics")
  (Variables|MechVehicle|SetJumpKnockMagnitude (ScalableFloat|GetValueatLevel (Variables|HotfixableValues|GetSetJumpKnockMagnitude)))
  (Variables|MechVehicle|SetDashKnockStrength (ScalableFloat|GetValueatLevel (Variables|HotfixableValues|GetSetDashKnockStrength)))
  (Variables|MechVehicle|SetJumpSlamDownForce (ScalableFloat|GetValueatLevel (Variables|HotfixableValues|GetSetJumpSlamDownForce)))
  (Variables|MechVehicle|SetJumpSlamForwardForce (ScalableFloat|GetValueatLevel (Variables|HotfixableValues|GetSetJumpSlamForwardForce)))
  (Variables|MechVehicle|SetKickKnockMagnitude (ScalableFloat|GetValueatLevel (Variables|HotfixableValues|GetSetKickLaunchPlayerStrength)))
  (Variables|MechVehicle|SetKickKnockVehicleMagnitude (ScalableFloat|GetValueatLevel (Variables|HotfixableValues|GetSetKickLaunchVehicleStrength)))
  (Variables|MechVehicle|SetVehicleKnockUpMagnitude (ScalableFloat|GetValueatLevel (Variables|HotfixableValues|GetSetVehicleKnockUpStrength)))
  (Variables|MechVehicle|SetKickPlayerZLaunch (ScalableFloat|GetValueatLevel (Variables|HotfixableValues|GetSetKickPlayerZ_LaunchStrength)))
  (Variables|MechVehicle|SetDashZLaunchAdditive (ScalableFloat|GetValueatLevel (Variables|HotfixableValues|GetSetDashZ_LaunchPlayerAdditive)))
  (Variables|MechVehicle|SetMinHeightToJumpSlam (ScalableFloat|GetValueatLevel (Variables|HotfixableValues|GetSetMinHeightToJumpSlam)))
  (Variables|MechVehicle|SetLandedDamageRadius (ScalableFloat|GetValueatLevel (Variables|HotfixableValues|GetSetLandedDamageRadius)))
  (Variables|Vehicle|SetShieldBoostResourceCost (Math|Float|Truncate (ScalableFloat|GetValueatLevel (Variables|HotfixableValues|GetSetOverchargeShieldCost))))
  (Variables|Vehicle|SetShieldBoostCooldown (+ (ScalableFloat|GetValueatLevel (Variables|HotfixableValues|GetSetOverchargeDuration)) (ScalableFloat|GetValueatLevel (Variables|HotfixableValues|GetSetOverchargeCooldown))))
  (Gameplay|SetLinearAudioComp (Variables|Default|GetLinearMovementAudio))
  (Gameplay|SetRotationalAudioComp (Variables|Default|GetRotationalMovementAudio))
  (Variables|Vehicle|SetbEnableTargetingLaser (> (ScalableFloat|GetValueatLevel (Variables|HotfixableValues|GetSetEnableTargetingLaser)) 0.0))
  (Variables|Vehicle|SetSimulatedProxyFixupRange (ScalableFloat|GetValueatLevel (Variables|HotfixableValues|GetSetSimProxyFixRange)))
  (Variables|Vehicle|SetbEnableSimulatedProxyFixup (> (ScalableFloat|GetValueatLevel (Variables|HotfixableValues|GetSetEnableSimProxyFix)) 0.0))
  (Variables|Vehicle|SetDetonateDelay (ScalableFloat|GetValueatLevel (Variables|HotfixableValues|GetSetDetonateDelay)))
  (Variables|Vehicle|SetDetonateDelaySelfDestruct (ScalableFloat|GetValueatLevel (Variables|HotfixableValues|GetSetSelfDetonateDelay))))
